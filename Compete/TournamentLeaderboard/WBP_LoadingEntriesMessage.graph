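# Tournament leaderboard widget that shows a loading-entries message with a throbber while letting pointer input pass through.
# A hit-test-invisible grid layers a translucent rounded background, centered text and a material-driven throbber over a UIKit blur scrim.
# PreConstruct sets the text from the LoadingMessageText variable.
# It also copies the background's draw type and outline settings onto the scrim's OverlayBlur brush, so the blur takes the same rounded-box shape.
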WidgetBlueprint /Compete/TournamentLeaderboard/WBP_LoadingEntriesMessage
graphs [EventGraph {nodes [
   Event PreConstruct()
   > "SetText (Text)"(
    self=Get LoadingLeaderboardEntries() -> (LoadingLeaderboardEntries:Common Text Object Reference)
    InText=Get LoadingMessageText() -> (LoadingMessageText:Text)
   )
   n1 = Get OverlayBlur!(self:WBP UIKit Scrim Object Reference=Get LoadingEntriesBackgroundBlur() -> (LoadingEntriesBackgroundBlur:WBP UIKit Scrim Object Reference)) -> (OverlayBlur:Image Object Reference)
   n2 = Get!(self:Image Object Reference=@n1) -> (
    Brush_TintColor:Slate Color Structure
    Brush_DrawAs:ESlateBrushDrawType Enum
    Brush_Tiling:ESlateBrushTileType Enum
    Brush_Mirroring:ESlateBrushMirrorType Enum
    Brush_ImageSize:Vector 2D Structure
    Brush_Margin:Margin Structure
    Brush_ResourceObject:Object Reference
    Brush_OutlineSettings:Slate Brush Outline Settings Structure
   )
   n3 = Get(self=Get LoadingEntriesBackground() -> (LoadingEntriesBackground:Image Object Reference))
   > SetBrush!(
    execute
    self=@n1
    InBrush_TintColor:Slate Color Structure=@n2
    InBrush_DrawAs:ESlateBrushDrawType Enum=@n3.Brush_DrawAs
    InBrush_Tiling:ESlateBrushTileType Enum=@n2.Brush_Tiling
    InBrush_Mirroring:ESlateBrushMirrorType Enum=@n2.Brush_Mirroring
    InBrush_ImageSize:Vector 2D Structure=@n2.Brush_ImageSize
    InBrush_Margin:Margin Structure=@n2.Brush_Margin
    InBrush_ResourceObject:Object Reference=@n2.Brush_ResourceObject
    InBrush_OutlineSettings:Slate Brush Outline Settings Structure=@n3.Brush_OutlineSettings
   ) -> (then)
  ]}]
variables [LoadingMessageText {type=string, default=Loading Entries Message}]
widgets {
 info {inheritedWidgetCount=0, parentClass=&CommonUserWidget, rootWidgetClass=&GridPanel, widgetCount=5}
 widgets [
  LeaderboradLoadingEntries {widgetClassPath=&GridPanel, widget=LeaderboradLoadingEntries, properties {visibility=HitTestInvisible}}
  LoadingEntriesBackgroundBlur {
   widgetClassPath=/FortUILibrary/Systems/Dialog/WBP_UIKit_Scrim
   bIsVariable=true
   parent=LeaderboradLoadingEntries
   widget=LoadingEntriesBackgroundBlur
   properties {blur=0.04}
   slot {class=GridSlot, properties {columnSpan=2}}
  }
  LoadingEntriesBackground {
   widgetClassPath=&Image
   bIsVariable=true
   parent=LeaderboradLoadingEntries
   widget=LoadingEntriesBackground
   properties {
    brush {
     bIsDynamicallyLoaded=false
     drawAs=RoundedBox
     imageType=FullColor
     mirroring=NoMirror
     resourceName=None
     resourceObject=null
     tiling=NoTile
     imageSize=<32 56>
     margin=<ltrb 0 0 0 0>
     outlineSettings {
      bUseBrushTransparency=false
      roundingType=FixedRadius
      width=0
      color {colorUseRule=UseColor_Specified, specifiedColor=<rgba 0 0 0 0>}
      cornerRadii=<12 12 12 12>
     }
     tintColor {colorUseRule=UseColor_Specified, specifiedColor=<rgba 1 1 1 1>}
     uVRegion {bIsValid=false, max=<0 0>, min=<0 0>}
    }
    colorAndOpacity=<rgba 0.1274 0.1274 0.1274 0.3686>
   }
   slot {class=GridSlot, properties {columnSpan=2, layer=5}}
  }
  LoadingLeaderboardEntries {
   widgetClassPath=&CommonTextBlock
   bIsVariable=true
   parent=LeaderboradLoadingEntries
   widget=LoadingLeaderboardEntries
   properties {
    justification=Center
    text=Loading entries
    font {
     bForceMonospaced=false
     bMaterialIsStencil=false
     fontMaterial=null
     fontObject=&HeadingNow
     letterSpacing=0
     monospacedWidth=1
     size=15.75
     skewAmount=0
     typefaceFontName=74 Regular
     outlineSettings {
      bApplyOutlineToDropShadows=false
      bMiteredCorners=false
      bSeparateFillAlpha=false
      outlineMaterial=null
      outlineSize=0
      outlineColor=<rgba 0 0 0 1>
     }
    }
    shadowOffset=<0 0>
   }
   slot {class=GridSlot, properties {horizontalAlignment=HAlign_Right, layer=5, verticalAlignment=VAlign_Center, padding=<ltrb 75 0 0 0>}}
  }
  Throbber {
   widgetClassPath=&Image
   parent=LeaderboradLoadingEntries
   widget=Throbber
   properties {brush {
     bIsDynamicallyLoaded=false
     drawAs=Image
     imageType=FullColor
     mirroring=NoMirror
     resourceName=None
     resourceObject=/FortUILibrary/UIKit/Throbber/MI_UIKit_Monochrome_Throbber
     tiling=NoTile
     imageSize=<20 20>
     margin=<ltrb 0 0 0 0>
     outlineSettings {
      bUseBrushTransparency=false
      roundingType=HalfHeightRadius
      width=0
      color {colorUseRule=UseColor_Specified, specifiedColor=<rgba 0 0 0 0>}
      cornerRadii=<0 0 0 0>
     }
     tintColor {colorUseRule=UseColor_Specified, specifiedColor=<rgba 1 1 1 1>}
     uVRegion {bIsValid=false, max=<0 0>, min=<0 0>}
    }}
   slot {class=GridSlot, properties {column=1, horizontalAlignment=HAlign_Left, layer=5, verticalAlignment=VAlign_Center, padding=<ltrb 15 0 75 0>}}
  }
 ]
}
