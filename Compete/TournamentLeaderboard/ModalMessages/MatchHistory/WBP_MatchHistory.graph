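# Tournament leaderboard modal dialog showing a team's match history, with a Watch button for replaying the selected match.
# On activation it reads the selected entry and tournament style from FortCompeteUIManager, setting the title and flag-decorated player names.
# It builds Watch and Close buttons on the dialog viewmodel if none exist, hiding the first on mobile.
# Watch passes the selected session ID and teammate IDs to WatchMatchReplay; gamepad input focuses the match list.
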WidgetBlueprint /Compete/TournamentLeaderboard/ModalMessages/MatchHistory/WBP_MatchHistory
graphs [
 BP_GetDesiredFocusTarget {nodes [
   Get Desired Focus Target()
   > Return Node!(
    execute
    ReturnValue:Widget Object Reference=Get CommonListView_MatchInfoList() -> (CommonListView_MatchInfoList:Common List View Object Reference)
   )
  ]}
 CreateDialogButtons {nodes [
   CreateDialogButtons() {is=title K2Node_FunctionEntry}
   n1 = Get Buttons!(self:UI Kit Dialog Viewmodel Object Reference=Get UIKitDialogViewModel() -> (UIKitDialogViewModel:UI Kit Dialog Viewmodel Object Reference)) -> (Buttons:Array of UI Kit Dialog Button View Model Object References)
   > Branch(Condition=Is Empty!(TargetArray:Array of UI Kit Dialog Button View Model Object References=@n1) -> (ReturnValue:bool))
   > n2 = Sequence()
   > n3 = Construct Object from Class(Class=/Script/FortniteUI.UIKitDialogButtonViewModel) -> (ReturnValue:UI Kit Dialog Button View Model Object Reference)
   > n4 = Sequence!(execute) -> (then_0, then_1, then_2:Exec, then_3:Exec, then_4:Exec)
   > Set Text(Text="LOCTEXT(\"Watch\")", self:UI Kit Dialog Button View Model Object Reference=@n3)
   Set DataTable!(
    execute=@n4.then_1
    DataTable:Data Table Soft Object Reference=&InputActionDataTable
    self:UI Kit Dialog Button View Model Object Reference=@n3
   ) -> (then, Output_Get:Data Table Soft Object Reference)
   Set InputAction!(execute=@n4.then_2, InputAction:Name=Confirm, self:UI Kit Dialog Button View Model Object Reference=@n3) -> (then, Output_Get:Name=None)
   Set ActionType!(execute=@n4.then_3, ActionType:EFortDialogResult Enum=Confirmed, self:UI Kit Dialog Button View Model Object Reference=@n3) -> (then, Output_Get:EFortDialogResult Enum=Confirmed)
   Add(
    execute=@n4.then_4
    TargetArray:Array of UI Kit Dialog Button View Model Object References=@n1
    NewItem:"UI Kit Dialog Button View Model Object Reference (by ref)"=@n3
   ) {is=title K2Node_CallArrayFunction}
   n5 = Construct Object from Class(execute=@n2.then_1, Class=&UIKitDialogButtonViewModel)
   > n6 = Sequence()
   > Set Text(Text="LOCTEXT(\"Close\")", self=@n5)
   Set DataTable(execute=@n6.then_1, DataTable=&InputActionDataTable, self=@n5)
   Set InputAction(execute=@n6.then_2, InputAction=Back, self=@n5) -> (Output_Get=None)
   Set ActionType(execute=@n6.then_3, ActionType=Declined, self=@n5) -> (Output_Get=Confirmed)
   Add(execute=@n6.then_4, TargetArray=@n1, NewItem=@n5)
  ]}
 EventGraph {nodes [
   n1 = "Dialog Button Clicked Dispatcher (WBP_UIKit_Dialog_Base)"!() -> (OutputDelegate, then, ButtonIndex:int=0) {is=title K2Node_ComponentBoundEvent}
   > n2 = Switch on Int!(execute:Exec, Selection:int=@n1.ButtonIndex) -> (Default:Exec, 0:Exec, 1:Exec)
   WatchSpecificMatchReplay(execute=@n2.0)
   > n3 = Cast To FortPlayerController(Object=GetOwningPlayer())
   > "Fire Event Element Interacted (Non-Localized)"!(
    execute
    FortPC:Fort Player Controller Object Reference=@n3
    ElementName:String=WatchButton
    AdditionalInfo:Flow Analytics Additional Info Structure
   ) -> (then)
   DeactivateWidget(execute=[@n2.1, @n2.Default])
   > n4 = Cast To FortPlayerController(Object=GetOwningPlayer())
   > "Fire Event Element Interacted (Non-Localized)"(FortPC=@n4, ElementName=CloseButton)
   n5 = OnInputMethodChanged!() -> (OutputDelegate, then, bNewInputType:ECommonInputType Enum) {is=title K2Node_CustomEvent}
   > n6 = Switch on ECommonInputType!(execute, Selection:ECommonInputType Enum=@n5.bNewInputType) -> (MouseAndKeyboard:Exec, Gamepad:Exec, Touch:Exec, Count:Exec) {is=title K2Node_SwitchEnum}
   SetFocus(execute=@n6.Gamepad, self=GetDesiredFocusTarget!(self:Common Activatable Widget Object Reference) -> (ReturnValue:Widget Object Reference))
   Event PreConstruct()
   n7 = Get WBP_UIKit_Dialog_Base() -> (WBP_UIKit_Dialog_Base:WBP UIKit Dialog Internal Object Reference)
   > SetPadding!(
    execute
    self:Grid Slot Object Reference=SlotAsGridSlot(Widget=Get Slot_Content!(self:WBP UIKit Dialog Internal Object Reference=@n7) -> (Slot_Content:Named Slot Object Reference))
    InPadding_Left:float=80
    InPadding_Top:float=0
    InPadding_Right:float=80
    InPadding_Bottom:float=0
   ) -> (then)
   > Assign On Input Method Changed!(execute:Exec, self:Common Input Object Reference=Get CommonInputSubsystem(), Delegate:"Delegate (by ref)"=@n5) -> (then:Exec) {is=title K2Node_AssignDelegate}
   > SetVisibility(
    self=Get CommonText_Points() -> (CommonText_Points:Common Text Object Reference)
    InVisibility=Select(Option 0=Collapsed, Option 1=SelfHitTestInvisible, Index=GetShowMatchHistoryPoints!() -> (ReturnValue:bool))
   )
   > SetScrollbarVisibility!(execute, self:List View Base Object Reference=Get CommonListView_MatchInfoList(), InVisibility:ESlateVisibility Enum=Collapsed) -> (then)
   Event Construct()
   > OnInputMethodChanged!(execute, self, bNewInputType:ECommonInputType Enum=GetCurrentInputType(self=Get CommonInputSubsystem())) -> (then) {is=title K2Node_CallFunction}
   Event On Activated()
   > Fire Event Sub View Opened!(execute, View:Widget Object Reference=Self-Reference(), AdditionalInfo:Flow Analytics Additional Info Structure, bIsOverlay:bool=true) -> (then)
   n8 = Get FortCompeteUIManager() -> (ReturnValue:Fort Compete UIManager Object Reference) {is=title K2Node_GetSubsystem}
   > n9 = GetTournamentLeaderboardScreenSelectedEntry!(execute, self:Fort Compete UIManager Object Reference=@n8) -> (then, ReturnValue:Tournament Leaderboard Entry VM Object Reference)
   > n10 = Set LeaderboardEntryVM(LeaderboardEntryVM:Tournament Leaderboard Entry VM Object Reference=@n9) -> (Output_Get:Tournament Leaderboard Entry VM Object Reference)
   > SetPlayersInfo!(
    execute
    self
    TargetMap:Map of Strings to Tournament Player Profile VM Object References=Get PlayersProfiles!(self:Fort Tournament Base Entry VM Object Reference=@n10) -> (PlayersProfiles:Map of Strings to Tournament Player Profile VM Object References)
   ) -> (then)
   n11 = GetTournamentStyle!(self:Fort Compete UIManager Object Reference=@n8) -> (ReturnValue:Showdown Tournament Entry Structure)
   n12 = Get UIKitDialogViewModel()
   > Set DialogTitle!(
    execute
    DialogTitle:Text="To Text (String)"(InString=Break Showdown Tournament Entry!(ShowdownTournamentEntry:"Showdown Tournament Entry Structure (by ref)"=@n11) -> (
      TournamentDisplayId:String
      TitleLine1:String
      TitleLine2:String
      ScheduleInfo:String
      PosterFrontImage:String
      PosterBackImage:String
      FlavorDescription:String
      DetailsDescription:String
      ShortFormatTitle:String
      LongFormatTitle:String
      BackgroundTitle:String
      PinScoreRequirement:int
      PinEarnedText:String
      BaseColor:String
      PrimaryColor:String
      SecondaryColor:String
      HighlightColor:String
      TitleColor:String
      ShadowColor:String
      BackgroundLeftColor:String
      BackgroundRightColor:String
      BackgroundTextColor:String
      PosterFadeColor:String
      PlaylistTileImage:String
      LoadingScreenImage:String
      TournamentViewBackgroundImage:String
      TournamentLandingPageImage:String
      SquarePosterImage:String
      StyleInfoId:String
      AlertText:String
      SeriesPointLeaderboardName:String
      PlaylistDescription:String
      AlertType:EFortTournamentAlertType Enum
      RoundDisplayNames:Array of Texts
     ) {is=title K2Node_BreakStruct}.ShortFormatTitle)
    self:UI Kit Dialog Viewmodel Object Reference=@n12
   ) -> (then, Output_Get:Text)
   > CreateDialogButtons() {is=title K2Node_CallFunction}
   > SetDialogData!(execute, self:WBP UIKit Dialog Internal Object Reference=Get WBP_UIKit_Dialog_Base(), Data:UI Kit Dialog Viewmodel Object Reference=@n12) -> (then)
   > Branch(Condition=IsMobileGame())
   n13 = Get WBP_UIKit_Dialog_Base()
   > SetVisibility(
    self="Get (a copy)"(Array:Array of WBP UIKit Button CTA Base Object References=Get ButtonWidgets!(self:WBP UIKit Dialog Internal Object Reference=@n13) -> (ButtonWidgets:Array of WBP UIKit Button CTA Base Object References)) -> (Output:WBP UIKit Button CTA Base Object Reference)
    InVisibility=Collapsed
   )
  ]}
 SetPlayersInfo {nodes [
   n1 = SetPlayersInfo!() -> (then, TargetMap:Map of Strings to Tournament Player Profile VM Object References) {is=title K2Node_FunctionEntry}
   > n2 = Values!(execute, TargetMap:Map of Strings to Tournament Player Profile VM Object References=@n1) -> (then, Values:Array of Tournament Player Profile VM Object References)
   > n3 = For Each Loop(Array:Array of Tournament Player Profile VM Object References=@n2) -> (Array Element:Tournament Player Profile VM Object Reference)
   n4 = Get DisplayName!(self:Tournament Player Profile VM Object Reference=@n3) -> (DisplayName:String)
   n5 = Get FlagName!(self:Tournament Player Profile VM Object Reference=@n3) -> (FlagName:String)
   n6 = "To String (Text)"(InText=Select(
     Option 0:"Text (by ref)"=Format Text!(Format="LOCTEXT(\"<img id=\\\"{Country}\\\" size=\\\"18,18\\\" height=\\\"25\\\"/> {Nickname}\")", Country:String=@n5, Nickname:String=@n4) -> (Result)
     Option 1:"Text (by ref)"=To Text!(Value:String=@n4) -> (ReturnValue:Text) {is=title K2Node_GenericToText}
     Index="Equal Exactly (String)"(A=@n5, B=fortnite)
    ) -> (ReturnValue:"Text (by ref)"))
   n7 = Get PlayersNicknames() -> (PlayersNicknames:String)
   n8 = "Equal (Integer)"(A=@n3.Array Index, B=Last Index!(TargetArray:Array of Tournament Player Profile VM Object References=@n2) -> (ReturnValue=0))
   > Set PlayersNicknames(PlayersNicknames:String=Append(A=Append(A=@n7, B=@n6), B=Select(Option 0:String=" && ", Option 1:String, Index=@n8) -> (ReturnValue:String))) -> (Output_Get:String)
   SetText(
    execute=@n3.Completed
    self=Get CommonRichTextBlock_PlayersInfo() -> (CommonRichTextBlock_PlayersInfo:Common Rich Text Block Object Reference)
    InText="To Text (String)"(InString=Get PlayersNicknames())
   )
  ]}
 WatchSpecificMatchReplay {nodes [
   WatchSpecificMatchReplay() {is=title K2Node_FunctionEntry}
   > n1 = Cast To FortTournamentLeaderboardSessionTeamEntryVM!(
    execute
    Object=Get Selected Item!(self:List View Object Reference=Get CommonListView_MatchInfoList()) -> (ReturnValue:Object Reference) {is=title K2Node_CallFunction}
   ) -> (then, CastFailed, AsTournament Leaderboard Session Team Entry VM:Tournament Leaderboard Session Team Entry VM Object Reference) {is=title K2Node_DynamicCast}
   > WatchMatchReplay!(
    execute
    self:Fort Compete UIManager Object Reference=Get FortCompeteUIManager()
    SessionID:String=Get SessionId!(self:Tournament Leaderboard Session Team Entry VM Object Reference=@n1) -> (SessionId:String)
    TeamAccountIds:Array of Strings=Get TeammatesAccountIds!(self:Fort Tournament Base Entry VM Object Reference=Get LeaderboardEntryVM() -> (LeaderboardEntryVM:Tournament Leaderboard Entry VM Object Reference)) -> (TeammatesAccountIds:Array of Strings)
   ) -> (then)
  ]}
]
variables []
widgets {
 info {
  inheritedWidgetCount=0
  namedSlotCount=1
  parentClass=&CommonActivatableWidget
  rootWidgetClass=/FortUILibrary/Systems/Dialog/WBP_UIKit_Dialog_Internal.WBP_UIKit_Dialog_Internal_C
  widgetCount=13
 }
 widgets [
  WBP_UIKit_Dialog_Base {
   widgetClassPath=/FortUILibrary/Systems/Dialog/WBP_UIKit_Dialog_Internal
   bIsVariable=true
   widget=WBP_UIKit_Dialog_Base
   properties {
    useText=None
    buttons {
     each Button Horizontal Fill=true
     each Button Min Width=0
     each Button Size=Medium
     focuse Button=0
     space Between Buttons=24
     cTA Buttons Types [Regular,Quiet]
     group Placement {
      horizontal Alignment=HAlign_Fill
      vertical Alignment=VAlign_Bottom
      grid {column=1, column Span=0, layer=0, row=2, row Span=0, nudge=<0 0>}
      padding=<ltrb 80 24 80 48>
     }
    }
    dialogSIze {
     height Override {enable=false, value=886}
     max Aspect Ratio {enable=false, value=0}
     max Desired Height {enable=true, value=886}
     max Desired Width {enable=false, value=0}
     min Aspect Ratio {enable=false, value=0}
     min Desired Height {enable=false, value=0}
     min Desired Width {enable=false, value=0}
     width Override {enable=true, value=950}
    }
    imageProperties {
     material=/FortUILibrary/Systems/Dialog/MaterialInstances/MI_UI_DialogImage
     image Size=<0 0>
     placement {
      horizontal Alignment=HAlign_Fill
      vertical Alignment=VAlign_Fill
      grid {column=1, column Span=0, layer=20, row=0, row Span=1, nudge=<0 0>}
      padding=<ltrb 0 0 0 0>
     }
    }
    titleProperties {
     auto Wrap Text=true
     justification=Left
     line Height Percentage=1
     mobile Font Size Multiplier=1.3
     transform Policy=ToUpper
     wrap Text At=0
     color {colorUseRule=UseColor_Specified, specifiedColor=<rgba 1 1 1 1>}
     font {
      bForceMonospaced=false
      bMaterialIsStencil=false
      fontMaterial=null
      fontObject=&HeadingNow
      letterSpacing=50
      monospacedWidth=1
      size=24
      skewAmount=0
      typefaceFontName=86 Bold Italic
      outlineSettings {
       bApplyOutlineToDropShadows=false
       bMiteredCorners=false
       bSeparateFillAlpha=false
       outlineMaterial=null
       outlineSize=0
       outlineColor=<rgba 0 0 0 1>
      }
     }
     placement {
      horizontal Alignment=HAlign_Fill
      vertical Alignment=VAlign_Top
      grid {column=1, column Span=0, layer=0, row=0, row Span=0, nudge=<0 0>}
      padding=<ltrb 80 50 80 18>
     }
    }
   }
  }
  Root {
   widgetClassPath=&GridPanel
   namedSlotHost=WBP_UIKit_Dialog_Base
   widget=Root
   properties {flowDirectionPreference=Culture, columnFill [1], rowFill [0,0,1,0]}
  }
  CommonRichTextBlock_PlayersInfo {
   widgetClassPath=&CommonRichTextBlock
   bIsVariable=true
   parent=Root
   widget=CommonRichTextBlock_PlayersInfo
   properties {
    autoWrapText=true
    bOverrideDefaultStyle=true
    clipping=ClipToBounds
    mobileTextBlockScale=1.14
    text="WWWWWWWWWWWWWWWW && WWWWWWWWWWWWWWWW && WWWWWWWWWWWWWWWW && WWWWWWWWWWWWWWWW "
    decoratorClasses [/Compete/DataTables/BP_FlagsDecorator.BP_FlagsDecorator_C]
    defaultTextStyleOverride {
     overflowPolicy=Clip
     transformPolicy=None
     colorAndOpacity {colorUseRule=UseColor_Specified, specifiedColor=<rgba 1 1 1 1>}
     font {
      bForceMonospaced=false
      bMaterialIsStencil=false
      fontMaterial=null
      fontObject=&HeadingNow
      letterSpacing=0
      monospacedWidth=1
      size=15.75
      skewAmount=0
      typefaceFontName=74 Regular
      outlineSettings {
       bApplyOutlineToDropShadows=false
       bMiteredCorners=false
       bSeparateFillAlpha=false
       outlineMaterial=null
       outlineSize=0
       outlineColor=<rgba 0 0 0 1>
      }
     }
     highlightColor {colorUseRule=UseColor_Specified, specifiedColor=<rgba 0 0 0 1>}
     highlightShape {
      bIsDynamicallyLoaded=false
      drawAs=Image
      imageType=NoImage
      mirroring=NoMirror
      resourceName=None
      resourceObject=null
      tiling=NoTile
      imageSize=<32 32>
      margin=<ltrb 0 0 0 0>
      outlineSettings {
       bUseBrushTransparency=false
       roundingType=HalfHeightRadius
       width=0
       color {colorUseRule=UseColor_Specified, specifiedColor=<rgba 0 0 0 0>}
       cornerRadii=<0 0 0 0>
      }
      tintColor {colorUseRule=UseColor_Specified, specifiedColor=<rgba 1 1 1 1>}
      uVRegion {bIsValid=false, max=<0 0>, min=<0 0>}
     }
     selectedBackgroundColor {colorUseRule=UseColor_Foreground, specifiedColor=<rgba 1 0 1 1>}
     shadowColorAndOpacity=<rgba 0 0 0 1>
     shadowOffset=<0 0>
     strikeBrush {
      bIsDynamicallyLoaded=false
      drawAs=Image
      imageType=NoImage
      mirroring=NoMirror
      resourceName=None
      resourceObject=null
      tiling=NoTile
      imageSize=<32 32>
      margin=<ltrb 0 0 0 0>
      outlineSettings {
       bUseBrushTransparency=false
       roundingType=HalfHeightRadius
       width=0
       color {colorUseRule=UseColor_Specified, specifiedColor=<rgba 0 0 0 0>}
       cornerRadii=<0 0 0 0>
      }
      tintColor {colorUseRule=UseColor_Specified, specifiedColor=<rgba 1 1 1 1>}
      uVRegion {bIsValid=false, max=<0 0>, min=<0 0>}
     }
     underlineBrush {
      bIsDynamicallyLoaded=false
      drawAs=Image
      imageType=NoImage
      mirroring=NoMirror
      resourceName=None
      resourceObject=null
      tiling=NoTile
      imageSize=<32 32>
      margin=<ltrb 0 0 0 0>
      outlineSettings {
       bUseBrushTransparency=false
       roundingType=HalfHeightRadius
       width=0
       color {colorUseRule=UseColor_Specified, specifiedColor=<rgba 0 0 0 0>}
       cornerRadii=<0 0 0 0>
      }
      tintColor {colorUseRule=UseColor_Specified, specifiedColor=<rgba 1 1 1 1>}
      uVRegion {bIsValid=false, max=<0 0>, min=<0 0>}
     }
    }
   }
   slot {class=GridSlot, properties {verticalAlignment=VAlign_Center}}
  }
  CommonText_MatchHistory {
   widgetClassPath=&CommonTextBlock
   parent=Root
   widget=CommonText_MatchHistory
   properties {
    autoWrapText=true
    mobileFontSizeMultiplier=1.14
    scrollStyle=&TextScrollStyle-Base_C
    text=MATCH HISTORY
    colorAndOpacity {colorUseRule=UseColor_Specified, specifiedColor=<rgba 0.5776 0.5776 0.5776 1>}
    font {
     bForceMonospaced=false
     bMaterialIsStencil=false
     fontMaterial=null
     fontObject=&HeadingNow
     letterSpacing=0
     monospacedWidth=1
     size=15.75
     skewAmount=0
     typefaceFontName=75 Medium
     outlineSettings {
      bApplyOutlineToDropShadows=false
      bMiteredCorners=false
      bSeparateFillAlpha=false
      outlineMaterial=null
      outlineSize=0
      outlineColor=<rgba 0 0 0 1>
     }
    }
    shadowOffset=<0 0>
   }
   slot {class=GridSlot, properties {row=1, verticalAlignment=VAlign_Top, padding=<ltrb 0 16 0 0>}}
  }
  GridPanel_MatchHistory {
   widgetClassPath=&GridPanel
   parent=Root
   widget=GridPanel_MatchHistory
   properties {columnFill [1], rowFill [0,1]}
   slot {class=GridSlot, properties {columnSpan=1, row=2, rowSpan=2, padding=<ltrb 0 17 0 21>}}
  }
  SizeBox {
   widgetClassPath=&SizeBox
   parent=GridPanel_MatchHistory
   widget=SizeBox
   properties {bOverride_WidthOverride=true, widthOverride=760}
   slot {class=GridSlot, properties {horizontalAlignment=HAlign_Left, padding=<ltrb 15 0 0 0>}}
  }
  Icons {widgetClassPath=/Script/UMG.UniformGridPanel, parent=SizeBox, widget=Icons, slot {class=SizeBoxSlot}}
  Image_Wins {
   widgetClassPath=&Image
   parent=Icons
   widget=Image_Wins
   properties {
    brush {
     bIsDynamicallyLoaded=false
     drawAs=Image
     imageType=FullColor
     mirroring=NoMirror
     resourceName=None
     resourceObject=/FortUILibrary/Textures/MaterialInstances/MI_UIKit_Icon_Compete
     tiling=NoTile
     imageSize=<32 32>
     margin=<ltrb 0 0 0 0>
     outlineSettings {
      bUseBrushTransparency=false
      roundingType=HalfHeightRadius
      width=0
      color {colorUseRule=UseColor_Specified, specifiedColor=<rgba 0 0 0 0>}
      cornerRadii=<0 0 0 0>
     }
     tintColor {colorUseRule=UseColor_Specified, specifiedColor=<rgba 1 1 1 1>}
     uVRegion {bIsValid=false, max=<0 0>, min=<0 0>}
    }
    colorAndOpacity=<rgba 0.5776 0.5776 0.5776 1>
   }
   slot {class=UniformGridSlot, properties {column=1, horizontalAlignment=HAlign_Center}}
  }
  Image_Elimination {
   widgetClassPath=&Image
   parent=Icons
   widget=Image_Elimination
   properties {
    brush {
     bIsDynamicallyLoaded=false
     drawAs=Image
     imageType=FullColor
     mirroring=NoMirror
     resourceName=None
     resourceObject=/FortUILibrary/Textures/MaterialInstances/MI_UIKit_Icon_Reticle
     tiling=NoTile
     imageSize=<32 32>
     margin=<ltrb 0 0 0 0>
     outlineSettings {
      bUseBrushTransparency=false
      roundingType=HalfHeightRadius
      width=0
      color {colorUseRule=UseColor_Specified, specifiedColor=<rgba 0 0 0 0>}
      cornerRadii=<0 0 0 0>
     }
     tintColor {colorUseRule=UseColor_Specified, specifiedColor=<rgba 1 1 1 1>}
     uVRegion {bIsValid=false, max=<0 0>, min=<0 0>}
    }
    colorAndOpacity=<rgba 0.5776 0.5776 0.5776 1>
   }
   slot {class=UniformGridSlot, properties {column=2, horizontalAlignment=HAlign_Center}}
  }
  CommonText_Points {
   widgetClassPath=&CommonTextBlock
   bIsVariable=true
   parent=Icons
   widget=CommonText_Points
   properties {
    text=Pts
    colorAndOpacity {colorUseRule=UseColor_Specified, specifiedColor=<rgba 0.5776 0.5776 0.5776 1>}
    font {
     bForceMonospaced=false
     bMaterialIsStencil=false
     fontMaterial=null
     fontObject=&HeadingNow
     letterSpacing=0
     monospacedWidth=1
     size=18
     skewAmount=0
     typefaceFontName=75 Medium
     outlineSettings {
      bApplyOutlineToDropShadows=false
      bMiteredCorners=false
      bSeparateFillAlpha=false
      outlineMaterial=null
      outlineSize=0
      outlineColor=<rgba 0 0 0 1>
     }
    }
    shadowOffset=<0 0>
   }
   slot {class=UniformGridSlot, properties {column=3, horizontalAlignment=HAlign_Center}}
  }
  ListWrapper {
   widgetClassPath=&GridPanel
   parent=GridPanel_MatchHistory
   widget=ListWrapper
   properties {clipping=ClipToBounds, columnFill [1], rowFill [1]}
   slot {class=GridSlot, properties {columnSpan=1, row=1, rowSpan=4}}
  }
  CommonListView_MatchInfoList {
   widgetClassPath=&CommonListView
   bIsVariable=true
   parent=ListWrapper
   widget=CommonListView_MatchInfoList
   properties {
    bReturnFocusToSelection=true
    clipping=Inherit
    entryWidgetClass=&/WBP_MatchHistoryListButton.WBP_MatchHistoryListButton_C
    navigation=&self:WidgetTree.CommonListView_MatchInfoList.WidgetNavigation_0
    numDesignerPreviewEntries=15
    scrollingAnimationInterpolationSpeed=10
    verticalEntrySpacing=7
    visibility=SelfHitTestInvisible
    scrollBarPadding=<ltrb 15 0 0 0>
    scrollBarStyle {
     thickness=12
     draggedThumbImage {
      bIsDynamicallyLoaded=false
      drawAs=RoundedBox
      imageType=FullColor
      mirroring=NoMirror
      resourceName=None
      resourceObject=null
      tiling=NoTile
      imageSize=<12 12>
      margin=<ltrb 0.5 0.5 0.5 0.5>
      outlineSettings {
       bUseBrushTransparency=false
       roundingType=FixedRadius
       width=2
       color {colorUseRule=UseColor_Specified, specifiedColor=<rgba 0 0 0 0>}
       cornerRadii=<6 6 6 6>
      }
      tintColor {colorUseRule=UseColor_Specified, specifiedColor=<rgba 1 1 1 1>}
      uVRegion {bIsValid=false, max=<0 0>, min=<0 0>}
     }
     horizontalBackgroundImage {
      bIsDynamicallyLoaded=false
      drawAs=NoDrawType
      imageType=NoImage
      mirroring=NoMirror
      resourceName=None
      resourceObject=null
      tiling=NoTile
      imageSize=<0 0>
      margin=<ltrb 0 0 0 0>
      outlineSettings {
       bUseBrushTransparency=false
       roundingType=HalfHeightRadius
       width=0
       color {colorUseRule=UseColor_Specified, specifiedColor=<rgba 0 0 0 0>}
       cornerRadii=<0 0 0 0>
      }
      tintColor {colorUseRule=UseColor_Specified, specifiedColor=<rgba 1 1 1 1>}
      uVRegion {bIsValid=false, max=<0 0>, min=<0 0>}
     }
     horizontalBottomSlotImage {
      bIsDynamicallyLoaded=false
      drawAs=NoDrawType
      imageType=FullColor
      mirroring=NoMirror
      resourceName=../../../Engine/Content/Slate/Common/Scrollbar_Background_Horizontal.png
      resourceObject=null
      tiling=NoTile
      imageSize=<8 8>
      margin=<ltrb 0 0 0 0>
      outlineSettings {
       bUseBrushTransparency=false
       roundingType=HalfHeightRadius
       width=0
       color {colorUseRule=UseColor_Specified, specifiedColor=<rgba 0 0 0 0>}
       cornerRadii=<0 0 0 0>
      }
      tintColor {colorUseRule=UseColor_Specified, specifiedColor=<rgba 1 1 1 1>}
      uVRegion {bIsValid=false, max=<0 0>, min=<0 0>}
     }
     horizontalTopSlotImage {
      bIsDynamicallyLoaded=false
      drawAs=NoDrawType
      imageType=FullColor
      mirroring=NoMirror
      resourceName=../../../Engine/Content/Slate/Common/Scrollbar_Background_Horizontal.png
      resourceObject=null
      tiling=NoTile
      imageSize=<8 8>
      margin=<ltrb 0 0 0 0>
      outlineSettings {
       bUseBrushTransparency=false
       roundingType=HalfHeightRadius
       width=0
       color {colorUseRule=UseColor_Specified, specifiedColor=<rgba 0 0 0 0>}
       cornerRadii=<0 0 0 0>
      }
      tintColor {colorUseRule=UseColor_Specified, specifiedColor=<rgba 1 1 1 1>}
      uVRegion {bIsValid=false, max=<0 0>, min=<0 0>}
     }
     hoveredThumbImage {
      bIsDynamicallyLoaded=false
      drawAs=RoundedBox
      imageType=FullColor
      mirroring=NoMirror
      resourceName=None
      resourceObject=null
      tiling=NoTile
      imageSize=<12 12>
      margin=<ltrb 0.5 0.5 0.5 0.5>
      outlineSettings {
       bUseBrushTransparency=false
       roundingType=FixedRadius
       width=2
       color {colorUseRule=UseColor_Specified, specifiedColor=<rgba 0 0 0 0>}
       cornerRadii=<6 6 6 6>
      }
      tintColor {colorUseRule=UseColor_Specified, specifiedColor=<rgba 1 1 1 1>}
      uVRegion {bIsValid=false, max=<0 0>, min=<0 0>}
     }
     normalThumbImage {
      bIsDynamicallyLoaded=false
      drawAs=RoundedBox
      imageType=FullColor
      mirroring=NoMirror
      resourceName=None
      resourceObject=null
      tiling=NoTile
      imageSize=<12 12>
      margin=<ltrb 0.5 0.5 0.5 0.5>
      outlineSettings {
       bUseBrushTransparency=false
       roundingType=FixedRadius
       width=2
       color {colorUseRule=UseColor_Specified, specifiedColor=<rgba 0 0 0 0>}
       cornerRadii=<6 6 6 6>
      }
      tintColor {colorUseRule=UseColor_Specified, specifiedColor=<rgba 1 1 1 0.75>}
      uVRegion {bIsValid=false, max=<0 0>, min=<0 0>}
     }
     verticalBackgroundImage {
      bIsDynamicallyLoaded=false
      drawAs=Box
      imageType=FullColor
      mirroring=NoMirror
      resourceName=None
      resourceObject=/FortUILibrary/UIKit/ScrollBar/MaterialInstances/MI_UIKit_ScrollbarBackground
      tiling=NoTile
      imageSize=<12 12>
      margin=<ltrb 0.5 0.5 0.5 0.5>
      outlineSettings {
       bUseBrushTransparency=false
       roundingType=HalfHeightRadius
       width=0
       color {colorUseRule=UseColor_Specified, specifiedColor=<rgba 0 0 0 0>}
       cornerRadii=<0 0 0 0>
      }
      tintColor {colorUseRule=UseColor_Specified, specifiedColor=<rgba 1 1 1 0.2>}
      uVRegion {bIsValid=false, max=<0 0>, min=<0 0>}
     }
     verticalBottomSlotImage {
      bIsDynamicallyLoaded=false
      drawAs=NoDrawType
      imageType=FullColor
      mirroring=NoMirror
      resourceName=../../../Engine/Content/Slate/Common/Scrollbar_Background_Vertical.png
      resourceObject=null
      tiling=NoTile
      imageSize=<8 8>
      margin=<ltrb 0 0 0 0>
      outlineSettings {
       bUseBrushTransparency=false
       roundingType=HalfHeightRadius
       width=0
       color {colorUseRule=UseColor_Specified, specifiedColor=<rgba 0 0 0 0>}
       cornerRadii=<0 0 0 0>
      }
      tintColor {colorUseRule=UseColor_Specified, specifiedColor=<rgba 1 1 1 1>}
      uVRegion {bIsValid=false, max=<0 0>, min=<0 0>}
     }
     verticalTopSlotImage {
      bIsDynamicallyLoaded=false
      drawAs=NoDrawType
      imageType=FullColor
      mirroring=NoMirror
      resourceName=../../../Engine/Content/Slate/Common/Scrollbar_Background_Vertical.png
      resourceObject=null
      tiling=NoTile
      imageSize=<8 8>
      margin=<ltrb 0 0 0 0>
      outlineSettings {
       bUseBrushTransparency=false
       roundingType=HalfHeightRadius
       width=0
       color {colorUseRule=UseColor_Specified, specifiedColor=<rgba 0 0 0 0>}
       cornerRadii=<0 0 0 0>
      }
      tintColor {colorUseRule=UseColor_Specified, specifiedColor=<rgba 1 1 1 1>}
      uVRegion {bIsValid=false, max=<0 0>, min=<0 0>}
     }
    }
   }
   slot {class=GridSlot, properties {padding=<ltrb 15 15 0 15>}}
  }
  LoadingEntriesMessage {
   widgetClassPath=/Compete/TournamentLeaderboard/WBP_LoadingEntriesMessage
   bIsVariable=true
   parent=Root
   widget=LoadingEntriesMessage
   properties {loadingMessageText=Loading match history entries, visibility=Collapsed}
   slot {class=GridSlot, properties {horizontalAlignment=HAlign_Center, layer=10, row=3, verticalAlignment=VAlign_Bottom, padding=<ltrb 0 0 0 40>}}
  }
 ]
}
