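# Tournament leaderboard modal that lists player profiles and opens the selected player's career stats screen.
# SetPlayerProfiles wraps each tournament profile in a VM_CompetePlayerProfile to fill the list hosted in a UIKit dialog slot.
# The view button shows ranked br-combined stats for the latest season track and closes the modal; the other button just closes, both firing analytics.
# PreConstruct supplies dialog data from a Debug-named viewmodel, and gamepad input focuses the list.
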WidgetBlueprint /Compete/TournamentLeaderboard/ModalMessages/Profiles/WBP_ProfilesModal
graphs [
 BP_GetDesiredFocusTarget {nodes [
   Get Desired Focus Target()
   > Return Node!(execute, ReturnValue:Widget Object Reference=Get ProfilesListView() -> (ProfilesListView:Common List View Object Reference))
  ]}
 EventGraph {nodes [
   n1 = "Dialog Button Clicked Dispatcher (WBP_UIKit_Dialog_Base)"!() -> (OutputDelegate, then, ButtonIndex:int=0) {is=title K2Node_ComponentBoundEvent}
   > n2 = Switch on Int!(execute:Exec, Selection:int=@n1.ButtonIndex) -> (Default:Exec, 0:Exec, 1:Exec)
   n3 = Cast To VM_CompetePlayerProfile!(
    execute=@n2.0
    Object=Get Selected Item!(self:List View Object Reference=Get ProfilesListView()) -> (ReturnValue:Object Reference) {is=title K2Node_CallFunction}
   ) -> (then, CastFailed, AsVM Compete Player Profile:VM Compete Player Profile Object Reference) {is=title K2Node_DynamicCast}
   > ShowCareerStatsScreen!(
    execute
    self:Fort Compete UIManager Object Reference=Get FortCompeteUIManager() -> (ReturnValue:Fort Compete UIManager Object Reference) {is=title K2Node_GetSubsystem}
    AccountId:String=Get AccountId!(self:Tournament Player Profile VM Object Reference=Get VMProfile!(self:VM Compete Player Profile Object Reference=@n3) -> (VMProfile:Tournament Player Profile VM Object Reference)) -> (AccountId:String)
    GameMode:String=br-combined
    bIsRanked:bool=true
    InSeasonTrack:String=GetLatestSeasonTrackGUID!(
     self:Fort Stats Page Manager Object Reference=Get FortStatsPageManager() -> (ReturnValue:Fort Stats Page Manager Object Reference) {is=title K2Node_GetSubsystem}
     GameMode:"Name (by ref)"=MakeLiteralName(Value=br-combined)
    ) -> (ReturnValue:String)
    InLimitedGameModeIndex:int=-1
   ) -> (then)
   > DeactivateWidget()
   > n4 = Cast To FortPlayerController(Object=GetOwningPlayer())
   > "Fire Event Element Interacted (Non-Localized)"!(
    execute
    FortPC:Fort Player Controller Object Reference=@n4
    ElementName:String=ViewButton
    AdditionalInfo:Flow Analytics Additional Info Structure
   ) -> (then)
   DeactivateWidget(execute=@n2.1)
   > n5 = Cast To FortPlayerController(Object=GetOwningPlayer())
   > "Fire Event Element Interacted (Non-Localized)"(FortPC=@n5, ElementName=CloseButton)
   n6 = OnInputMethodChanged!() -> (OutputDelegate, then, bNewInputType:ECommonInputType Enum) {is=title K2Node_CustomEvent}
   > n7 = Switch on ECommonInputType!(execute, Selection:ECommonInputType Enum=@n6.bNewInputType) -> (MouseAndKeyboard:Exec, Gamepad:Exec, Touch:Exec, Count:Exec) {is=title K2Node_SwitchEnum}
   SetFocus(execute=@n7.Gamepad, self=Get ProfilesListView())
   Event PreConstruct()
   n8 = Get WBP_UIKit_Dialog_Base() -> (WBP_UIKit_Dialog_Base:WBP UIKit Dialog Internal Object Reference)
   > SetPadding!(
    execute
    self:Grid Slot Object Reference=SlotAsGridSlot(Widget=Get Slot_Content!(self:WBP UIKit Dialog Internal Object Reference=@n8) -> (Slot_Content:Named Slot Object Reference))
    InPadding_Left:float=80
    InPadding_Top:float=15
    InPadding_Right:float=80
    InPadding_Bottom:float=15
   ) -> (then)
   > n9 = Construct Object from Class(Class=&/MVVM_Dialog_ProfilesModal_Debug.MVVM_Dialog_ProfilesModal_Debug_C) -> (ReturnValue:MVVM Dialog Profiles Modal Debug Object Reference)
   > SetDialogData!(execute, self:WBP UIKit Dialog Internal Object Reference=Get WBP_UIKit_Dialog_Base(), Data:UI Kit Dialog Viewmodel Object Reference=@n9) -> (then)
   > Assign On Input Method Changed!(execute:Exec, self:Common Input Object Reference=Get CommonInputSubsystem(), Delegate:"Delegate (by ref)"=@n6) -> (then:Exec) {is=title K2Node_AssignDelegate}
   Event Construct()
   > OnInputMethodChanged!(execute, self, bNewInputType:ECommonInputType Enum=GetCurrentInputType(self=Get CommonInputSubsystem())) -> (then) {is=title K2Node_CallFunction}
   > Fire Event Sub View Opened!(execute, View:Widget Object Reference=Self-Reference(), AdditionalInfo:Flow Analytics Additional Info Structure, bIsOverlay:bool=true) -> (then)
  ]}
 SetPlayerProfiles {nodes [
   n1 = SetPlayerProfiles!() -> (then, Profiles:Map of Strings to Tournament Player Profile VM Object References) {is=title K2Node_FunctionEntry}
   > Clear(TargetArray:Array of VM Compete Player Profile Object References=Get PlayersProfilesVM() -> (PlayersProfilesVM:Array of VM Compete Player Profile Object References))
   > n2 = Values!(execute, TargetMap:Map of Strings to Tournament Player Profile VM Object References=@n1) -> (then, Values:Array of Tournament Player Profile VM Object References)
   > n3 = For Each Loop(Array:Array of Tournament Player Profile VM Object References=@n2) -> (Array Element:Tournament Player Profile VM Object Reference)
   Set List Items!(
    execute=@n3.Completed
    self:List View Object Reference=Get ProfilesListView()
    InListItems:Array of Object References=Get PlayersProfilesVM()
   ) -> (then) {is=title K2Node_CallFunction}
   n4 = Construct Object from Class!(
    execute=@n3.LoopBody
    Class=&/VM_CompetePlayerProfile.VM_CompetePlayerProfile_C
    self=GetOwningPlayer()
    VMProfile:Tournament Player Profile VM Object Reference=@n3
    Index:int=@n3.Array Index
   ) -> (then, ReturnValue:VM Compete Player Profile Object Reference)
   > Add(
    TargetArray:Array of VM Compete Player Profile Object References=Get PlayersProfilesVM()
    NewItem:"VM Compete Player Profile Object Reference (by ref)"=@n4
   ) {is=title K2Node_CallArrayFunction}
  ]}
]
variables []
widgets {
 info {
  inheritedWidgetCount=0
  namedSlotCount=1
  parentClass=&CommonActivatableWidget
  rootWidgetClass=/FortUILibrary/Systems/Dialog/WBP_UIKit_Dialog_Internal.WBP_UIKit_Dialog_Internal_C
  widgetCount=2
 }
 widgets [
  WBP_UIKit_Dialog_Base {
   widgetClassPath=/FortUILibrary/Systems/Dialog/WBP_UIKit_Dialog_Internal
   bIsVariable=true
   widget=WBP_UIKit_Dialog_Base
   properties {
    useText=None
    buttons {
     each Button Horizontal Fill=true
     each Button Min Width=0
     each Button Size=Medium
     focuse Button=0
     space Between Buttons=24
     cTA Buttons Types [Regular,Quiet]
     group Placement {
      horizontal Alignment=HAlign_Fill
      vertical Alignment=VAlign_Bottom
      grid {column=1, column Span=0, layer=0, row=2, row Span=0, nudge=<0 0>}
      padding=<ltrb 80 24 80 48>
     }
    }
    dialogSIze {
     height Override {enable=false, value=592}
     max Aspect Ratio {enable=false, value=0}
     max Desired Height {enable=true, value=886}
     max Desired Width {enable=false, value=0}
     min Aspect Ratio {enable=false, value=0}
     min Desired Height {enable=false, value=0}
     min Desired Width {enable=false, value=0}
     width Override {enable=true, value=595}
    }
    imageProperties {
     material=/FortUILibrary/Systems/Dialog/MaterialInstances/MI_UI_DialogImage
     image Size=<0 0>
     placement {
      horizontal Alignment=HAlign_Fill
      vertical Alignment=VAlign_Fill
      grid {column=1, column Span=0, layer=20, row=0, row Span=1, nudge=<0 0>}
      padding=<ltrb 0 0 0 0>
     }
    }
    titleProperties {
     auto Wrap Text=true
     justification=Left
     line Height Percentage=1
     mobile Font Size Multiplier=1.3
     transform Policy=ToUpper
     wrap Text At=0
     color {colorUseRule=UseColor_Specified, specifiedColor=<rgba 1 1 1 1>}
     font {
      bForceMonospaced=false
      bMaterialIsStencil=false
      fontMaterial=null
      fontObject=&HeadingNow
      letterSpacing=50
      monospacedWidth=1
      size=24
      skewAmount=0
      typefaceFontName=86 Bold Italic
      outlineSettings {
       bApplyOutlineToDropShadows=false
       bMiteredCorners=false
       bSeparateFillAlpha=false
       outlineMaterial=null
       outlineSize=0
       outlineColor=<rgba 0 0 0 1>
      }
     }
     placement {
      horizontal Alignment=HAlign_Fill
      vertical Alignment=VAlign_Top
      grid {column=1, column Span=0, layer=0, row=0, row Span=0, nudge=<0 0>}
      padding=<ltrb 80 50 80 18>
     }
    }
   }
  }
  ProfilesListView {
   widgetClassPath=&CommonListView
   bIsVariable=true
   namedSlotHost=WBP_UIKit_Dialog_Base
   widget=ProfilesListView
   properties {
    allowOverscroll=false
    bEnableRightClickScrolling=false
    bReturnFocusToSelection=true
    clipping=Inherit
    entryWidgetClass=&/WBP_ProfilesListButton.WBP_ProfilesListButton_C
    navigation=&self:WidgetTree.ProfilesListView.WidgetNavigation_0
    numDesignerPreviewEntries=4
    scrollingAnimationInterpolationSpeed=10
    verticalEntrySpacing=7
    visibility=SelfHitTestInvisible
    scrollBarPadding=<ltrb 15 0 0 0>
    scrollBarStyle {
     thickness=12
     draggedThumbImage {
      bIsDynamicallyLoaded=false
      drawAs=RoundedBox
      imageType=FullColor
      mirroring=NoMirror
      resourceName=None
      resourceObject=null
      tiling=NoTile
      imageSize=<12 12>
      margin=<ltrb 0.5 0.5 0.5 0.5>
      outlineSettings {
       bUseBrushTransparency=false
       roundingType=FixedRadius
       width=2
       color {colorUseRule=UseColor_Specified, specifiedColor=<rgba 0 0 0 0>}
       cornerRadii=<6 6 6 6>
      }
      tintColor {colorUseRule=UseColor_Specified, specifiedColor=<rgba 1 1 1 1>}
      uVRegion {bIsValid=false, max=<0 0>, min=<0 0>}
     }
     horizontalBackgroundImage {
      bIsDynamicallyLoaded=false
      drawAs=NoDrawType
      imageType=NoImage
      mirroring=NoMirror
      resourceName=None
      resourceObject=null
      tiling=NoTile
      imageSize=<0 0>
      margin=<ltrb 0 0 0 0>
      outlineSettings {
       bUseBrushTransparency=false
       roundingType=HalfHeightRadius
       width=0
       color {colorUseRule=UseColor_Specified, specifiedColor=<rgba 0 0 0 0>}
       cornerRadii=<0 0 0 0>
      }
      tintColor {colorUseRule=UseColor_Specified, specifiedColor=<rgba 1 1 1 1>}
      uVRegion {bIsValid=false, max=<0 0>, min=<0 0>}
     }
     horizontalBottomSlotImage {
      bIsDynamicallyLoaded=false
      drawAs=NoDrawType
      imageType=FullColor
      mirroring=NoMirror
      resourceName=../../../Engine/Content/Slate/Common/Scrollbar_Background_Horizontal.png
      resourceObject=null
      tiling=NoTile
      imageSize=<8 8>
      margin=<ltrb 0 0 0 0>
      outlineSettings {
       bUseBrushTransparency=false
       roundingType=HalfHeightRadius
       width=0
       color {colorUseRule=UseColor_Specified, specifiedColor=<rgba 0 0 0 0>}
       cornerRadii=<0 0 0 0>
      }
      tintColor {colorUseRule=UseColor_Specified, specifiedColor=<rgba 1 1 1 1>}
      uVRegion {bIsValid=false, max=<0 0>, min=<0 0>}
     }
     horizontalTopSlotImage {
      bIsDynamicallyLoaded=false
      drawAs=NoDrawType
      imageType=FullColor
      mirroring=NoMirror
      resourceName=../../../Engine/Content/Slate/Common/Scrollbar_Background_Horizontal.png
      resourceObject=null
      tiling=NoTile
      imageSize=<8 8>
      margin=<ltrb 0 0 0 0>
      outlineSettings {
       bUseBrushTransparency=false
       roundingType=HalfHeightRadius
       width=0
       color {colorUseRule=UseColor_Specified, specifiedColor=<rgba 0 0 0 0>}
       cornerRadii=<0 0 0 0>
      }
      tintColor {colorUseRule=UseColor_Specified, specifiedColor=<rgba 1 1 1 1>}
      uVRegion {bIsValid=false, max=<0 0>, min=<0 0>}
     }
     hoveredThumbImage {
      bIsDynamicallyLoaded=false
      drawAs=RoundedBox
      imageType=FullColor
      mirroring=NoMirror
      resourceName=None
      resourceObject=null
      tiling=NoTile
      imageSize=<12 12>
      margin=<ltrb 0.5 0.5 0.5 0.5>
      outlineSettings {
       bUseBrushTransparency=false
       roundingType=FixedRadius
       width=2
       color {colorUseRule=UseColor_Specified, specifiedColor=<rgba 0 0 0 0>}
       cornerRadii=<6 6 6 6>
      }
      tintColor {colorUseRule=UseColor_Specified, specifiedColor=<rgba 1 1 1 1>}
      uVRegion {bIsValid=false, max=<0 0>, min=<0 0>}
     }
     normalThumbImage {
      bIsDynamicallyLoaded=false
      drawAs=RoundedBox
      imageType=FullColor
      mirroring=NoMirror
      resourceName=None
      resourceObject=null
      tiling=NoTile
      imageSize=<12 12>
      margin=<ltrb 0.5 0.5 0.5 0.5>
      outlineSettings {
       bUseBrushTransparency=false
       roundingType=FixedRadius
       width=2
       color {colorUseRule=UseColor_Specified, specifiedColor=<rgba 0 0 0 0>}
       cornerRadii=<6 6 6 6>
      }
      tintColor {colorUseRule=UseColor_Specified, specifiedColor=<rgba 1 1 1 0.75>}
      uVRegion {bIsValid=false, max=<0 0>, min=<0 0>}
     }
     verticalBackgroundImage {
      bIsDynamicallyLoaded=false
      drawAs=Box
      imageType=FullColor
      mirroring=NoMirror
      resourceName=None
      resourceObject=/FortUILibrary/UIKit/ScrollBar/MaterialInstances/MI_UIKit_ScrollbarBackground
      tiling=NoTile
      imageSize=<12 12>
      margin=<ltrb 0.5 0.5 0.5 0.5>
      outlineSettings {
       bUseBrushTransparency=false
       roundingType=HalfHeightRadius
       width=0
       color {colorUseRule=UseColor_Specified, specifiedColor=<rgba 0 0 0 0>}
       cornerRadii=<0 0 0 0>
      }
      tintColor {colorUseRule=UseColor_Specified, specifiedColor=<rgba 1 1 1 0.2>}
      uVRegion {bIsValid=false, max=<0 0>, min=<0 0>}
     }
     verticalBottomSlotImage {
      bIsDynamicallyLoaded=false
      drawAs=NoDrawType
      imageType=FullColor
      mirroring=NoMirror
      resourceName=../../../Engine/Content/Slate/Common/Scrollbar_Background_Vertical.png
      resourceObject=null
      tiling=NoTile
      imageSize=<8 8>
      margin=<ltrb 0 0 0 0>
      outlineSettings {
       bUseBrushTransparency=false
       roundingType=HalfHeightRadius
       width=0
       color {colorUseRule=UseColor_Specified, specifiedColor=<rgba 0 0 0 0>}
       cornerRadii=<0 0 0 0>
      }
      tintColor {colorUseRule=UseColor_Specified, specifiedColor=<rgba 1 1 1 1>}
      uVRegion {bIsValid=false, max=<0 0>, min=<0 0>}
     }
     verticalTopSlotImage {
      bIsDynamicallyLoaded=false
      drawAs=NoDrawType
      imageType=FullColor
      mirroring=NoMirror
      resourceName=../../../Engine/Content/Slate/Common/Scrollbar_Background_Vertical.png
      resourceObject=null
      tiling=NoTile
      imageSize=<8 8>
      margin=<ltrb 0 0 0 0>
      outlineSettings {
       bUseBrushTransparency=false
       roundingType=HalfHeightRadius
       width=0
       color {colorUseRule=UseColor_Specified, specifiedColor=<rgba 0 0 0 0>}
       cornerRadii=<0 0 0 0>
      }
      tintColor {colorUseRule=UseColor_Specified, specifiedColor=<rgba 1 1 1 1>}
      uVRegion {bIsValid=false, max=<0 0>, min=<0 0>}
     }
    }
   }
  }
 ]
}
