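# Tournament leaderboard modal, built on the UIKit dialog base, showing qualification points, last-updated time, optional top-percentile thresholds and scrollable tiebreaker rules.
# Setters fill the content, hiding the criterion row when points are -1.
# Percentiles and prize notice start collapsed when thresholds are set, reappearing only if feature-gated data yields valid entries.
# Custom up/down navigation scrolls the rules, gamepad switches refocus the dialog, and button clicks log analytics and close it.
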WidgetBlueprint /Compete/TournamentLeaderboard/ModalMessages/QualificationRules/WBP_QualificationRulesModal
graphs [
 BP_GetDesiredFocusTarget {nodes [
   Get Desired Focus Target()
   > Return Node!(execute, ReturnValue:Widget Object Reference=Get Dialog_Base() -> (Dialog_Base:WBP UIKit Dialog Internal Object Reference))
  ]}
 DoCustomNavigation {nodes [
   n1 = DoCustomNavigation!() -> (then, Navigation:EUINavigation Enum) {is=title K2Node_FunctionEntry}
   > n2 = Switch on EUINavigation!(execute, Selection:EUINavigation Enum=@n1) -> (Left:Exec, Right:Exec, Up:Exec, Down:Exec, Next:Exec, Previous:Exec, Invalid:Exec) {is=title K2Node_SwitchEnum}
   n3 = GetScrollOffset!(self:Scroll Box Object Reference=Get ScrollBox() -> (ScrollBox:Scroll Box Object Reference)) -> (ReturnValue:float)
   n4 = SetScrollOffset!(execute=@n2.Up, self:Scroll Box Object Reference=Get ScrollBox(), NewScrollOffset:float=float - float(A=@n3, B=10)) -> (then)
   Return Node(execute=[@n5.then, @n4.then])
   n5 = SetScrollOffset(execute=@n2.Down, self=Get ScrollBox(), NewScrollOffset=float + float(A=@n3, B=10))
  ]}
 EventGraph {nodes [
   n1 = OnInputMethodChanged!() -> (OutputDelegate, then, bNewInputType:ECommonInputType Enum) {is=title K2Node_CustomEvent}
   > n2 = Switch on ECommonInputType!(execute, Selection:ECommonInputType Enum=@n1.bNewInputType) -> (MouseAndKeyboard:Exec, Gamepad:Exec, Touch:Exec, Count:Exec) {is=title K2Node_SwitchEnum}
   SetFocus(execute=@n2.Gamepad, self=Get Dialog_Base())
   OnButtonClicked!() -> (OutputDelegate, then, ButtonIndex:int) {is=title K2Node_CustomEvent}
   > n3 = Cast To FortPlayerController(Object=GetOwningPlayer())
   > n4 = "Fire Event Element Interacted (Non-Localized)"!(
    execute
    FortPC:Fort Player Controller Object Reference=@n3
    ElementName:String=CloseButton
    AdditionalInfo:Flow Analytics Additional Info Structure
   ) -> (then)
   Branch(execute=[@n4.then, @n3.CastFailed], Condition=Get bIsActive() -> (bIsActive:bool))
   > DeactivateWidget()
   OnValidPercentileEntriesAdded() {is=title K2Node_CustomEvent}
   > SetPercentilesVisibility!(execute, self, InVisibility:ESlateVisibility Enum=SelfHitTestInvisible) -> (then)
   Event PreConstruct()
   n5 = SlotAsGridSlot(Widget=Get Slot_Content!(self:WBP UIKit Dialog Internal Object Reference=Get Dialog_Base()) -> (Slot_Content:Named Slot Object Reference))
   > SetPadding!(
    execute
    self:Grid Slot Object Reference=@n5
    InPadding_Left:float=80
    InPadding_Top:float=0
    InPadding_Right:float=80
    InPadding_Bottom:float=0
   ) -> (then)
   > SetLayer!(execute, self:Grid Slot Object Reference=@n5, InLayer:int=10) -> (then)
   > n6 = Construct Object from Class(Class=&/VM_Dialog_QualificationRulesModal.VM_Dialog_QualificationRulesModal_C) -> (ReturnValue:VM Dialog Qualification Rules Modal Object Reference)
   > SetDialogData!(execute, self:WBP UIKit Dialog Internal Object Reference=Get Dialog_Base(), Data:UI Kit Dialog Viewmodel Object Reference=@n6) -> (then)
   > SetState!(
    execute
    self:WBP Eligibility Criterion Object Reference=Get WBP_EligibilityCriterion() -> (WBP_EligibilityCriterion:WBP Eligibility Criterion Object Reference)
    Passed:bool=false
   ) -> (then)
   Event On Activated()
   > Parent: On Activated() {is=title K2Node_CallParentFunction}
   > Fire Event Sub View Opened!(execute, View:Widget Object Reference=Self-Reference(), AdditionalInfo:Flow Analytics Additional Info Structure, bIsOverlay:bool=true) -> (then)
   n7 = Get Dialog_Base()
   > Unbind all Events from Dialog Button Clicked Dispatcher!(execute:Exec, self:WBP UIKit Dialog Internal Object Reference=@n7) -> (then:Exec) {is=title K2Node_ClearDelegate}
   > Bind Event to Dialog Button Clicked Dispatcher(self:WBP UIKit Dialog Internal Object Reference=@n7, Delegate=Create Event()) {is=title K2Node_AddDelegate}
   > Assign On Input Method Changed!(execute:Exec, self:Common Input Object Reference=Get CommonInputSubsystem(), Delegate:"Delegate (by ref)"=@n1) -> (then:Exec) {is=title K2Node_AssignDelegate}
   Event On Deactivated()
   > Parent: On Deactivated() {is=title K2Node_CallParentFunction}
   > Unbind Event from On Input Method Changed(self=Get CommonInputSubsystem(), Delegate=Create Event())
  ]}
 SetCurrentLeaderboardCategory {nodes [
   n1 = SetCurrentLeaderboardCategory!() -> (then, InLeaderboardCategory:Fort Compete UILeaderboard Category Structure) {is=title K2Node_FunctionEntry}
   > Set CurrentLeaderboardCategory(CurrentLeaderboardCategory:Fort Compete UILeaderboard Category Structure=@n1) -> (Output_Get:Fort Compete UILeaderboard Category Structure)
  ]}
 SetLastUpdatedTime {nodes [
   n1 = SetLastUpdatedTime!() -> (then, LastUpdatedTime:Date Time Structure) {is=title K2Node_FunctionEntry}
   > "SetText (Text)"(
    self=Get CommonText_LastUpdatedTime() -> (CommonText_LastUpdatedTime:Common Text Object Reference)
    InText=Format Text!(
     Format="LOCTEXT(\"last updated {Time}\")"
     Time:Text="As Time (from UTC)"!(InDateTime:"Date Time Structure (by ref)"=@n1, InTimeZone:String, InTimeStyle:EDateTimeStyle Enum=Default) -> (ReturnValue:Text)
    ) -> (Result)
   )
  ]}
 SetPercentileThresholds {nodes [
   n1 = SetPercentileThresholds!() -> (then, InThresholds:Array of Tournament Payout Threshold Data Structures, TotalTournamentEntries:int, CanShowPercentiles:bool=~) {is=title K2Node_FunctionEntry}
   > SetPercentilesVisibility(InVisibility=Collapsed)
   > n2 = IsQualificationRulesPercentileFeatureEnabled!(execute) -> (then, ReturnValue:bool)
   > Branch(Condition=AND Boolean!(
     A=@n2
     B=Is Not Empty!(TargetArray:Array of Tournament Payout Threshold Data Structures=@n1) -> (ReturnValue:bool)
     C:bool=@n1.CanShowPercentiles
    ) -> (ReturnValue))
   n3 = Get WBP_QualificationRules_TopPercentiles() -> (WBP_QualificationRules_TopPercentiles:WBP Qualification Rules Percentiles Object Reference)
   > Unbind all Events from On Added Valid Entry!(execute:Exec, self:WBP Qualification Rules Percentiles Object Reference=@n3) -> (then:Exec) {is=title K2Node_ClearDelegate}
   > Bind Event to On Added Valid Entry(self:WBP Qualification Rules Percentiles Object Reference=@n3, Delegate=Create Event()) {is=title K2Node_AddDelegate}
   > SetDialogTitle!(execute, self:WBP UIKit Dialog Internal Object Reference=Get Dialog_Base(), Title:Text="LOCTEXT(\"Scoring Rules\")") -> (then)
   > SetPercentileThresholds!(
    execute
    self:WBP Qualification Rules Percentiles Object Reference=Get WBP_QualificationRules_TopPercentiles()
    InThresholds:Array of Tournament Payout Threshold Data Structures=Get InThresholds() -> (InThresholds:Array of Tournament Payout Threshold Data Structures)
    TotalTournamentEntries:int=Get TotalTournamentEntries() -> (TotalTournamentEntries:int=0)
    LeaderboardCategory:Fort Compete UILeaderboard Category Structure=Get CurrentLeaderboardCategory() -> (CurrentLeaderboardCategory:Fort Compete UILeaderboard Category Structure)
   ) -> (then) {is=title K2Node_CallFunction}
  ]}
 SetPercentilesVisibility {nodes [
   n1 = SetPercentilesVisibility!() -> (then, InVisibility:ESlateVisibility Enum=Visible) {is=title K2Node_FunctionEntry}
   > SetVisibility(self=Get WBP_QualificationRules_TopPercentiles(), InVisibility=@n1)
   > SetVisibility(self=Get CommonText_PrizeDetailsNotice() -> (CommonText_PrizeDetailsNotice:Common Text Object Reference), InVisibility=@n1)
  ]}
 SetPlaceAndPoints {nodes [
   n1 = SetPlaceAndPoints!() -> (then, Points:int) {is=title K2Node_FunctionEntry}
   > n2 = Branch(Condition="Equal (Integer)"(A=@n1, B=-1))
   > SetVisibility(self=Get WBP_EligibilityCriterion(), InVisibility=Collapsed)
   n3 = Get TokenData!(self:Tournament Leaderboard VM Object Reference=Get FortTournamentLeaderboardVM() -> (FortTournamentLeaderboardVM:Tournament Leaderboard VM Object Reference)) -> (TokenData:Tournament Payout Threshold Data Structure)
   SetName!(
    execute=@n2.else
    self:WBP Eligibility Criterion Object Reference=Get WBP_EligibilityCriterion()
    InText:Text=Format Text!(
     Format="LOCTEXT(\"{Position}: {Points} pts\")"
     Position:Text=Break Tournament Payout Threshold Data!(TournamentPayoutThresholdData:"Tournament Payout Threshold Data Structure (by ref)"=@n3) -> (
      ScoringType:EPayoutScoringType Enum
      Threshold:double=~
      ThresholdDisplayText:Text
      ThresholdDisplayTextHeader:Text
      SourceLeaderboardDisplayName:Text
      PayoutItemsData:Array of Tournament Payout Item Data Structures
      CosmeticPayoutItemsVM:Array of Fort Item VM Object References
     ) {is=title K2Node_BreakStruct}.ThresholdDisplayText
     Points:int=Get Points() -> (Points:int=0)
    ) -> (Result)
   ) -> (then)
   > SetVisibility(self=Get WBP_EligibilityCriterion(), InVisibility=SelfHitTestInvisible)
  ]}
]
variables [CurrentLeaderboardCategory {type=FortCompeteUILeaderboardCategory, default {groupingKey="", instanceId="", leaderboardName=""}}]
widgets {
 info {
  inheritedWidgetCount=0
  namedSlotCount=1
  parentClass=&CommonActivatableWidget
  rootWidgetClass=/FortUILibrary/Systems/Dialog/WBP_UIKit_Dialog_Internal.WBP_UIKit_Dialog_Internal_C
  widgetCount=10
 }
 widgets [
  Dialog_Base {
   widgetClassPath=/FortUILibrary/Systems/Dialog/WBP_UIKit_Dialog_Internal
   bIsVariable=true
   widget=Dialog_Base
   properties {
    bIsFocusable=true
    navigation=&self:WidgetTree.Dialog_Base.WidgetNavigation_0
    useClickCatch=true
    buttons {
     each Button Horizontal Fill=true
     each Button Min Width=0
     each Button Size=Medium
     focuse Button=0
     space Between Buttons=30
     cTA Buttons Types [Quiet]
     group Placement {
      horizontal Alignment=HAlign_Fill
      vertical Alignment=VAlign_Bottom
      grid {column=1, column Span=0, layer=0, row=2, row Span=0, nudge=<0 0>}
      padding=<ltrb 80 64 80 48>
     }
    }
    dialogSIze {
     height Override {enable=false, value=610}
     max Aspect Ratio {enable=false, value=0}
     max Desired Height {enable=true, value=1000}
     max Desired Width {enable=false, value=0}
     min Aspect Ratio {enable=false, value=0}
     min Desired Height {enable=false, value=0}
     min Desired Width {enable=false, value=0}
     width Override {enable=true, value=864}
    }
    textProperties {
     auto Wrap Text=true
     justification=Left
     line Height Percentage=1.3
     mobile Font Size Multiplier=1.3
     transform Policy=None
     wrap Text At=0
     color {colorUseRule=UseColor_Specified, specifiedColor=<rgba 1 1 1 0.8>}
     font {
      bForceMonospaced=false
      bMaterialIsStencil=false
      fontMaterial=null
      fontObject=&HeadingNow
      letterSpacing=0
      monospacedWidth=1
      size=15.75
      skewAmount=0
      typefaceFontName=74 Regular
      outlineSettings {
       bApplyOutlineToDropShadows=false
       bMiteredCorners=false
       bSeparateFillAlpha=false
       outlineMaterial=null
       outlineSize=0
       outlineColor=<rgba 0 0 0 1>
      }
     }
     placement {
      horizontal Alignment=HAlign_Fill
      vertical Alignment=VAlign_Top
      grid {column=1, column Span=0, layer=0, row=1, row Span=0, nudge=<0 0>}
      padding=<ltrb 80 0 80 0>
     }
    }
    titleProperties {
     auto Wrap Text=true
     justification=Left
     line Height Percentage=1
     mobile Font Size Multiplier=1.3
     transform Policy=ToUpper
     wrap Text At=0
     color {colorUseRule=UseColor_Specified, specifiedColor=<rgba 1 1 1 1>}
     font {
      bForceMonospaced=false
      bMaterialIsStencil=false
      fontMaterial=null
      fontObject=&HeadingNow
      letterSpacing=0
      monospacedWidth=1
      size=25.5
      skewAmount=0
      typefaceFontName=86 Bold Italic
      outlineSettings {
       bApplyOutlineToDropShadows=false
       bMiteredCorners=false
       bSeparateFillAlpha=false
       outlineMaterial=null
       outlineSize=0
       outlineColor=<rgba 0 0 0 1>
      }
     }
     placement {
      horizontal Alignment=HAlign_Fill
      vertical Alignment=VAlign_Top
      grid {column=1, column Span=0, layer=0, row=0, row Span=0, nudge=<0 0>}
      padding=<ltrb 80 64 80 16>
     }
    }
   }
  }
  Content {
   widgetClassPath=&GridPanel
   namedSlotHost=Dialog_Base
   widget=Content
   properties {flowDirectionPreference=Culture, columnFill [1], rowFill [0,0,1]}
  }
  CommonText_PlaceAndPoints {
   widgetClassPath=&CommonTextBlock
   bIsVariable=true
   parent=Content
   widget=CommonText_PlaceAndPoints
   properties {
    mobileFontSizeMultiplier=1.14
    scrollStyle=&TextScrollStyle-Base_C
    text=Points to qualify are shown dynamically.
    font {
     bForceMonospaced=false
     bMaterialIsStencil=false
     fontMaterial=null
     fontObject=&HeadingNow
     letterSpacing=0
     monospacedWidth=1
     size=15.75
     skewAmount=0
     typefaceFontName=74 Regular
     outlineSettings {
      bApplyOutlineToDropShadows=false
      bMiteredCorners=false
      bSeparateFillAlpha=false
      outlineMaterial=null
      outlineSize=0
      outlineColor=<rgba 0 0 0 1>
     }
    }
    shadowOffset=<0 0>
   }
   slot {class=GridSlot, properties {verticalAlignment=VAlign_Bottom}}
  }
  CommonText_LastUpdatedTime {
   widgetClassPath=&CommonTextBlock
   bIsVariable=true
   parent=Content
   widget=CommonText_LastUpdatedTime
   properties {
    mobileFontSizeMultiplier=1.14
    scrollStyle=&TextScrollStyle-Base_C
    text=last updated 00:00:00
    colorAndOpacity {colorUseRule=UseColor_Specified, specifiedColor=<rgba 0.5776 0.5776 0.5776 0.6>}
    font {
     bForceMonospaced=false
     bMaterialIsStencil=false
     fontMaterial=null
     fontObject=&HeadingNow
     letterSpacing=0
     monospacedWidth=1
     size=15.75
     skewAmount=0
     typefaceFontName=74 Regular
     outlineSettings {
      bApplyOutlineToDropShadows=false
      bMiteredCorners=false
      bSeparateFillAlpha=false
      outlineMaterial=null
      outlineSize=0
      outlineColor=<rgba 0 0 0 1>
     }
    }
    shadowOffset=<0 0>
   }
   slot {class=GridSlot, properties {row=1, verticalAlignment=VAlign_Bottom, padding=<ltrb 0 3 0 0>}}
  }
  WBP_QualificationRules_TopPercentiles {
   widgetClassPath=&/WBP_QualificationRulesPercentiles
   bIsVariable=true
   parent=Content
   widget=WBP_QualificationRules_TopPercentiles
   slot {class=GridSlot, properties {row=2, padding=<ltrb 30 15 30 15>}}
  }
  WBP_EligibilityCriterion {
   widgetClassPath=/Compete/TournamentLeaderboard/ModalMessages/EligibilityCheck/WBP_EligibilityCriterion
   bIsVariable=true
   parent=Content
   widget=WBP_EligibilityCriterion
   properties {criterionName="221 pts to Qualify - Top ##50"}
   slot {class=GridSlot, properties {row=3, padding=<ltrb 7 10 0 0>}}
  }
  CommonText_PrizeDetailsNotice {
   widgetClassPath=&CommonTextBlock
   bIsVariable=true
   parent=Content
   widget=CommonText_PrizeDetailsNotice
   properties {
    mobileFontSizeMultiplier=1.14
    scrollStyle=&TextScrollStyle-Base_C
    text=Prize details available in tournament info.
    font {
     bForceMonospaced=false
     bMaterialIsStencil=false
     fontMaterial=null
     fontObject=&HeadingNow
     letterSpacing=0
     monospacedWidth=1
     size=15.75
     skewAmount=0
     typefaceFontName=74 Regular
     outlineSettings {
      bApplyOutlineToDropShadows=false
      bMiteredCorners=false
      bSeparateFillAlpha=false
      outlineMaterial=null
      outlineSize=0
      outlineColor=<rgba 0 0 0 1>
     }
    }
    shadowOffset=<0 0>
   }
   slot {class=GridSlot, properties {row=4, verticalAlignment=VAlign_Bottom, padding=<ltrb 0 36 0 0>}}
  }
  SizeBox_0 {
   widgetClassPath=&SizeBox
   parent=Content
   widget=SizeBox_0
   properties {bOverride_MaxDesiredHeight=true, bOverride_MinDesiredHeight=true, maxDesiredHeight=500, minDesiredHeight=228}
   slot {class=GridSlot, properties {horizontalAlignment=HAlign_Left, row=5, verticalAlignment=VAlign_Top, padding=<ltrb 0 18 0 0>}}
  }
  ScrollBox {
   widgetClassPath=&ScrollBox
   bIsVariable=true
   parent=SizeBox_0
   widget=ScrollBox
   properties {
    allowOverscroll=false
    scrollbarPadding=<ltrb 10 2 2 2>
    widgetBarStyle {
     thickness=12
     draggedThumbImage {
      bIsDynamicallyLoaded=false
      drawAs=Box
      imageType=FullColor
      mirroring=NoMirror
      resourceName=../../../Engine/Content/Slate/Common/Scrollbar_Thumb.png
      resourceObject=null
      tiling=NoTile
      imageSize=<0 0>
      margin=<ltrb 0.25 0.25 0.25 0.25>
      outlineSettings {
       bUseBrushTransparency=false
       roundingType=HalfHeightRadius
       width=0
       color {colorUseRule=UseColor_Specified, specifiedColor=<rgba 0 0 0 0>}
       cornerRadii=<0 0 0 0>
      }
      tintColor {colorUseRule=UseColor_Specified, specifiedColor=<rgba 1 1 1 1>}
      uVRegion {bIsValid=false, max=<0 0>, min=<0 0>}
     }
     horizontalBackgroundImage {
      bIsDynamicallyLoaded=false
      drawAs=NoDrawType
      imageType=NoImage
      mirroring=NoMirror
      resourceName=None
      resourceObject=null
      tiling=NoTile
      imageSize=<0 0>
      margin=<ltrb 0 0 0 0>
      outlineSettings {
       bUseBrushTransparency=false
       roundingType=HalfHeightRadius
       width=0
       color {colorUseRule=UseColor_Specified, specifiedColor=<rgba 0 0 0 0>}
       cornerRadii=<0 0 0 0>
      }
      tintColor {colorUseRule=UseColor_Specified, specifiedColor=<rgba 1 1 1 1>}
      uVRegion {bIsValid=false, max=<0 0>, min=<0 0>}
     }
     horizontalBottomSlotImage {
      bIsDynamicallyLoaded=false
      drawAs=Image
      imageType=FullColor
      mirroring=NoMirror
      resourceName=../../../Engine/Content/Slate/Common/Scrollbar_Background_Horizontal.png
      resourceObject=null
      tiling=NoTile
      imageSize=<8 8>
      margin=<ltrb 0 0 0 0>
      outlineSettings {
       bUseBrushTransparency=false
       roundingType=HalfHeightRadius
       width=0
       color {colorUseRule=UseColor_Specified, specifiedColor=<rgba 0 0 0 0>}
       cornerRadii=<0 0 0 0>
      }
      tintColor {colorUseRule=UseColor_Specified, specifiedColor=<rgba 1 1 1 1>}
      uVRegion {bIsValid=false, max=<0 0>, min=<0 0>}
     }
     horizontalTopSlotImage {
      bIsDynamicallyLoaded=false
      drawAs=Image
      imageType=FullColor
      mirroring=NoMirror
      resourceName=../../../Engine/Content/Slate/Common/Scrollbar_Background_Horizontal.png
      resourceObject=null
      tiling=NoTile
      imageSize=<8 8>
      margin=<ltrb 0 0 0 0>
      outlineSettings {
       bUseBrushTransparency=false
       roundingType=HalfHeightRadius
       width=0
       color {colorUseRule=UseColor_Specified, specifiedColor=<rgba 0 0 0 0>}
       cornerRadii=<0 0 0 0>
      }
      tintColor {colorUseRule=UseColor_Specified, specifiedColor=<rgba 1 1 1 1>}
      uVRegion {bIsValid=false, max=<0 0>, min=<0 0>}
     }
     hoveredThumbImage {
      bIsDynamicallyLoaded=false
      drawAs=Box
      imageType=FullColor
      mirroring=NoMirror
      resourceName=../../../Engine/Content/Slate/Common/Scrollbar_Thumb.png
      resourceObject=null
      tiling=NoTile
      imageSize=<0 0>
      margin=<ltrb 0.25 0.25 0.25 0.25>
      outlineSettings {
       bUseBrushTransparency=false
       roundingType=HalfHeightRadius
       width=0
       color {colorUseRule=UseColor_Specified, specifiedColor=<rgba 0 0 0 0>}
       cornerRadii=<0 0 0 0>
      }
      tintColor {colorUseRule=UseColor_Specified, specifiedColor=<rgba 1 1 1 1>}
      uVRegion {bIsValid=false, max=<0 0>, min=<0 0>}
     }
     normalThumbImage {
      bIsDynamicallyLoaded=false
      drawAs=Box
      imageType=FullColor
      mirroring=NoMirror
      resourceName=../../../Engine/Content/Slate/Common/Scrollbar_Thumb.png
      resourceObject=null
      tiling=NoTile
      imageSize=<0 0>
      margin=<ltrb 0.25 0.25 0.25 0.25>
      outlineSettings {
       bUseBrushTransparency=false
       roundingType=HalfHeightRadius
       width=0
       color {colorUseRule=UseColor_Specified, specifiedColor=<rgba 0 0 0 0>}
       cornerRadii=<0 0 0 0>
      }
      tintColor {colorUseRule=UseColor_Specified, specifiedColor=<rgba 1 1 1 1>}
      uVRegion {bIsValid=false, max=<0 0>, min=<0 0>}
     }
     verticalBackgroundImage {
      bIsDynamicallyLoaded=false
      drawAs=NoDrawType
      imageType=NoImage
      mirroring=NoMirror
      resourceName=None
      resourceObject=null
      tiling=NoTile
      imageSize=<0 0>
      margin=<ltrb 0 0 0 0>
      outlineSettings {
       bUseBrushTransparency=false
       roundingType=HalfHeightRadius
       width=0
       color {colorUseRule=UseColor_Specified, specifiedColor=<rgba 0 0 0 0>}
       cornerRadii=<0 0 0 0>
      }
      tintColor {colorUseRule=UseColor_Specified, specifiedColor=<rgba 1 1 1 1>}
      uVRegion {bIsValid=false, max=<0 0>, min=<0 0>}
     }
     verticalBottomSlotImage {
      bIsDynamicallyLoaded=false
      drawAs=Image
      imageType=FullColor
      mirroring=NoMirror
      resourceName=../../../Engine/Content/Slate/Common/Scrollbar_Background_Vertical.png
      resourceObject=null
      tiling=NoTile
      imageSize=<8 8>
      margin=<ltrb 0 0 0 0>
      outlineSettings {
       bUseBrushTransparency=false
       roundingType=HalfHeightRadius
       width=0
       color {colorUseRule=UseColor_Specified, specifiedColor=<rgba 0 0 0 0>}
       cornerRadii=<0 0 0 0>
      }
      tintColor {colorUseRule=UseColor_Specified, specifiedColor=<rgba 1 1 1 1>}
      uVRegion {bIsValid=false, max=<0 0>, min=<0 0>}
     }
     verticalTopSlotImage {
      bIsDynamicallyLoaded=false
      drawAs=Image
      imageType=FullColor
      mirroring=NoMirror
      resourceName=../../../Engine/Content/Slate/Common/Scrollbar_Background_Vertical.png
      resourceObject=null
      tiling=NoTile
      imageSize=<8 8>
      margin=<ltrb 0 0 0 0>
      outlineSettings {
       bUseBrushTransparency=false
       roundingType=HalfHeightRadius
       width=0
       color {colorUseRule=UseColor_Specified, specifiedColor=<rgba 0 0 0 0>}
       cornerRadii=<0 0 0 0>
      }
      tintColor {colorUseRule=UseColor_Specified, specifiedColor=<rgba 1 1 1 1>}
      uVRegion {bIsValid=false, max=<0 0>, min=<0 0>}
     }
    }
    widgetStyle {
     barThickness=12
     bottomShadowBrush {
      bIsDynamicallyLoaded=false
      drawAs=Box
      imageType=FullColor
      mirroring=NoMirror
      resourceName=../../../Engine/Content/Slate/Common/ScrollBoxShadowBottom.png
      resourceObject=null
      tiling=NoTile
      imageSize=<16 8>
      margin=<ltrb 0.5 0 0.5 1>
      outlineSettings {
       bUseBrushTransparency=false
       roundingType=HalfHeightRadius
       width=0
       color {colorUseRule=UseColor_Specified, specifiedColor=<rgba 0 0 0 0>}
       cornerRadii=<0 0 0 0>
      }
      tintColor {colorUseRule=UseColor_Specified, specifiedColor=<rgba 1 1 1 1>}
      uVRegion {bIsValid=false, max=<0 0>, min=<0 0>}
     }
     horizontalScrolledContentPadding=<ltrb 0 0 1 0>
     leftShadowBrush {
      bIsDynamicallyLoaded=false
      drawAs=Box
      imageType=FullColor
      mirroring=NoMirror
      resourceName=../../../Engine/Content/Slate/Common/ScrollBoxShadowLeft.png
      resourceObject=null
      tiling=NoTile
      imageSize=<8 16>
      margin=<ltrb 1 0.5 0 0.5>
      outlineSettings {
       bUseBrushTransparency=false
       roundingType=HalfHeightRadius
       width=0
       color {colorUseRule=UseColor_Specified, specifiedColor=<rgba 0 0 0 0>}
       cornerRadii=<0 0 0 0>
      }
      tintColor {colorUseRule=UseColor_Specified, specifiedColor=<rgba 1 1 1 1>}
      uVRegion {bIsValid=false, max=<0 0>, min=<0 0>}
     }
     rightShadowBrush {
      bIsDynamicallyLoaded=false
      drawAs=Box
      imageType=FullColor
      mirroring=NoMirror
      resourceName=../../../Engine/Content/Slate/Common/ScrollBoxShadowRight.png
      resourceObject=null
      tiling=NoTile
      imageSize=<8 16>
      margin=<ltrb 0 0.5 1 0.5>
      outlineSettings {
       bUseBrushTransparency=false
       roundingType=HalfHeightRadius
       width=0
       color {colorUseRule=UseColor_Specified, specifiedColor=<rgba 0 0 0 0>}
       cornerRadii=<0 0 0 0>
      }
      tintColor {colorUseRule=UseColor_Specified, specifiedColor=<rgba 1 1 1 1>}
      uVRegion {bIsValid=false, max=<0 0>, min=<0 0>}
     }
     topShadowBrush {
      bIsDynamicallyLoaded=false
      drawAs=Box
      imageType=FullColor
      mirroring=NoMirror
      resourceName=../../../Engine/Content/Slate/Common/ScrollBoxShadowTop.png
      resourceObject=null
      tiling=NoTile
      imageSize=<16 8>
      margin=<ltrb 0.5 1 0.5 0>
      outlineSettings {
       bUseBrushTransparency=false
       roundingType=HalfHeightRadius
       width=0
       color {colorUseRule=UseColor_Specified, specifiedColor=<rgba 0 0 0 0>}
       cornerRadii=<0 0 0 0>
      }
      tintColor {colorUseRule=UseColor_Specified, specifiedColor=<rgba 1 1 1 1>}
      uVRegion {bIsValid=false, max=<0 0>, min=<0 0>}
     }
     verticalScrolledContentPadding=<ltrb 0 0 0 1>
    }
   }
   slot {class=SizeBoxSlot}
  }
  CommonText_LastUpdatedTime_1 {
   widgetClassPath=&CommonTextBlock
   bIsVariable=true
   parent=ScrollBox
   widget=CommonText_LastUpdatedTime_1
   properties {
    autoWrapText=true
    bIsScrollingEnabled=false
    mobileFontSizeMultiplier=1.14
    scrollStyle=&TextScrollStyle-Base_C
    text="Tiebreakers will be determined in the order presented here: (1) total points scored; (2) total Victory Royales in the session; (3) average eliminations in the session; (4) average placement per match in the session; and finally (5) total seconds survived across all matches."
    colorAndOpacity {colorUseRule=UseColor_Specified, specifiedColor=<rgba 0.5776 0.5776 0.5776 0.6>}
    font {
     bForceMonospaced=false
     bMaterialIsStencil=false
     fontMaterial=null
     fontObject=&HeadingNow
     letterSpacing=0
     monospacedWidth=1
     size=15.75
     skewAmount=0
     typefaceFontName=74 Regular
     outlineSettings {
      bApplyOutlineToDropShadows=false
      bMiteredCorners=false
      bSeparateFillAlpha=false
      outlineMaterial=null
      outlineSize=0
      outlineColor=<rgba 0 0 0 1>
     }
    }
    shadowOffset=<0 0>
   }
   slot {class=ScrollBoxSlot}
  }
 ]
}
