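# A tournament leaderboard stat row widget that shows a name label beside a value in a horizontal box.
# PreConstruct writes ValueName into the label, then calls UpdateStyle.
# SetValueFromInt and SetValueFromFloat format the value text, the latter to one decimal place, and ResetValue restores a "---" placeholder.
# UpdateStyle branches on IsLodaingSceen to swap stylesheet gameplay tags on both text blocks for a loading-screen look.
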WidgetBlueprint /Compete/TournamentLeaderboard/WBP_TournamentLeaderboardStatsItem
graphs [
 EventGraph {nodes [
   Event PreConstruct()
   > "SetText (Text)"(self=Get CommonTextBlock_ValueName() -> (CommonTextBlock_ValueName:Common Text Object Reference), InText=Get ValueName() -> (ValueName:Text))
   > UpdateStyle()
  ]}
 ResetValue {nodes [
   ResetValue() {is=title K2Node_FunctionEntry}
   > "SetText (Text)"(self=Get CommonTextBlock_Value() -> (CommonTextBlock_Value:Common Text Object Reference), InText="LOCTEXT(\"---\")")
  ]}
 SetValueFromFloat {nodes [
   n1 = SetValueFromFloat!() -> (then, Value:double=~) {is=title K2Node_FunctionEntry}
   > "SetText (Text)"(
    self=Get CommonTextBlock_Value()
    InText="To Text (Float)"(Value=@n1, MinimumIntegralDigits=1, MaximumIntegralDigits=324, MinimumFractionalDigits=1, MaximumFractionalDigits=1)
   )
  ]}
 SetValueFromInt {nodes [
   n1 = SetValueFromInt!() -> (then, Value:int) {is=title K2Node_FunctionEntry}
   > "SetText (Text)"(self=Get CommonTextBlock_Value(), InText="To Text (Integer)"(Value=@n1))
  ]}
 UpdateStyle {nodes [
   UpdateStyle() {is=title K2Node_FunctionEntry}
   > Branch(Condition=Get IsLodaingSceen() -> (IsLodaingSceen:bool))
   > UpdateWidgetStyleTags!(
    execute
    InWidget:Widget Object Reference=Get CommonTextBlock_ValueName()
    InAddedTags:Array of Gameplay Tag Structures=Make Array()
    InRemovedTags:Array of Gameplay Tag Structures=Make Array("[0]"=(TagName="StyleSheet.Style.Strong"))
   ) -> (then, ReturnValue:bool)
   > UpdateWidgetStyleTags(
    InWidget=Get CommonTextBlock_Value()
    InAddedTags=Make Array!("[0]"=(TagName="StyleSheet.Style.Heading"), "[1]":Gameplay Tag Structure=(TagName="StyleSheet.Style.10")) -> (Array)
    InRemovedTags=Make Array("[0]"=(TagName="StyleSheet.Style.30"), "[1]"=(TagName="StyleSheet.Style.Body"))
   )
  ]}
]
variables [
 IsLodaingSceen {type=boolean}
 ValueName {type=string, default="[PH]Value"}
]
widgets {
 info {inheritedWidgetCount=0, parentClass=&CommonUserWidget, rootWidgetClass=&HorizontalBox, widgetCount=3}
 widgets [
  Root {widgetClassPath=&HorizontalBox, widget=Root, properties {bOverride_Cursor=true}}
  CommonTextBlock_ValueName {
   widgetClassPath=&CommonTextBlock
   bIsVariable=true
   parent=Root
   widget=CommonTextBlock_ValueName
   properties {
    scrollStyle=&TextScrollStyle-Base_C
    text="[PH] Victory Royales"
    colorAndOpacity {colorUseRule=UseColor_Specified, specifiedColor=<rgba 1 1 1 0.6>}
    font {
     bForceMonospaced=false
     bMaterialIsStencil=false
     fontMaterial=null
     fontObject=&NotoSans
     letterSpacing=0
     monospacedWidth=1
     size=20.25
     skewAmount=0
     typefaceFontName=NotoSans-Regular
     outlineSettings {
      bApplyOutlineToDropShadows=false
      bMiteredCorners=false
      bSeparateFillAlpha=false
      outlineMaterial=null
      outlineSize=0
      outlineColor=<rgba 0 0 0 1>
     }
    }
    shadowOffset=<0 0>
   }
   slot {class=HorizontalBoxSlot, properties {verticalAlignment=VAlign_Center, size {sizeRule=Fill, value=1}}}
  }
  CommonTextBlock_Value {
   widgetClassPath=&CommonTextBlock
   bIsVariable=true
   parent=Root
   widget=CommonTextBlock_Value
   properties {
    text=---
    font {
     bForceMonospaced=false
     bMaterialIsStencil=false
     fontMaterial=null
     fontObject=&NotoSans
     letterSpacing=0
     monospacedWidth=1
     size=20.25
     skewAmount=0
     typefaceFontName=NotoSans-Regular
     outlineSettings {
      bApplyOutlineToDropShadows=false
      bMiteredCorners=false
      bSeparateFillAlpha=false
      outlineMaterial=null
      outlineSize=0
      outlineColor=<rgba 0 0 0 1>
     }
    }
    shadowOffset=<0 0>
   }
   slot {class=HorizontalBoxSlot}
  }
 ]
}
