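# Tournament leaderboard stats panel showing a selected entry's total points, placement and elimination stats, last-update time and a points-to-qualify banner.
# A Compete leaderboard subsystem event sets the entry viewmodel, while separate functions format points, update time and threshold text from payout data.
# SetDataVisibility toggles and dims the data, resetting stat items when hidden.
# PreConstruct copies rounded-box brush settings onto blur scrims and zeroes the background materials' CornerRoundness.
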WidgetBlueprint /Compete/TournamentLeaderboard/WBP_TournamentLeaderboardStats
graphs [
 EventGraph {nodes [
   n1 = On Set Selected Entry!() -> (OutputDelegate, then, InEntry:Tournament Leaderboard Entry VM Object Reference) {is=title K2Node_CustomEvent}
   > SetVisibility(InVisibility=SelfHitTestInvisible)
   > Set VM_LeaderboardEntry(VM_LeaderboardEntry:Tournament Leaderboard Entry VM Object Reference=@n1.InEntry) -> (Output_Get:Tournament Leaderboard Entry VM Object Reference)
   n2 = SetDataVisibility!() -> (OutputDelegate, then, Visible:bool=~) {is=title K2Node_CustomEvent}
   > SetVisibility(
    self=[
     Get Points_Text() -> (Points_Text:Common Text Object Reference)
     Get CommonTextBlock_LastUpdate() -> (CommonTextBlock_LastUpdate:Common Text Object Reference)
     Get Points() -> (Points:Horizontal Box Object Reference)
     Get Top50_Text() -> (Top50_Text:Common Text Object Reference)
    ]
    InVisibility=Select(Option 0=Hidden, Option 1=HitTestInvisible, Index=@n2.Visible)
   )
   > Set RenderOpacity(RenderOpacity:float=Select(Option 0:float=0.5, Option 1:float=1, Index=@n2.Visible) -> (ReturnValue:float)) -> (Output_Get:float)
   > n3 = Branch(Condition=@n2.Visible)
   ResetValue(
    execute=@n3.else
    self:WBP Tournament Leaderboard Stats Item Object Reference=[
     Get StatsItem_Placements_Top5() -> (StatsItem_Placements_Top5:WBP Tournament Leaderboard Stats Item Object Reference)
     Get VictoryRoyales() -> (VictoryRoyales:WBP Tournament Leaderboard Stats Item Object Reference)
     Get AvgElims() -> (AvgElims:WBP Tournament Leaderboard Stats Item Object Reference)
     Get AvgPlacement() -> (AvgPlacement:WBP Tournament Leaderboard Stats Item Object Reference)
     Get StatsItem_Placements_Top10() -> (StatsItem_Placements_Top10:WBP Tournament Leaderboard Stats Item Object Reference)
     Get StatsItem_TotalElims() -> (StatsItem_TotalElims:WBP Tournament Leaderboard Stats Item Object Reference)
    ]
   )
   Event Construct()
   > Bind Event to On Set Tournament Leaderboard Screen Selected Entry(
    self:Fort Compete Leaderboard Subsystem Object Reference=Get FortCompeteLeaderboardSubsystem() -> (ReturnValue:Fort Compete Leaderboard Subsystem Object Reference) {is=title K2Node_GetSubsystem}
    Delegate=@n1
   ) {is=title K2Node_AddDelegate}
   Event PreConstruct()
   > SetBlurCornerRoundness()
   > SetVisibility(
    self=Get PointsToQualifyOverlay() -> (PointsToQualifyOverlay:Overlay Object Reference)
    InVisibility=Select(
     Option 0:ESlateVisibility Enum=Collapsed
     Option 1:ESlateVisibility Enum=SelfHitTestInvisible
     Index=GetShowPointsToQualify!() -> (ReturnValue:bool)
    ) -> (ReturnValue:ESlateVisibility Enum)
   )
  ]}
 Set Total Points {nodes [
   n1 = Set Total Points!() -> (then, Points:int) {is=title K2Node_FunctionEntry}
   > "SetText (Text)"(self=Get GainedPoints() -> (GainedPoints:Common Text Object Reference), InText="To Text (Integer)"(Value=@n1))
   > SetDataVisibility!(execute, self, Visible:bool=true) -> (then) {is=title K2Node_CallFunction}
  ]}
 Set VM Leaderboard Entry {nodes [
   n1 = Set VM Leaderboard Entry!() -> (then, Input:"Tournament Leaderboard Entry VM Object Reference (by ref)") {is=title K2Node_FunctionEntry}
   > Set VM_LeaderboardEntry(VM_LeaderboardEntry=@n1)
  ]}
 SetBlurCornerRoundness {nodes [
   SetBlurCornerRoundness() {is=title K2Node_FunctionEntry}
   n1 = Get OverlayBlur!(self:WBP UIKit Scrim Object Reference=Get PointsBG_Blur() -> (PointsBG_Blur:WBP UIKit Scrim Object Reference)) -> (OverlayBlur:Image Object Reference)
   n2 = Get!(self:Image Object Reference=@n1) -> (
    Brush_TintColor:Slate Color Structure
    Brush_DrawAs:ESlateBrushDrawType Enum
    Brush_Tiling:ESlateBrushTileType Enum
    Brush_Mirroring:ESlateBrushMirrorType Enum
    Brush_ImageSize:Vector 2D Structure
    Brush_Margin:Margin Structure
    Brush_ResourceObject:Object Reference
    Brush_OutlineSettings:Slate Brush Outline Settings Structure
   )
   n3 = Get(self=Get PointsQuality() -> (PointsQuality:Image Object Reference))
   > SetBrush!(
    execute
    self=[@n1, Get OverlayBlur(self=Get PTSToQualify_BGBlur() -> (PTSToQualify_BGBlur:WBP UIKit Scrim Object Reference))]
    InBrush_TintColor:Slate Color Structure=@n2
    InBrush_DrawAs:ESlateBrushDrawType Enum=@n3.Brush_DrawAs
    InBrush_Tiling:ESlateBrushTileType Enum=@n2.Brush_Tiling
    InBrush_Mirroring:ESlateBrushMirrorType Enum=@n2.Brush_Mirroring
    InBrush_ImageSize:Vector 2D Structure=@n2.Brush_ImageSize
    InBrush_Margin:Margin Structure=@n2.Brush_Margin
    InBrush_ResourceObject:Object Reference=@n2.Brush_ResourceObject
    InBrush_OutlineSettings:Slate Brush Outline Settings Structure=@n3.Brush_OutlineSettings
   ) -> (then)
   > n4 = GetDynamicMaterial(self=Get PointsQuality())
   > SetScalarParameterValue(self=@n4, ParameterName=CornerRoundness, Value=0)
   > n5 = GetDynamicMaterial(self=Get PointsBackground() -> (PointsBackground:Image Object Reference))
   > SetScalarParameterValue(self=@n5, ParameterName=CornerRoundness, Value=0)
  ]}
 SetLastUpdateTime {nodes [
   n1 = SetLastUpdateTime!() -> (then, Time:Date Time Structure) {is=title K2Node_FunctionEntry}
   > SetVisibility(self=Get CommonTextBlock_LastUpdate(), InVisibility=SelfHitTestInvisible)
   > "SetText (Text)"(
    self=Get CommonTextBlock_LastUpdate()
    InText=Format Text!(
     Format="LOCTEXT(\"Last Updated {time}\")"
     time:Text="As Time (from UTC)"!(InDateTime:"Date Time Structure (by ref)"=@n1, InTimeZone:String, InTimeStyle:EDateTimeStyle Enum=Default) -> (ReturnValue:Text)
    ) -> (Result)
   )
  ]}
 SetShowPointsToQualifyDisplay {nodes [
   n1 = SetShowPointsToQualifyDisplay!() -> (then, PayoutThresholdData:Tournament Payout Threshold Data Structure) {is=title K2Node_FunctionEntry}
   n2 = Break Tournament Payout Threshold Data!(TournamentPayoutThresholdData:"Tournament Payout Threshold Data Structure (by ref)"=@n1) -> (
    ScoringType:EPayoutScoringType Enum
    Threshold:double=~
    ThresholdDisplayText:Text
    ThresholdDisplayTextHeader:Text
    SourceLeaderboardDisplayName:Text
    PayoutItemsData:Array of Tournament Payout Item Data Structures
    CosmeticPayoutItemsVM:Array of Fort Item VM Object References
   ) {is=title K2Node_BreakStruct}
   > SetVisibility(
    self=Get PointsToQualifyOverlay()
    InVisibility=Select(
     Option 0=Collapsed
     Option 1=Visible
     Index="Not Equal (Enum)"!(A:EPayoutScoringType Enum=@n2, B:EPayoutScoringType Enum=Invalid) -> (ReturnValue:bool)
    )
   )
  ]}
 SetTop50Text {nodes [
   n1 = SetTop50Text!() -> (then, Points:int) {is=title K2Node_FunctionEntry}
   n2 = Get Top50_Text()
   > n3 = Branch(Condition=AND Boolean(A="Not Equal (Integer)"(A=@n1, B=-1), B=IsVisible(self=@n2)))
   n4 = Get PointsToQualifyOverlay()
   > SetVisibility(self=@n4, InVisibility=SelfHitTestInvisible)
   n5 = Get TokenData!(self:Tournament Leaderboard VM Object Reference=Get FortTournamentLeaderboardVM() -> (FortTournamentLeaderboardVM:Tournament Leaderboard VM Object Reference)) -> (TokenData:Tournament Payout Threshold Data Structure)
   > "SetText (Text)"(
    self=Get Top50_Text()
    InText=Format Text!(
     Format="LOCTEXT(\"{Position}: {Points} pts\")"
     Position:Text=Break Tournament Payout Threshold Data(TournamentPayoutThresholdData=@n5) -> (Threshold=~).ThresholdDisplayText
     Points:int=Get Points() -> (Points:int=0)
    ) -> (Result)
   )
   SetVisibility(execute=@n3.else, self=@n4, InVisibility=Collapsed)
  ]}
]
variables []
widgets {
 info {inheritedWidgetCount=0, parentClass=&CommonUserWidget, rootWidgetClass=&GridPanel, widgetCount=20}
 widgets [
  Root {widgetClassPath=&GridPanel, widget=Root, properties {flowDirectionPreference=Culture, columnFill [1]}}
  CommonTextBlock_LastUpdate {
   widgetClassPath=&CommonTextBlock
   bIsVariable=true
   parent=Root
   widget=CommonTextBlock_LastUpdate
   properties {
    mobileFontSizeMultiplier=1.5
    text=Updated 30s ago
    visibility=Collapsed
    colorAndOpacity {colorUseRule=UseColor_Specified, specifiedColor=<rgba 0.5776 0.5776 0.5776 1>}
    font {
     bForceMonospaced=false
     bMaterialIsStencil=false
     fontMaterial=null
     fontObject=&HeadingNow
     letterSpacing=0
     monospacedWidth=1
     size=12
     skewAmount=0
     typefaceFontName=74 Regular
     outlineSettings {
      bApplyOutlineToDropShadows=false
      bMiteredCorners=false
      bSeparateFillAlpha=false
      outlineMaterial=null
      outlineSize=2
      outlineColor=<rgba 0 0 0 0.2>
     }
    }
    shadowOffset=<0 0>
   }
   slot {class=GridSlot, properties {layer=1, padding=<ltrb 0 0 0 3>}}
  }
  PointsToQualifyOverlay {
   widgetClassPath=&Overlay
   bIsVariable=true
   parent=Root
   widget=PointsToQualifyOverlay
   properties {visibility=Visible}
   slot {class=GridSlot, properties {layer=1, row=1, padding=<ltrb 0 0 0 56>}}
  }
  PTSToQualify_BGBlur {
   widgetClassPath=/FortUILibrary/Systems/Dialog/WBP_UIKit_Scrim
   bIsVariable=true
   parent=PointsToQualifyOverlay
   widget=PTSToQualify_BGBlur
   properties {blur=0.04}
   slot {class=OverlaySlot, properties {horizontalAlignment=HAlign_Fill, verticalAlignment=VAlign_Fill}}
  }
  PointsQuality {
   widgetClassPath=&Image
   bIsVariable=true
   parent=PointsToQualifyOverlay
   widget=PointsQuality
   properties {brush {
     bIsDynamicallyLoaded=false
     drawAs=RoundedBox
     imageType=FullColor
     mirroring=NoMirror
     resourceName=None
     resourceObject=/Compete/MaterialInstances/MI_UI_CompeteLeaderboardBackground
     tiling=NoTile
     imageSize=<560 56>
     margin=<ltrb 0 0 0 0>
     outlineSettings {
      bUseBrushTransparency=false
      roundingType=FixedRadius
      width=0
      color {colorUseRule=UseColor_Specified, specifiedColor=<rgba 0 0 0 0>}
      cornerRadii=<10 10 10 10>
     }
     tintColor {colorUseRule=UseColor_Specified, specifiedColor=<rgba 1 1 1 1>}
     uVRegion {bIsValid=false, max=<0 0>, min=<0 0>}
    }}
   slot {class=OverlaySlot, properties {horizontalAlignment=HAlign_Fill, verticalAlignment=VAlign_Fill}}
  }
  HorizontalBox_36 {
   widgetClassPath=&HorizontalBox
   parent=PointsToQualifyOverlay
   widget=HorizontalBox_36
   slot {class=OverlaySlot, properties {horizontalAlignment=HAlign_Fill, verticalAlignment=VAlign_Fill}}
  }
  Points_Text {
   widgetClassPath=&CommonTextBlock
   bIsVariable=true
   parent=HorizontalBox_36
   widget=Points_Text
   properties {
    justification=Right
    text=Points to qualify
    visibility=Collapsed
    font {
     bForceMonospaced=false
     bMaterialIsStencil=false
     fontMaterial=null
     fontObject=&HeadingNow
     letterSpacing=0
     monospacedWidth=1
     size=18
     skewAmount=0
     typefaceFontName=74 Regular
     outlineSettings {
      bApplyOutlineToDropShadows=false
      bMiteredCorners=false
      bSeparateFillAlpha=false
      outlineMaterial=null
      outlineSize=0
      outlineColor=<rgba 0 0 0 1>
     }
    }
    shadowOffset=<0 0>
   }
   slot {class=HorizontalBoxSlot, properties {horizontalAlignment=HAlign_Right, verticalAlignment=VAlign_Center, padding=<ltrb 18 0 0 0>}}
  }
  Top50_Text {
   widgetClassPath=&CommonTextBlock
   bIsVariable=true
   parent=HorizontalBox_36
   widget=Top50_Text
   properties {
    justification=Right
    text="Top ##50: 221 pts"
    visibility=Collapsed
    font {
     bForceMonospaced=false
     bMaterialIsStencil=false
     fontMaterial=null
     fontObject=&HeadingNow
     letterSpacing=0
     monospacedWidth=1
     size=22.5
     skewAmount=0
     typefaceFontName=74 Regular
     outlineSettings {
      bApplyOutlineToDropShadows=false
      bMiteredCorners=false
      bSeparateFillAlpha=false
      outlineMaterial=null
      outlineSize=0
      outlineColor=<rgba 0 0 0 1>
     }
    }
    shadowOffset=<0 0>
   }
   slot {
    class=HorizontalBoxSlot
    properties {horizontalAlignment=HAlign_Right, verticalAlignment=VAlign_Center, padding=<ltrb 0 0 9 0>, size {sizeRule=Fill, value=1}}
   }
  }
  PointsBG_Blur {
   widgetClassPath=&WBP_UIKit_Scrim
   bIsVariable=true
   parent=Root
   widget=PointsBG_Blur
   properties {blur=0.04}
   slot {class=GridSlot, properties {row=3}}
  }
  PointsBackground {
   widgetClassPath=&Image
   bIsVariable=true
   parent=Root
   widget=PointsBackground
   properties {brush {
     bIsDynamicallyLoaded=false
     drawAs=RoundedBox
     imageType=FullColor
     mirroring=NoMirror
     resourceName=None
     resourceObject=&MI_UI_CompeteLeaderboardBackground
     tiling=NoTile
     imageSize=<560 100>
     margin=<ltrb 0 0 0 0>
     outlineSettings {
      bUseBrushTransparency=false
      roundingType=FixedRadius
      width=0
      color {colorUseRule=UseColor_Specified, specifiedColor=<rgba 0 0 0 0>}
      cornerRadii=<10 10 10 10>
     }
     tintColor {colorUseRule=UseColor_Specified, specifiedColor=<rgba 1 1 1 1>}
     uVRegion {bIsValid=false, max=<0 0>, min=<0 0>}
    }}
   slot {class=GridSlot, properties {layer=1, row=3}}
  }
  Points {
   widgetClassPath=&HorizontalBox
   bIsVariable=true
   parent=Root
   widget=Points
   properties {bOverride_Cursor=true, visibility=Collapsed}
   slot {class=GridSlot, properties {horizontalAlignment=HAlign_Center, layer=1, row=3, verticalAlignment=VAlign_Center, padding=<ltrb 0 0 0 2>}}
  }
  GainedPoints {
   widgetClassPath=&CommonTextBlock
   bIsVariable=true
   parent=Points
   widget=GainedPoints
   properties {
    justification=Center
    text=---
    font {
     bForceMonospaced=false
     bMaterialIsStencil=false
     fontMaterial=null
     fontObject=&HeadingNow
     letterSpacing=-40
     monospacedWidth=1
     size=41.25
     skewAmount=0
     typefaceFontName=86 Bold Italic
     outlineSettings {
      bApplyOutlineToDropShadows=false
      bMiteredCorners=false
      bSeparateFillAlpha=false
      outlineMaterial=null
      outlineSize=0
      outlineColor=<rgba 0 0 0 1>
     }
    }
    shadowOffset=<0 0>
   }
   slot {class=HorizontalBoxSlot}
  }
  GainedPointsText {
   widgetClassPath=&CommonTextBlock
   parent=Points
   widget=GainedPointsText
   properties {
    justification=Center
    text=pts
    font {
     bForceMonospaced=false
     bMaterialIsStencil=false
     fontMaterial=null
     fontObject=&HeadingNow
     letterSpacing=-40
     monospacedWidth=1
     size=32.25
     skewAmount=0
     typefaceFontName=86 Bold Italic
     outlineSettings {
      bApplyOutlineToDropShadows=false
      bMiteredCorners=false
      bSeparateFillAlpha=false
      outlineMaterial=null
      outlineSize=0
      outlineColor=<rgba 0 0 0 1>
     }
    }
    shadowOffset=<0 0>
   }
   slot {class=HorizontalBoxSlot, properties {verticalAlignment=VAlign_Center, padding=<ltrb 10 0 0 0>}}
  }
  StatsContainer {
   widgetClassPath=/Script/UMG.UniformGridPanel
   parent=Root
   widget=StatsContainer
   properties {slotPadding=<ltrb 0 0 0 27>}
   slot {class=GridSlot, properties {layer=1, row=4, padding=<ltrb 0 42 0 0>}}
  }
  VictoryRoyales {
   widgetClassPath=&/WBP_TournamentLeaderboardStatsItem
   bIsVariable=true
   parent=StatsContainer
   widget=VictoryRoyales
   properties {valueName=Victory Royales}
   slot {class=UniformGridSlot, properties {horizontalAlignment=HAlign_Fill}}
  }
  StatsItem_Placements_Top5 {
   widgetClassPath=&WBP_TournamentLeaderboardStatsItem
   bIsVariable=true
   parent=StatsContainer
   widget=StatsItem_Placements_Top5
   properties {valueName=Top 5s}
   slot {class=UniformGridSlot, properties {horizontalAlignment=HAlign_Fill, row=1}}
  }
  StatsItem_Placements_Top10 {
   widgetClassPath=&WBP_TournamentLeaderboardStatsItem
   bIsVariable=true
   parent=StatsContainer
   widget=StatsItem_Placements_Top10
   properties {valueName=Top 10s}
   slot {class=UniformGridSlot, properties {horizontalAlignment=HAlign_Fill, row=2}}
  }
  StatsItem_TotalElims {
   widgetClassPath=&WBP_TournamentLeaderboardStatsItem
   bIsVariable=true
   parent=StatsContainer
   widget=StatsItem_TotalElims
   properties {valueName=Total Elims}
   slot {class=UniformGridSlot, properties {horizontalAlignment=HAlign_Fill, row=3}}
  }
  AvgElims {
   widgetClassPath=&WBP_TournamentLeaderboardStatsItem
   bIsVariable=true
   parent=StatsContainer
   widget=AvgElims
   properties {valueName=Avg. Elims}
   slot {class=UniformGridSlot, properties {horizontalAlignment=HAlign_Fill, row=4}}
  }
  AvgPlacement {
   widgetClassPath=&WBP_TournamentLeaderboardStatsItem
   bIsVariable=true
   parent=StatsContainer
   widget=AvgPlacement
   properties {valueName=Avg. Placement}
   slot {class=UniformGridSlot, properties {horizontalAlignment=HAlign_Fill, row=5}}
  }
 ]
}
