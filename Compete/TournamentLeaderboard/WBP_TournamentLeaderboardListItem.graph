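# Tournament leaderboard row button showing place, teammates with flag icons via rich text, points, matches, wins and elims, hiding the last three for series-point leaderboards.
# Setter functions fill the texts, live dot and alternating or local-best-team background from the entry view model.
# Selection fades an outline at block-timing speed and, outside series-point boards, reports the entry to FortCompeteUIManager.
# A profile context menu lists players and opens their ranked career stats.
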WidgetBlueprint /Compete/TournamentLeaderboard/WBP_TournamentLeaderboardListItem
animations [Anim_OnSelected {
  binding_count=1
  length_seconds=1
  object=&self:Anim_OnSelected
  sequence {bindings [SelectionOutline {
     id=#1
     possessed_class=CommonBorder
     tracks [{
       type=MovieSceneFloatTrack
       display_name=Render Opacity
       sections [{type=MovieSceneFloatSection, end_frame=20, start_frame=0, channels [None {type=MovieSceneScriptingFloatChannel, keys [[0,0],[20,1]]}]}]
      }]
    }]}
 }]
graphs [
 EventGraph {nodes [
   n1 = Event PreConstruct()
   > Parent: PreConstruct(IsDesignTime=@n1.IsDesignTime)
   > n2 = Sequence!(execute) -> (then_0, then_1, then_2:Exec, then_3:Exec)
   n3 = Is Valid(exec=@n2.then_1, InputObject=Get ContextMenuActionContext() -> (ContextMenuActionContext:UIKit Context Menu Action Context Object Reference))
   n4 = Construct Object from Class(execute=@n3.Is Not Valid, Class=/Script/FortniteUI.UIKitContextMenuActionContext) -> (ReturnValue:UIKit Context Menu Action Context Object Reference)
   > Set ContextMenuActionContext(ContextMenuActionContext:UIKit Context Menu Action Context Object Reference=@n4) -> (Output_Get:UIKit Context Menu Action Context Object Reference)
   Branch(execute=@n2.then_2, Condition=IsMobileGame())
   n5 = Self-Reference()
   > SetMinDimensions!(execute, self:Common Button Base Object Reference=@n5, InMinWidth:int=Get MinWidth(None=@n5) -> (MinWidth:int=0), InMinHeight:int=105) -> (then)
   RegisterModularBlock!(
    execute=@n2.then_0
    self:UIKit Modular Button Object Reference
    Block:UIKit Block Object Reference=Get Block_Outline() -> (Block_Outline:WBP UIKit Block Outline Object Reference)
   ) -> (then)
   > n6 = Branch(Condition=Get Is Local Player Best Team Item() -> (Is Local Player Best Team Item:bool))
   > RemoveFromParent(self=Get StatusIndicator_Live() -> (StatusIndicator_Live:WBP UIKit Status Indicator Object Reference))
   RemoveFromParent(execute=@n6.else, self=Get InputActionWidget() -> (InputActionWidget:Common Action Widget Object Reference))
   Event On Selected()
   > Parent: On Selected() {is=title K2Node_CallParentFunction}
   n7 = float / float(
    A=1
    B=Break UIKit Block Timing!(UIKitBlockTiming:"UIKit Block Timing Structure (by ref)"=GetBlockTimings!() -> (BlockTiming:UIKit Block Timing Structure)) -> (Selecting:float=~) {is=title K2Node_BreakStruct}
   )
   > PlayAnimation(InAnimation=Get Anim_OnSelected() -> (Anim_OnSelected:Widget Animation Object Reference), PlaybackSpeed=@n7)
   n8 = Get Leaderboard Entry Data() -> (Leaderboard Entry Data:Tournament Leaderboard Entry VM Object Reference)
   > Branch(Condition=NOT Boolean(A=Get bIsSeriesPointLeaderboards!(self:Tournament Leaderboard Entry VM Object Reference=@n8) -> (bIsSeriesPointLeaderboards:bool)))
   > SetTournamentLeaderboardScreenSelectedEntry!(
    execute
    self:Fort Compete UIManager Object Reference=Get FortCompeteUIManager() -> (ReturnValue:Fort Compete UIManager Object Reference) {is=title K2Node_GetSubsystem}
    InEntry:Tournament Leaderboard Entry VM Object Reference=Get Leaderboard Entry Data()
   ) -> (then)
   Event On Deselected()
   > Parent: On Deselected() {is=title K2Node_CallParentFunction}
   > PlayAnimation(InAnimation=Get Anim_OnSelected(), PlayMode=Reverse, PlaybackSpeed=@n7)
   n9 = Event OnListItemObjectSet()
   > n10 = Cast To FortTournamentLeaderboardEntryData!(execute, Object=@n9.ListItemObject) -> (then, CastFailed, AsTournament Leaderboard Entry VM:Tournament Leaderboard Entry VM Object Reference) {is=title K2Node_DynamicCast}
   > Set Leaderboard Entry Data(Leaderboard Entry Data:Tournament Leaderboard Entry VM Object Reference=@n10) -> (Output_Get:Tournament Leaderboard Entry VM Object Reference)
   Event On Focused() {is=title K2Node_Event}
   > RequestTransitionOnBlocks!(execute, self:UIKit Modular Button Object Reference, State:EUIKitBlockVisualState Enum=Hovered) -> (then)
   Event On Unfocused() {is=title K2Node_Event}
   > RequestTransitionOnBlocks(State=Unhovered)
   n11 = Get Is Local Player Best Team Item()
   Branch(Condition=AND Boolean(A=OnProfileClicked!() -> (OutputDelegate, then, OpenContextMenu:bool=~) {is=title K2Node_CustomEvent}.OpenContextMenu, B=NOT Boolean(A=@n11)))
   n12 = Get Leaderboard Entry Data()
   > Compare Int!(
    Exec:Exec
    Input:int=Length(TargetArray:Array of Strings=Get TeammatesAccountIds!(self:Fort Tournament Base Entry VM Object Reference=@n12) -> (TeammatesAccountIds:Array of Strings))
    CompareWith:int
   ) -> (">":Exec, "==":Exec, "<":Exec) {is=title K2Node_MacroInstance}
   > OpenContextMenu()
   SetVisibility(self=Get ContextMenu() -> (ContextMenu:WBP UIKit Context Menu Anchor Button Object Reference), InVisibility=Visible)
   SetVisibility(self=Get ContextMenu(), InVisibility=Collapsed)
   "On Context Menu Closed (ContextMenu)"() {is=title K2Node_ComponentBoundEvent}
   "On Context Menu Opened (ContextMenu)"() {is=title K2Node_ComponentBoundEvent}
  ]}
 GenerateContext {nodes [
   GenerateContext() {is=title K2Node_FunctionEntry}
   > RemoveAllInteractions(self:UIKit Context Menu Action Context Object Reference=Get ContextMenuActionContext())
   > n1 = Values!(
    execute
    TargetMap:Map of Strings to Tournament Player Profile VM Object References=Get PlayersProfiles!(self:Fort Tournament Base Entry VM Object Reference=Get Leaderboard Entry Data()) -> (PlayersProfiles:Map of Strings to Tournament Player Profile VM Object References)
   ) -> (then, Values:Array of Tournament Player Profile VM Object References)
   > n2 = For Each Loop(Array:Array of Tournament Player Profile VM Object References=@n1) -> (Array Element:Tournament Player Profile VM Object Reference)
   SetContext!(
    execute=@n2.Completed
    self:WBP UIKit Context Menu Anchor Button Base Object Reference=Get ContextMenu()
    InteractionContext_Data:UIKit Context Menu Action Context Object Reference=Get ContextMenuActionContext()
   ) -> (then)
   AddInteraction!(
    execute=@n2.LoopBody
    self:UIKit Context Menu Action Context Object Reference=Get ContextMenuActionContext()
    InActionInfo_DisplayName:Text="To Text (String)"(InString=Get DisplayName!(self:Tournament Player Profile VM Object Reference=@n2) -> (DisplayName:String))
    InActionInfo_IconTexture:Texture 2D Soft Object Reference
    InActionInfo_IconMaterial:Material Instance Soft Object Reference
    InActionInfo_SortPriority:int=1200
    InActionInfo_bIsAvailable:bool=True
    InActionInfo_bIsEnabled:bool=True
    InActionInfo_bIsSelected:bool=False
    InActionInfo_bAddDivider:bool=False
    InActionInfo_bPositionDividerAfter:bool=False
    InActionInfo_bCloseMenuOnSelect:bool=False
    InActionInfo_InteractionNameID:Name=String To Name(InString=Get AccountId!(self:Tournament Player Profile VM Object Reference=@n2) -> (AccountId:String))
    InActionInfo_Interaction:Data Table Row Handle Structure=(DataTable=None,RowName="")
    InActionInfo_PressedDelegate:Delegate="(null).None"
    InActionInfo_ReleasedDelegate:Delegate="(null).None"
    InInteractionDelegate:"Delegate (by ref)"=Create Event()
   ) -> (then)
  ]}
 OnContextClicked {nodes [
   n1 = OnContextClicked!() -> (then, InteractionNameID:Name) {is=title K2Node_FunctionEntry}
   > ShowCareerStatsScreen!(
    execute
    self:Fort Compete UIManager Object Reference=Get FortCompeteUIManager()
    AccountId:String="To String (Name)"(InName=@n1)
    GameMode:String=br-combined
    bIsRanked:bool=true
    InSeasonTrack:String=GetLatestSeasonTrackGUID!(
     self:Fort Stats Page Manager Object Reference=Get FortStatsPageManager() -> (ReturnValue:Fort Stats Page Manager Object Reference) {is=title K2Node_GetSubsystem}
     GameMode:"Name (by ref)"=MakeLiteralName(Value=br-combined)
    ) -> (ReturnValue:String)
    InLimitedGameModeIndex:int=-1
   ) -> (then)
  ]}
 OpenContextMenu {nodes [
   OpenContextMenu() {is=title K2Node_FunctionEntry}
   > GenerateContext() {is=title K2Node_CallFunction}
   > ToggleContextMenu(self:WBP UIKit Context Menu Anchor Button Base Object Reference=Get ContextMenu())
  ]}
 Set Entry Index {nodes [
   n1 = Set Entry Index!() -> (then, Index:int) {is=title K2Node_FunctionEntry}
   n2 = "Equal (Integer)"(A="% (Integer)"(A=@n1, B=2) -> (ReturnValue=0))
   > SetBrushFromMaterial(
    self=Get Background() -> (Background:Image Object Reference)
    Material=SelectMaterialInstance!(
     A:Material Instance Object Reference=/Compete/MaterialInstances/MI_UI_CompeteLeaderboardBackground
     B:Material Instance Object Reference=SelectMaterialInstance(
      A=&MI_UI_CompeteLeaderboardBackground
      B=/Compete/MaterialInstances/MI_UI_CompeteLeaderboardTileBackground
      bSelectA=@n2
     )
     bSelectA:bool=Get Is Local Player Best Team Item()
    ) -> (ReturnValue:Material Instance Object Reference)
   )
  ]}
 Set Teammates {nodes [
   n1 = Set Teammates!() -> (then, Players Profiles:Map of Strings to Tournament Player Profile VM Object References) {is=title K2Node_FunctionEntry}
   > SetPlayersInfo!(execute, self, Players Profiles:Map of Strings to Tournament Player Profile VM Object References=@n1) -> (then)
  ]}
 SetElims {nodes [
   n1 = SetElims!() -> (then, Elims:int) {is=title K2Node_FunctionEntry}
   n2 = Get Leaderboard Entry Data()
   > Is Valid(InputObject=@n2)
   > n3 = Branch(Condition=Get bIsSeriesPointLeaderboards(self=@n2))
   "SetText (Text)"(execute=@n3.else, self=Get Elims() -> (Elims:Common Text Object Reference), InText="To Text (Integer)"(Value=@n1))
  ]}
 SetIsLive {nodes [
   n1 = SetIsLive!() -> (then, IsLive:"String (by ref)") {is=title K2Node_FunctionEntry}
   > n2 = Get!(execute:Exec) -> (then:Exec, else:Exec, StatusIndicator_Live:WBP UIKit Status Indicator Object Reference)
   > Set Bang State!(execute, self:Fort Bang Wrapper NUI Object Reference=@n2, bEnabled:bool=NOT Boolean(A=IsEmpty(InString=@n1)), Count:int=0) -> (then) {is=title K2Node_CallFunction}
  ]}
 SetIsSeriesPointLeaderboards {nodes [
   n1 = SetIsSeriesPointLeaderboards!() -> (then, IsSeriesPointLeaderboards:bool=~) {is=title K2Node_FunctionEntry}
   > SetVisibility(
    self=[Get Matches() -> (Matches:Common Text Object Reference), Get Wins() -> (Wins:Common Text Object Reference), Get Elims()]
    InVisibility=Select(Option 0=HitTestInvisible, Option 1=Collapsed, Index=@n1)
   )
  ]}
 SetMatches {nodes [
   n1 = SetMatches!() -> (then, Amount:int) {is=title K2Node_FunctionEntry}
   n2 = Get Leaderboard Entry Data()
   > Is Valid(InputObject=@n2)
   > n3 = Branch(Condition=Get bIsSeriesPointLeaderboards(self=@n2))
   "SetText (Text)"(execute=@n3.else, self=Get Matches(), InText="To Text (Integer)"(Value=@n1))
  ]}
 SetPercentile {nodes [
   n1 = SetPercentile!() -> (then, Percentile:double=~) {is=title K2Node_FunctionEntry}
   > Set Percentile(Percentile:double=@n1) -> (Output_Get:double)
  ]}
 SetPlace {nodes [
   n1 = SetPlace!() -> (then, Index:int) {is=title K2Node_FunctionEntry}
   > "SetText (Text)"(self=Get Place() -> (Place:Common Text Object Reference), InText=Format Text!(Format="LOCTEXT(\"##{Place}\")", Place:int=@n1) -> (Result))
  ]}
 SetPlayersInfo {nodes [
   n1 = SetPlayersInfo!() -> (then, Players Profiles:Map of Strings to Tournament Player Profile VM Object References) {is=title K2Node_FunctionEntry}
   > n2 = Keys!(execute, TargetMap:Map of Strings to Tournament Player Profile VM Object References=@n1) -> (then, Keys:Array of Strings)
   > n3 = For Each Loop(Array:Array of Strings=@n2) -> (Array Element:String)
   n4 = Find(TargetMap:Map of Strings to Tournament Player Profile VM Object References=@n1, Key=@n3) -> (Value:Tournament Player Profile VM Object Reference)
   n5 = Get DisplayName(self=@n4)
   n6 = Get FlagName!(self:Tournament Player Profile VM Object Reference=@n4) -> (FlagName:String)
   n7 = Select(
    Option 0:String="To String (Text)"(InText=Format Text!(Format="LOCTEXT(\"<img id=\\\"{Country}\\\" size=\\\"18,18\\\" height=\\\"25\\\"/> {Nickname}\")", Country:String=@n6, Nickname:String=@n5) -> (Result))
    Option 1:String=@n5
    Index="Equal Exactly (String)"(A=@n6, B=fortnite)
   ) -> (ReturnValue:String)
   n8 = Get PlayersInfo() -> (PlayersInfo:String)
   n9 = "Equal (Integer)"(A=Last Index!(TargetArray=@n2) -> (ReturnValue=0), B=@n3.Array Index)
   > Set PlayersInfo(PlayersInfo:String=Append(A=Append(A=@n8, B=@n7), B=Select(Option 0=" && ", Index=@n9))) -> (Output_Get:String)
   SetText(
    execute=@n3.Completed
    self=Get CommonRichText_PlayersInfo() -> (CommonRichText_PlayersInfo:Common Rich Text Block Object Reference)
    InText="To Text (String)"(InString=Get PlayersInfo())
   )
  ]}
 SetPoints {nodes [
   n1 = SetPoints!() -> (then, Score:int) {is=title K2Node_FunctionEntry}
   > "SetText (Text)"(self=Get Points() -> (Points:Common Text Object Reference), InText="To Text (Integer)"(Value=@n1))
  ]}
 SetWins {nodes [
   n1 = SetWins!() -> (then, Wins:int) {is=title K2Node_FunctionEntry}
   n2 = Get Leaderboard Entry Data()
   > Is Valid(InputObject=@n2)
   > n3 = Branch(Condition=Get bIsSeriesPointLeaderboards(self=@n2))
   "SetText (Text)"(execute=@n3.else, self=Get Wins(), InText="To Text (Integer)"(Value=@n1))
  ]}
]
variables [
 ContextMenuActionContext {type=UIKitContextMenuActionContext, default=null}
 Is Local Player Best Team Item {type=boolean}
 Percentile {type=number}
 PlaceFont {
  type=SlateFontInfo
  default {
   bForceMonospaced=false
   bMaterialIsStencil=false
   fontMaterial=/FortUILibrary/UIKit/Buttons/MaterialInstances/SelectableButtons/MI_UIKit_SelectableText_Quiet
   fontObject=&HeadingNow
   letterSpacing=0
   monospacedWidth=1
   size=18
   skewAmount=0
   typefaceFontName=75 Medium
   outlineSettings {
    bApplyOutlineToDropShadows=false
    bMiteredCorners=false
    bSeparateFillAlpha=false
    outlineMaterial=null
    outlineSize=0
    outlineColor=<rgba 0 0 0 1>
   }
  }
 }
 Stats_MinWidth {type=number, default=110}
 StatsFont {
  type=SlateFontInfo
  default {
   bForceMonospaced=false
   bMaterialIsStencil=false
   fontMaterial=/Compete/MaterialInstances/MI_CompeteLeaderboard_SelectableText
   fontObject=&HeadingNow
   letterSpacing=0
   monospacedWidth=1
   size=18
   skewAmount=0
   typefaceFontName=75 Medium
   outlineSettings {
    bApplyOutlineToDropShadows=false
    bMiteredCorners=false
    bSeparateFillAlpha=false
    outlineMaterial=null
    outlineSize=0
    outlineColor=<rgba 0 0 0 1>
   }
  }
 }
]
widgets {
 info {
  inheritedWidgetCount=1
  parentClass=/Compete/WBP_CompeteTileButton_Base.WBP_CompeteTileButton_Base_C
  rootWidgetClass=&GridPanel
  widgetCount=14
 }
 widgets [
  Root {
   widgetClassPath=&GridPanel
   widget=Root
   properties {
    flowDirectionPreference=Culture
    navigation=&self:WidgetTree.Root.WidgetNavigation_0
    columnFill [1,0,0]
    rowFill [1]
   }
  }
  Background {
   widgetClassPath=&Image
   bIsVariable=true
   parent=Root
   widget=Background
   properties {brush {
     bIsDynamicallyLoaded=false
     drawAs=Image
     imageType=FullColor
     mirroring=NoMirror
     resourceName=None
     resourceObject=&MI_UI_CompeteLeaderboardTileBackground
     tiling=NoTile
     imageSize=<32 32>
     margin=<ltrb 0 0 0 0>
     outlineSettings {
      bUseBrushTransparency=false
      roundingType=HalfHeightRadius
      width=0
      color {colorUseRule=UseColor_Specified, specifiedColor=<rgba 0 0 0 0>}
      cornerRadii=<0 0 0 0>
     }
     tintColor {colorUseRule=UseColor_Specified, specifiedColor=<rgba 1 1 1 1>}
     uVRegion {bIsValid=false, max=<0 0>, min=<0 0>}
    }}
   slot {class=GridSlot}
  }
  Block_Outline {
   widgetClassPath=/FortUILibrary/UIKit/Buttons/Blocks/WBP_UIKit_Block_Outline
   bIsVariable=true
   parent=Root
   widget=Block_Outline
   properties {bottomLeftCornerRadius=25, bottomRightCornerRadius=25, radiusInPixels=true, topLeftCornerRadius=25, topRightCornerRadius=25}
   slot {class=GridSlot, properties {layer=20}}
  }
  InputActionWidget {
   widgetClassPath=&CommonActionWidget
   bInherited=true
   bIsVariable=true
   parent=Root
   widget=InputActionWidget
   properties {designTimeKey=Gamepad_RightStick_Down, visibility=Collapsed}
   slot {class=GridSlot, properties {horizontalAlignment=HAlign_Left, layer=1, verticalAlignment=VAlign_Center, padding=<ltrb 12 0 0 0>}}
  }
  ContextMenu {
   widgetClassPath=/FortUILibrary/UIKit/ContextMenu/WBP_UIKit_ContextMenuAnchorButton
   bIsVariable=true
   parent=Root
   widget=ContextMenu
   properties {
    button Background Material=null
    button Outline Material=null
    buttonClass=/Game/UI/Frontend/Lobby/Nameplates/WBP_EmptyButton_ContextMenu.WBP_EmptyButton_ContextMenu_C
    buttonIcon=null
    buttonText=""
    closeMenuOnContextClicked=true
    visibility=Collapsed
    button Padding=<ltrb 0 0 0 0>
    button Text Font {
     bForceMonospaced=false
     bMaterialIsStencil=false
     fontMaterial=null
     fontObject=&Burbank
     letterSpacing=180
     monospacedWidth=1
     size=26.25
     skewAmount=0
     typefaceFontName=BurbankRegular-Black
     outlineSettings {
      bApplyOutlineToDropShadows=false
      bMiteredCorners=false
      bSeparateFillAlpha=false
      outlineMaterial=null
      outlineSize=0
      outlineColor=<rgba 0 0 0 1>
     }
    }
   }
   slot {class=GridSlot, properties {horizontalAlignment=HAlign_Right, verticalAlignment=VAlign_Top}}
  }
  SelectionOutline {
   widgetClassPath=&CommonBorder
   bIsVariable=true
   parent=Root
   widget=SelectionOutline
   properties {
    horizontalAlignment=HAlign_Left
    renderOpacity=0
    verticalAlignment=VAlign_Bottom
    background {
     bIsDynamicallyLoaded=false
     drawAs=Image
     imageType=FullColor
     mirroring=NoMirror
     resourceName=None
     resourceObject=/Compete/MaterialInstances/MI_UIKit_CompeteLeaderboardTile_SelectedStateOutline
     tiling=NoTile
     imageSize=<0 0>
     margin=<ltrb 0 0 0 0>
     outlineSettings {
      bUseBrushTransparency=false
      roundingType=HalfHeightRadius
      width=0
      color {colorUseRule=UseColor_Specified, specifiedColor=<rgba 0 0 0 0>}
      cornerRadii=<0 0 0 0>
     }
     tintColor {colorUseRule=UseColor_Specified, specifiedColor=<rgba 1 1 1 1>}
     uVRegion {bIsValid=false, max=<0 0>, min=<0 0>}
    }
    contentColorAndOpacity=<rgba 1 1 1 0.995>
    padding=<ltrb 0 0 0 0>
   }
   slot {class=GridSlot, properties {columnSpan=6, layer=10, padding=<ltrb -2 -2 -2 -2>}}
  }
  Content {widgetClassPath=&HorizontalBox, parent=Root, widget=Content, slot {class=GridSlot}}
  StatusIndicator_Live {
   widgetClassPath=/FortUILibrary/UIKit/StatusIndicator/WBP_UIKit_StatusIndicator
   bIsVariable=true
   parent=Content
   widget=StatusIndicator_Live
   properties {configuration=Dot}
   slot {class=HorizontalBoxSlot, properties {horizontalAlignment=HAlign_Left, padding=<ltrb 22 0 -36 0>}}
  }
  Place {
   widgetClassPath=&CommonTextBlock
   bIsVariable=true
   parent=Content
   widget=Place
   properties {
    minDesiredWidth=90
    mobileFontSizeMultiplier=1.2
    text="##999"
    font {
     bForceMonospaced=false
     bMaterialIsStencil=false
     fontMaterial=null
     fontObject=&HeadingNow
     letterSpacing=0
     monospacedWidth=1
     size=18
     skewAmount=0
     typefaceFontName=75 Medium
     outlineSettings {
      bApplyOutlineToDropShadows=false
      bMiteredCorners=false
      bSeparateFillAlpha=false
      outlineMaterial=null
      outlineSize=0
      outlineColor=<rgba 0 0 0 1>
     }
    }
    shadowOffset=<0 0>
   }
   slot {class=HorizontalBoxSlot, properties {verticalAlignment=VAlign_Center, padding=<ltrb 56 0 0 0>}}
  }
  CommonRichText_PlayersInfo {
   widgetClassPath=&CommonRichTextBlock
   bIsVariable=true
   parent=Content
   widget=CommonRichText_PlayersInfo
   properties {
    autoWrapText=true
    bIsScrollingEnabled=false
    bOverrideDefaultStyle=true
    clipping=ClipToBounds
    minDesiredWidth=680
    mobileTextBlockScale=1.2
    text="WWWWWWWWWWWWWWWW && "
    wrapTextAt=680
    decoratorClasses [/Compete/DataTables/BP_FlagsDecorator.BP_FlagsDecorator_C]
    defaultTextStyleOverride {
     overflowPolicy=Clip
     transformPolicy=None
     colorAndOpacity {colorUseRule=UseColor_Specified, specifiedColor=<rgba 1 1 1 1>}
     font {
      bForceMonospaced=false
      bMaterialIsStencil=false
      fontMaterial=&MI_UIKit_SelectableText_Quiet
      fontObject=&HeadingNow
      letterSpacing=0
      monospacedWidth=1
      size=15.75
      skewAmount=0
      typefaceFontName=74 Regular
      outlineSettings {
       bApplyOutlineToDropShadows=false
       bMiteredCorners=false
       bSeparateFillAlpha=false
       outlineMaterial=null
       outlineSize=0
       outlineColor=<rgba 0 0 0 1>
      }
     }
     highlightColor {colorUseRule=UseColor_Specified, specifiedColor=<rgba 0 0 0 0>}
     highlightShape {
      bIsDynamicallyLoaded=false
      drawAs=Image
      imageType=NoImage
      mirroring=NoMirror
      resourceName=None
      resourceObject=null
      tiling=NoTile
      imageSize=<32 32>
      margin=<ltrb 0 0 0 0>
      outlineSettings {
       bUseBrushTransparency=false
       roundingType=HalfHeightRadius
       width=0
       color {colorUseRule=UseColor_Specified, specifiedColor=<rgba 0 0 0 0>}
       cornerRadii=<0 0 0 0>
      }
      tintColor {colorUseRule=UseColor_Specified, specifiedColor=<rgba 1 1 1 0>}
      uVRegion {bIsValid=false, max=<0 0>, min=<0 0>}
     }
     selectedBackgroundColor {colorUseRule=UseColor_Foreground, specifiedColor=<rgba 1 0 1 0>}
     shadowColorAndOpacity=<rgba 0 0 0 0>
     shadowOffset=<0 0>
     strikeBrush {
      bIsDynamicallyLoaded=false
      drawAs=NoDrawType
      imageType=NoImage
      mirroring=NoMirror
      resourceName=None
      resourceObject=null
      tiling=NoTile
      imageSize=<32 32>
      margin=<ltrb 0 0 0 0>
      outlineSettings {
       bUseBrushTransparency=false
       roundingType=HalfHeightRadius
       width=0
       color {colorUseRule=UseColor_Specified, specifiedColor=<rgba 0 0 0 0>}
       cornerRadii=<0 0 0 0>
      }
      tintColor {colorUseRule=UseColor_Specified, specifiedColor=<rgba 1 1 1 -1>}
      uVRegion {bIsValid=false, max=<0 0>, min=<0 0>}
     }
     underlineBrush {
      bIsDynamicallyLoaded=false
      drawAs=Image
      imageType=NoImage
      mirroring=NoMirror
      resourceName=None
      resourceObject=null
      tiling=NoTile
      imageSize=<32 32>
      margin=<ltrb 0 0 0 0>
      outlineSettings {
       bUseBrushTransparency=false
       roundingType=HalfHeightRadius
       width=0
       color {colorUseRule=UseColor_Specified, specifiedColor=<rgba 0 0 0 0>}
       cornerRadii=<0 0 0 0>
      }
      tintColor {colorUseRule=UseColor_Specified, specifiedColor=<rgba 1 1 1 0>}
      uVRegion {bIsValid=false, max=<0 0>, min=<0 0>}
     }
    }
   }
   slot {class=HorizontalBoxSlot, properties {verticalAlignment=VAlign_Center, size {sizeRule=Fill, value=1}}}
  }
  Points {
   widgetClassPath=&CommonTextBlock
   bIsVariable=true
   parent=Content
   widget=Points
   properties {
    justification=Center
    minDesiredWidth=110
    mobileFontSizeMultiplier=1.2
    text=999
    font {
     bForceMonospaced=false
     bMaterialIsStencil=false
     fontMaterial=null
     fontObject=&HeadingNow
     letterSpacing=0
     monospacedWidth=1
     size=18
     skewAmount=0
     typefaceFontName=75 Medium
     outlineSettings {
      bApplyOutlineToDropShadows=false
      bMiteredCorners=false
      bSeparateFillAlpha=false
      outlineMaterial=null
      outlineSize=0
      outlineColor=<rgba 0 0 0 1>
     }
    }
    shadowOffset=<0 0>
   }
   slot {class=HorizontalBoxSlot, properties {verticalAlignment=VAlign_Center, padding=<ltrb 4 0 0 0>}}
  }
  Matches {
   widgetClassPath=&CommonTextBlock
   bIsVariable=true
   parent=Content
   widget=Matches
   properties {
    justification=Center
    minDesiredWidth=110
    mobileFontSizeMultiplier=1.2
    text=999
    colorAndOpacity {colorUseRule=UseColor_Specified, specifiedColor=<rgba 1 1 1 0.6>}
    font {
     bForceMonospaced=false
     bMaterialIsStencil=false
     fontMaterial=null
     fontObject=&HeadingNow
     letterSpacing=0
     monospacedWidth=1
     size=18
     skewAmount=0
     typefaceFontName=75 Medium
     outlineSettings {
      bApplyOutlineToDropShadows=false
      bMiteredCorners=false
      bSeparateFillAlpha=false
      outlineMaterial=null
      outlineSize=0
      outlineColor=<rgba 0 0 0 1>
     }
    }
    shadowOffset=<0 0>
   }
   slot {class=HorizontalBoxSlot, properties {verticalAlignment=VAlign_Center, padding=<ltrb 4 0 0 0>}}
  }
  Wins {
   widgetClassPath=&CommonTextBlock
   bIsVariable=true
   parent=Content
   widget=Wins
   properties {
    justification=Center
    minDesiredWidth=110
    mobileFontSizeMultiplier=1.2
    text=999
    colorAndOpacity {colorUseRule=UseColor_Specified, specifiedColor=<rgba 1 1 1 0.6>}
    font {
     bForceMonospaced=false
     bMaterialIsStencil=false
     fontMaterial=null
     fontObject=&HeadingNow
     letterSpacing=0
     monospacedWidth=1
     size=18
     skewAmount=0
     typefaceFontName=75 Medium
     outlineSettings {
      bApplyOutlineToDropShadows=false
      bMiteredCorners=false
      bSeparateFillAlpha=false
      outlineMaterial=null
      outlineSize=0
      outlineColor=<rgba 0 0 0 1>
     }
    }
    shadowOffset=<0 0>
   }
   slot {class=HorizontalBoxSlot, properties {verticalAlignment=VAlign_Center, padding=<ltrb 4 0 0 0>}}
  }
  Elims {
   widgetClassPath=&CommonTextBlock
   bIsVariable=true
   parent=Content
   widget=Elims
   properties {
    justification=Center
    minDesiredWidth=110
    mobileFontSizeMultiplier=1.2
    text=999
    colorAndOpacity {colorUseRule=UseColor_Specified, specifiedColor=<rgba 1 1 1 0.6>}
    font {
     bForceMonospaced=false
     bMaterialIsStencil=false
     fontMaterial=null
     fontObject=&HeadingNow
     letterSpacing=0
     monospacedWidth=1
     size=18
     skewAmount=0
     typefaceFontName=75 Medium
     outlineSettings {
      bApplyOutlineToDropShadows=false
      bMiteredCorners=false
      bSeparateFillAlpha=false
      outlineMaterial=null
      outlineSize=0
      outlineColor=<rgba 0 0 0 1>
     }
    }
    shadowOffset=<0 0>
   }
   slot {class=HorizontalBoxSlot, properties {verticalAlignment=VAlign_Center, padding=<ltrb 4 0 0 0>}}
  }
 ]
}
