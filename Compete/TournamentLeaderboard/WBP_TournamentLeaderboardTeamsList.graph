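# Tournament leaderboard teams list widget with player search, a Friends toggle, a category context menu and a pinned local-player team row that jumps to that entry.
# Input drives LeaderboardVM filtering and category selection; Set Teams Info fills the list view and switches to the empty state.
# Scrolling requests further pages near the loaded end or missing-page gaps, using two loading-info objects as offsets.
# Scroll position is saved and restored by item across page loads.
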WidgetBlueprint /Compete/TournamentLeaderboard/WBP_TournamentLeaderboardTeamsList
graphs [
 CalcVisibleListElementsNum {nodes [
   CalcVisibleListElementsNum!() -> (then, GeneratedWidget:User Widget Object Reference) {is=title K2Node_FunctionEntry}
   > Unbind all Events from On Entry Generated!(
    execute:Exec
    self:List View Base Object Reference=Get CommonListView_TeamsInfo() -> (CommonListView_TeamsInfo:Common List View Object Reference)
   ) -> (then:Exec) {is=title K2Node_ClearDelegate}
   n1 = GetDesiredSize!(self:Widget Object Reference=Get GeneratedWidget() -> (GeneratedWidget:User Widget Object Reference)) -> (ReturnValue_X:double, ReturnValue_Y:double)
   n2 = Get HeightOverride!(self:Size Box Object Reference=Get SizeBox_TeamsList() -> (SizeBox_TeamsList:Size Box Object Reference)) -> (HeightOverride:float)
   > Set VisibleListElementsNum(VisibleListElementsNum:int=Floor(A=float / float(A=@n2, B=@n1.ReturnValue_Y)) {is=title K2Node_CallFunction}) -> (Output_Get:int=0)
  ]}
 EventGraph {nodes [
   "On Item Scrolled Into View (CommonListView_TeamsInfo)"!() -> (OutputDelegate, then, Item:Object Reference, Widget:User Widget Object Reference) {is=title K2Node_ComponentBoundEvent}
   > n1 = Branch(Condition=Get BlockPagesLoading() -> (BlockPagesLoading:bool))
   > Set BlockPagesLoading(BlockPagesLoading:bool=false) -> (Output_Get:bool)
   n2 = Branch(execute=[@n12.else, @n1.else], Condition=Get NextPageIsRequested() -> (NextPageIsRequested:bool))
   n3 = GetLeaderboardPageToLoadBasedOnOffset!(execute=@n2.else, self) -> (then, Page:int=0)
   > Branch(Condition="integer > integer"(A=@n3))
   > Branch(Condition=CanRequestLeaderboardPage!(
     self:Tournament Leaderboard VM Object Reference=Get LeaderboardVM() -> (LeaderboardVM:Tournament Leaderboard VM Object Reference)
     PageIndex:int=@n3
    ) -> (ReturnValue:bool))
   > Set NextPageIsRequested(NextPageIsRequested:bool=true) -> (Output_Get:bool)
   > ShowLoadingEntriesMessage!(execute, self, bShow:bool=true) -> (then)
   > LeaderboardRequestPage!(execute, self:Tournament Leaderboard VM Object Reference=Get LeaderboardVM(), PageIndex:int=@n3) -> (then)
   "On Clicked (LocalPlayerBestTeamItem)"!() -> (OutputDelegate, then, Button:Common Button Base Object Reference) {is=title K2Node_ComponentBoundEvent}
   > Set BlockPagesLoading(BlockPagesLoading=true)
   n4 = Get CommonListView_TeamsInfo()
   n5 = GetIndexForItem!(
    self:List View Object Reference=@n4
    Item:Object Reference=Get Local Player Best Team Data() -> (Local Player Best Team Data:Tournament Leaderboard Entry VM Object Reference)
   ) -> (ReturnValue:int=0)
   > ScrollIndexIntoView!(execute, self:List View Object Reference=@n4, Index:int=@n5) -> (then)
   > SetSelectedIndex!(execute, self:List View Object Reference=@n4, Index:int=@n5) -> (then)
   > SetFocus(self=@n4)
   n6 = "On Entry Generated (CommonListView_TeamsInfo)"!() -> (OutputDelegate, then, Widget:User Widget Object Reference) {is=title K2Node_ComponentBoundEvent}
   > n7 = Cast To WBP_TournamentLeaderboardListItem!(execute, Object=@n6.Widget) -> (then, CastFailed, AsWBP Tournament Leaderboard List Item:WBP Tournament Leaderboard List Item Object Reference) {is=title K2Node_DynamicCast}
   n8 = OnContextMenuOpened_Event() {is=title K2Node_CustomEvent}
   n9 = Get ContextMenu!(self:WBP Tournament Leaderboard List Item Object Reference=@n7) -> (ContextMenu:WBP UIKit Context Menu Anchor Button Object Reference)
   n10 = Assign On Context Menu Opened!(execute:Exec=@n7.then, self:WBP UIKit Context Menu Anchor Button Base Object Reference=@n9, Delegate:"Delegate (by ref)"=@n8) -> (then:Exec) {is=title K2Node_AssignDelegate}
   n11 = OnContextMenuClosed_Event() {is=title K2Node_CustomEvent}
   Assign On Context Menu Closed!(execute:Exec=@n10.then, self:WBP UIKit Context Menu Anchor Button Base Object Reference=@n9, Delegate:"Delegate (by ref)"=@n11) -> (then:Exec) {is=title K2Node_AssignDelegate}
   "On List View Scrolled (CommonListView_TeamsInfo)"!() -> (OutputDelegate, then, ItemOffset:float, DistanceRemaining:float) {is=title K2Node_ComponentBoundEvent}
   > n12 = Branch(Condition=Get BlockPagesLoading())
   "On Context Menu Opened (Alternative_Button)"() {is=title K2Node_ComponentBoundEvent}
   > Fire Event Element Interacted(Element=Get Alternative_Button() -> (Alternative_Button:WBP UIKit Context Menu Anchor Button Object Reference))
   n13 = "On Item Selection Changed (CommonListView_TeamsInfo)"!() -> (OutputDelegate, then, Item:Object Reference, bIsSelected:bool) {is=title K2Node_ComponentBoundEvent}
   > Branch(Condition=@n13.bIsSelected)
   > Scroll Item Into View!(execute, self:List View Object Reference=Get CommonListView_TeamsInfo(), Item:Object Reference=@n13.Item) -> (then)
   "On Search Added to Focus Path (FilteredInputField)"() {is=title K2Node_ComponentBoundEvent}
   > Fire Event Element Interacted(Element=Get FilteredInputField() -> (FilteredInputField:WBP UIKit Filtered Input Field Object Reference))
   n14 = "On Text Changed (FilteredInputField)"!() -> (OutputDelegate, then, Text:Text) {is=title K2Node_ComponentBoundEvent}
   n15 = TextIsEmpty(InText=Get LastSearchedText() -> (LastSearchedText:Text))
   > Branch(Condition=AND Boolean(A=TextIsEmpty(InText=@n14.Text), B=NOT Boolean(A=@n15)))
   > n16 = Set LastSearchedText(LastSearchedText:Text=@n14.Text) -> (Output_Get:Text)
   n17 = Get LeaderboardVM()
   > FilterCurrentViewEntriesByText!(execute, self:Tournament Leaderboard VM Object Reference=@n17, SearchText:"Text (by ref)"=@n16) -> (then)
   n18 = "On Selected Changed (FriendsSwitcher)"!() -> (OutputDelegate, then, Button:Common Button Base Object Reference, Selected:bool) {is=title K2Node_ComponentBoundEvent}
   > SetShowFriendsOnly!(execute, self:Tournament Leaderboard VM Object Reference=Get LeaderboardVM(), InbShowFriendsOnly:bool=@n18.Selected) -> (then)
   > Fire Event Element Interacted(Element=@n18.Button)
   "On Text Commit by Enter (FilteredInputField)"() {is=title K2Node_ComponentBoundEvent}
   > n19 = GetCurrentText!(execute, self:WBP UIKit Filtered Input Field Object Reference=Get FilteredInputField()) -> (then, Text:Text)
   > n20 = Set LastSearchedText(LastSearchedText=@n19)
   > FilterCurrentViewEntriesByText(self=@n17, SearchText=@n20)
   n21 = ShowLoadingEntriesMessage!() -> (OutputDelegate, then, bShow:bool=~) {is=title K2Node_CustomEvent}
   > SetVisibility(
    self=Get LoadingEntriesMessage() -> (LoadingEntriesMessage:WBP Loading Entries Message Object Reference)
    InVisibility=Select(Option 0=Collapsed, Option 1=HitTestInvisible, Index=@n21.bShow)
   )
   Event Construct()
   n22 = Get FilteredInputField()
   > SetGridSlotNudge!(
    execute
    Widget:Widget Object Reference=Get InputActionWidget!(self:Common Button Base Object Reference=Get InputField!(self:WBP UIKit Filtered Input Field Object Reference=@n22) -> (InputField:WBP UIKit Input Field Object Reference)) -> (InputActionWidget:Common Action Widget Object Reference)
    Nudge:Vector 2D Structure=(X=40,Y=0)
   ) -> (then)
   n23 = IsTournamentLeaderboardSearchBarEnabled!(self:Fort Compete UIManager Object Reference=Get FortCompeteUIManager() -> (ReturnValue:Fort Compete UIManager Object Reference) {is=title K2Node_GetSubsystem}) -> (ReturnValue:bool)
   > Branch(Condition=@n23)
   > Assign On Clicked!(
    execute:Exec
    self:Common Button Base Object Reference=Get InputField(self=Get FilteredInputField())
    Delegate:"Delegate (by ref)"=Create Event()
   ) -> (then:Exec) {is=title K2Node_AssignDelegate}
   > Assign On Entry Generated!(execute:Exec, self:List View Base Object Reference=Get CommonListView_TeamsInfo(), Delegate:"Delegate (by ref)"=Create Event()) -> (then:Exec) {is=title K2Node_AssignDelegate}
   > n24 = Construct Object from Class(Class=/Compete/TournamentLeaderboard/Data/BP_LeaderboardPageLoadingInfo.BP_LeaderboardPageLoadingInfo_C) -> (ReturnValue:BP Leaderboard Page Loading Info Object Reference)
   > Set NextPageLoadingInfo(NextPageLoadingInfo:BP Leaderboard Page Loading Info Object Reference=@n24) -> (Output_Get:BP Leaderboard Page Loading Info Object Reference)
   > n25 = Construct Object from Class(Class=&BP_LeaderboardPageLoadingInfo_C)
   > Set NextPageInGapLoadingInfo(NextPageInGapLoadingInfo:BP Leaderboard Page Loading Info Object Reference=@n25) -> (Output_Get:BP Leaderboard Page Loading Info Object Reference)
   Event PreConstruct()
   > n26 = Sequence()
   > n27 = Is Valid(InputObject=Get ContextMenuActionContext() -> (ContextMenuActionContext:UIKit Context Menu Action Context Object Reference))
   n28 = Construct Object from Class(execute=@n27.Is Not Valid, Class=/Script/FortniteUI.UIKitContextMenuActionContext) -> (ReturnValue:UIKit Context Menu Action Context Object Reference)
   > n29 = Set ContextMenuActionContext(ContextMenuActionContext:UIKit Context Menu Action Context Object Reference=@n28) -> (Output_Get:UIKit Context Menu Action Context Object Reference)
   SetErrorPosition(execute=[@n27.Is Valid, @n29.then])
   SetVisibility(execute=@n26.then_1, self=Get FilteredInputField(), InVisibility=Select(Option 0=Collapsed, Option 1=SelfHitTestInvisible, Index=@n23))
   SetWheelScrollMultiplier!(execute=@n11.then, self:List View Base Object Reference=Get CommonListView_TeamsInfo(), NewWheelScrollMultiplier:float=1) -> (then)
   SetWheelScrollMultiplier(execute=@n8.then, self=Get CommonListView_TeamsInfo(), NewWheelScrollMultiplier=0)
  ]}
 GenerateContextMenu {nodes [
   n1 = GenerateContextMenu!() -> (then, LeaderboardCategories:Array of Fort Compete UILeaderboard Category Structures) {is=title K2Node_FunctionEntry}
   > RemoveAllInteractions(self:UIKit Context Menu Action Context Object Reference=Get ContextMenuActionContext())
   > n2 = For Loop(LastIndex=Last Index!(TargetArray:Array of Fort Compete UILeaderboard Category Structures=@n1) -> (ReturnValue=0))
   n3 = MakeLiteralInt!(Value:int=@n2) -> (ReturnValue:int=0)
   > AddInteraction!(
    execute
    self:UIKit Context Menu Action Context Object Reference=Get ContextMenuActionContext()
    InActionInfo_DisplayName:Text=Break Fort Compete UILeaderboard Category!(FortCompeteUILeaderboardCategory:"Fort Compete UILeaderboard Category Structure (by ref)"="Get (a ref)"(Array:Array of Fort Compete UILeaderboard Category Structures=@n1, Dimension 1=@n2) -> (Output:"Fort Compete UILeaderboard Category Structure (by ref)")) -> (LeaderboardName:Text, GroupingKey:String, InstanceId:String) {is=title K2Node_BreakStruct}
    InActionInfo_IconTexture:Texture 2D Soft Object Reference
    InActionInfo_IconMaterial:Material Instance Soft Object Reference
    InActionInfo_SortPriority:int=1200
    InActionInfo_bIsAvailable:bool=True
    InActionInfo_bIsEnabled:bool=True
    InActionInfo_bIsSelected:bool=False
    InActionInfo_bAddDivider:bool=False
    InActionInfo_bPositionDividerAfter:bool=False
    InActionInfo_bCloseMenuOnSelect:bool=False
    InActionInfo_InteractionNameID:Name=String To Name(InString="To String (Integer)"(InInt=@n3))
    InActionInfo_Interaction:Data Table Row Handle Structure=(DataTable=None,RowName="")
    InActionInfo_PressedDelegate:Delegate="(null).None"
    InActionInfo_ReleasedDelegate:Delegate="(null).None"
    InInteractionDelegate:"Delegate (by ref)"=Create Event()
   ) -> (then)
   SetContext!(
    execute=@n2.Completed
    self:WBP UIKit Context Menu Anchor Button Base Object Reference=Get Alternative_Button()
    InteractionContext_Data:UIKit Context Menu Action Context Object Reference=Get ContextMenuActionContext()
   ) -> (then)
  ]}
 GetLeaderboardPageToLoadBasedOnOffset {nodes [
   GetLeaderboardPageToLoadBasedOnOffset() {is=title K2Node_FunctionEntry}
   n1 = Get CommonListView_TeamsInfo()
   > Set TopListItemOffset(TopListItemOffset:int=Round(A=GetScrollOffset!(self:List View Base Object Reference=@n1) -> (ReturnValue:float)) {is=title K2Node_CallFunction}) -> (Output_Get:int=0)
   > Set BottomListItemOffset(BottomListItemOffset:int=int + int(A=Get TopListItemOffset() -> (TopListItemOffset:int=0), B=Get VisibleListElementsNum() -> (VisibleListElementsNum:int=0))) -> (Output_Get:int=0)
   > n2 = Sequence!(execute) -> (then_0, then_1, then_2:Exec)
   > Branch(Condition=IsValidInfo!(self:BP Leaderboard Page Loading Info Object Reference=Get NextPageLoadingInfo() -> (NextPageLoadingInfo:BP Leaderboard Page Loading Info Object Reference)) -> (Result:bool))
   n3 = Get ListOffset!(self:BP Leaderboard Page Loading Info Object Reference=Get NextPageLoadingInfo()) -> (ListOffset:int=0)
   n4 = Get BottomListItemOffset() -> (BottomListItemOffset:int=0)
   > Branch(Condition="integer <= integer"(A=int - int(A=@n3, B=@n4), B=Get MinElementsNumRemainingForNextPageRequest() -> (MinElementsNumRemainingForNextPageRequest:int=0)))
   > Return Node!(execute, Page:int=Get Page!(self:BP Leaderboard Page Loading Info Object Reference=Get NextPageLoadingInfo()) -> (Page:int=0))
   Branch(
    execute=@n2.then_1
    Condition=IsValidInfo(self=Get NextPageInGapLoadingInfo() -> (NextPageInGapLoadingInfo:BP Leaderboard Page Loading Info Object Reference))
   )
   > n5 = Sequence!(execute) -> (then_0, then_1)
   n6 = int - int(A=Get ListOffset(self=Get NextPageInGapLoadingInfo()) -> (ListOffset=0), B=Get BottomListItemOffset() -> (BottomListItemOffset=0))
   n7 = Get MinElementsNumRemainingForNextPageRequest() -> (MinElementsNumRemainingForNextPageRequest=0)
   n8 = int * int(A=Get MinElementsNumRemainingForNextPageRequest() -> (MinElementsNumRemainingForNextPageRequest=0), B=-1)
   > Branch(Condition=AND Boolean(A="integer <= integer"(A=@n6, B=@n7), B="integer > integer"(A=@n6, B=@n8)))
   > Return Node(Page=Get Page(self=Get NextPageInGapLoadingInfo()) -> (Page=0))
   n9 = int - int(A=Get TopListItemOffset() -> (TopListItemOffset=0), B=Get ListOffset(self=Get NextPageInGapLoadingInfo()) -> (ListOffset=0))
   n10 = Get MinElementsNumRemainingForNextPageRequest() -> (MinElementsNumRemainingForNextPageRequest=0)
   n11 = int * int(A=Get MinElementsNumRemainingForNextPageRequest() -> (MinElementsNumRemainingForNextPageRequest=0), B=-1)
   Branch(execute=@n5.then_1, Condition=AND Boolean(A="integer <= integer"(A=@n9, B=@n10), B="integer > integer"(A=@n9, B=@n11)))
   > Return Node(Page=GetNextLeaderboardPageToLoadInGapOfMissingPages!(self:Tournament Leaderboard VM Object Reference=Get LeaderboardVM(), bFromStartOfTheGap:bool=false) -> (ReturnValue:int=0))
   Return Node(execute=@n2.then_2, Page=-1)
  ]}
 Initialize {nodes [
   n1 = Initialize!() -> (then, IsSeriesPointLeaderboards:bool=~) {is=title K2Node_FunctionEntry}
   > SetVisibility(
    self=[
     Get Matches() -> (Matches:Common Text Object Reference)
     Get Wins() -> (Wins:Common Text Object Reference)
     Get Elims() -> (Elims:Common Text Object Reference)
    ]
    InVisibility=Select(Option 0=HitTestInvisible, Option 1=Collapsed, Index=@n1)
   )
   n2 = Get LeaderboardCategories!(self:Tournament Leaderboard VM Object Reference=Get LeaderboardVM()) -> (LeaderboardCategories:Array of Fort Compete UILeaderboard Category Structures)
   > Branch(Condition="integer <= integer"(A=Length(TargetArray:Array of Fort Compete UILeaderboard Category Structures=@n2), B=1))
   > SetVisibility(
    self=Get Alternative_Button()
    InVisibility=Literal enum ESlateVisibility!(Enum:ESlateVisibility Enum=Collapsed) -> (ReturnValue:ESlateVisibility Enum) {is=title K2Node_EnumLiteral}
   )
  ]}
 OnContextMenuInteracted {nodes [
   n1 = OnContextMenuInteracted!() -> (then, InteractionNameID:Name) {is=title K2Node_FunctionEntry}
   > SetCurrentLeaderboardCategoryIndex!(
    execute
    self:Tournament Leaderboard VM Object Reference=Get LeaderboardVM()
    InCurrentLeaderboardCategoryIndex:int=String To Integer(InString="To String (Name)"(InName=@n1))
   ) -> (then)
  ]}
 OnCurrentLeaderboardCategoryIndexUpdate {nodes [
   n1 = OnCurrentLeaderboardCategoryIndexUpdate!() -> (then, CurrentLeaderboardCategoryIndex:int) {is=title K2Node_FunctionEntry}
   n2 = Get LeaderboardCategories(self=Get LeaderboardVM())
   > n3 = Set IsCurrentLeaderboardCategoryValid(IsCurrentLeaderboardCategoryValid:bool=Is Valid Index!(TargetArray:Array of Fort Compete UILeaderboard Category Structures=@n2, IndexToTest:int=@n1) -> (ReturnValue:bool)) -> (Output_Get:bool)
   > Branch(Condition=@n3)
   > Set CurrentLeaderboardCategory(CurrentLeaderboardCategory:Fort Compete UILeaderboard Category Structure="Get (a copy)"(Array:Array of Fort Compete UILeaderboard Category Structures=@n2, Dimension 1=@n1) -> (Output:Fort Compete UILeaderboard Category Structure)) -> (Output_Get:Fort Compete UILeaderboard Category Structure)
   > Update Alternative Button Text()
  ]}
 OnParentActivated {nodes [
   OnParentActivated() {is=title K2Node_FunctionEntry}
   > SetShowFriendsOnly(
    self=Get LeaderboardVM()
    InbShowFriendsOnly=GetSelected!(self:Common Button Base Object Reference=Get FriendsSwitcher() -> (FriendsSwitcher:WBP UIKit Switcher Button Object Reference)) -> (ReturnValue:bool)
   )
  ]}
 OnPlayerFound {nodes [
   n1 = OnPlayerFound!() -> (then, bFound:bool=~) {is=title K2Node_FunctionEntry}
   n2 = Get ErrorText!(self:WBP UIKit Filtered Input Field Object Reference=Get FilteredInputField()) -> (ErrorText:Common Text Object Reference)
   > SetVisibility(self=@n2, InVisibility=Select(Option 0=SelfHitTestInvisible, Option 1=Collapsed, Index=@n1))
   > "SetText (Text)"(self=@n2, InText="LOCTEXT(\"Player not found.\")")
  ]}
 RestoreListScrollOffset {nodes [
   RestoreListScrollOffset() {is=title K2Node_FunctionEntry}
   n1 = Get CommonListView_TeamsInfo()
   n2 = Get LastScrollOffsetObject() -> (LastScrollOffsetObject:Object Reference)
   > SetScrollOffset!(
    execute
    self:List View Base Object Reference=Get CommonListView_TeamsInfo()
    InScrollOffset:float=float + float(
     A:int=GetIndexForItem(self=@n1, Item=@n2) -> (ReturnValue=0)
     B=Get LastScrollOffsetFractionalPart() -> (LastScrollOffsetFractionalPart:double)
    )
   ) -> (then)
  ]}
 SaveListScrollOffset {nodes [
   SaveListScrollOffset() {is=title K2Node_FunctionEntry}
   n1 = GetScrollOffset(self=Get CommonListView_TeamsInfo())
   > Set LastScrollOffsetFractionalPart(LastScrollOffsetFractionalPart:double="% (Float)"!(A:double=@n1, B:double=1) -> (ReturnValue:double)) -> (Output_Get:double)
   n2 = Get CommonListView_TeamsInfo()
   > Set LastScrollOffsetObject(LastScrollOffsetObject:Object Reference="Get (a copy)"(
     Array:Array of Object References=GetListItems!(self:List View Object Reference=@n2) -> (ReturnValue:Array of Object References)
     Dimension 1=Floor(A=@n1)
    ) -> (Output:Object Reference)) -> (Output_Get:Object Reference)
  ]}
 SearchBarCustomNavigation {nodes [
   SearchBarCustomNavigation!() -> (then, Navigation:EUINavigation Enum) {is=title K2Node_FunctionEntry}
   n1 = Get FortCompeteUIManager()
   n2 = Select(
    Option 0:Widget Object Reference=Get CommonListView_TeamsInfo()
    Option 1:Widget Object Reference=Get FilteredInputField()
    Index=AND Boolean(A=Get IsLeaderboardListEmpty() -> (IsLeaderboardListEmpty:bool), B=IsTournamentLeaderboardSearchBarEnabled(self=@n1))
   ) -> (ReturnValue:Widget Object Reference)
   > SetFocus(self=@n2)
   > Return Node!(execute, ReturnValue:Widget Object Reference=@n2)
  ]}
 Set Leaderboard View Model {nodes [
   n1 = Set Leaderboard View Model!() -> (then, " Leaderboard VM":"Tournament Leaderboard VM Object Reference (by ref)") {is=title K2Node_FunctionEntry}
   > Set LeaderboardVM(LeaderboardVM:Tournament Leaderboard VM Object Reference=@n1) -> (Output_Get:Tournament Leaderboard VM Object Reference)
  ]}
 Set Local Player Best Team {nodes [
   n1 = Set Local Player Best Team!() -> (then, New Best Team:Tournament Leaderboard Entry VM Object Reference) {is=title K2Node_FunctionEntry}
   n2 = IsValid(Object=@n1)
   n3 = Get LocalPlayerBestTeamItem() -> (LocalPlayerBestTeamItem:WBP Tournament Leaderboard List Item Object Reference)
   > SetVisibility(
    self=[Get LocalPlayerSeparator() -> (LocalPlayerSeparator:Image Object Reference), @n3]
    InVisibility=Select(Option 0:ESlateVisibility Enum=Collapsed, Option 1:ESlateVisibility Enum=SelfHitTestInvisible, Index=@n2) -> (ReturnValue:ESlateVisibility Enum)
   )
   > Branch(Condition=@n2)
   > n4 = Set Local Player Best Team Data(Local Player Best Team Data:Tournament Leaderboard Entry VM Object Reference=Get New Best Team() -> (New Best Team:Tournament Leaderboard Entry VM Object Reference)) -> (Output_Get:Tournament Leaderboard Entry VM Object Reference)
   > SetLeaderboard Entry Data!(execute, self:WBP Tournament Leaderboard List Item Object Reference=@n3, Viewmodel:Tournament Leaderboard Entry VM Object Reference=@n4) -> (then)
  ]}
 Set Teams Info {nodes [
   n1 = Set Teams Info!() -> (then, TeamsInfo:Array of Tournament Leaderboard Entry VM Object References) {is=title K2Node_FunctionEntry}
   n2 = Is Empty!(TargetArray:Array of Tournament Leaderboard Entry VM Object References=@n1) -> (ReturnValue:bool)
   > Set IsLeaderboardListEmpty(IsLeaderboardListEmpty:bool=@n2) -> (Output_Get:bool)
   n3 = Get bIsModelInitialized!(self:Tournament Leaderboard VM Object Reference=Get LeaderboardVM()) -> (bIsModelInitialized:bool)
   > SetActiveWidgetIndex(
    self:Common Visibility Switcher Object Reference=Get LeaderboardListSwitcher() -> (LeaderboardListSwitcher:Common Visibility Switcher Object Reference)
    Index="To Integer (Boolean)"!(InBool:bool=AND Boolean(A=@n2, B=@n3)) -> (ReturnValue:int=0)
   )
   > n4 = Sequence!(execute) -> (then_0, then_1, then_2:Exec, then_3:Exec, then_4:Exec, then_5:Exec)
   > n5 = For Each Loop(Array:Array of Tournament Leaderboard Entry VM Object References=Get TeamsInfo() -> (TeamsInfo:Array of Tournament Leaderboard Entry VM Object References)) -> (Array Element:Tournament Leaderboard Entry VM Object Reference)
   > Set DisplayIndex!(execute, DisplayIndex:int=@n5.Array Index, self:Tournament Leaderboard Entry VM Object Reference=@n5) -> (then, Output_Get:int=0)
   > Broadcast Field Value Changed!(execute, Object:Object Reference=@n5, FieldId:Field Notification Id Structure=(FieldName="DisplayIndex")) -> (then)
   n6 = Branch(execute=@n5.Completed, Condition=Get NextPageIsRequested())
   > n7 = SaveListScrollOffset() {is=title K2Node_CallFunction}
   n8 = Get TeamsInfo()
   Set List Items!(execute=[@n7.then, @n6.else], self:List View Object Reference=Get CommonListView_TeamsInfo(), InListItems:Array of Object References=@n8) -> (then) {is=title K2Node_CallFunction}
   > n9 = Branch(Condition=OR Boolean(A=Is Empty(TargetArray=@n8), B=Get NextPageIsRequested()))
   > RestoreListScrollOffset() {is=title K2Node_CallFunction}
   n10 = Get CommonListView_TeamsInfo()
   n11 = Get Selected Item!(self:List View Object Reference=@n10) -> (ReturnValue:Object Reference) {is=title K2Node_CallFunction}
   SetSelectedIndex(execute=@n9.else, self=@n10, Index="Max (Integer)"(A=GetIndexForItem(self=@n10, Item=@n11) -> (ReturnValue=0), B=0))
   UpdateListOffsetsToLoadLeaderboardPages(execute=@n4.then_1)
   UpdateDebugText(execute=@n4.then_4)
   Set NextPageIsRequested(execute=@n4.then_5, NextPageIsRequested=false)
   SetScrollbarVisibility!(
    execute=@n4.then_2
    self:List View Base Object Reference=Get CommonListView_TeamsInfo()
    InVisibility:ESlateVisibility Enum=SelfHitTestInvisible
   ) -> (then)
   n12 = Get TeamsInfo()
   n13 = Get bIsModelInitialized(self=Get LeaderboardVM())
   ShowLoadingEntriesMessage(execute=@n4.then_3, bShow=AND Boolean(A=Is Empty(TargetArray=@n12), B=NOT Boolean(A=@n13))) {is=title K2Node_CallFunction}
  ]}
 SetDataVisibility {nodes [
   n1 = SetDataVisibility!() -> (then, Visible:bool=~) {is=title K2Node_FunctionEntry}
   > SetRenderOpacity(
    self=Get StatsNames() -> (StatsNames:Horizontal Box Object Reference)
    InOpacity=Select(Option 0:float=0.5, Option 1:float=1, Index=@n1) -> (ReturnValue:float)
   )
  ]}
 SetErrorPosition {nodes [
   SetErrorPosition() {is=title K2Node_FunctionEntry}
   n1 = Get FilteredInputField()
   n2 = Get Info!(self:WBP UIKit Filtered Input Field Object Reference=@n1) -> (Info:Overlay Object Reference)
   > SetVisibility(self=@n2, InVisibility=SelfHitTestInvisible)
   n3 = SlotAsGridSlot(Widget=@n2)
   > SetRow!(execute, self:Grid Slot Object Reference=@n3, InRow:int=1) -> (then)
   > SetNudge!(execute, self:Grid Slot Object Reference=@n3, InNudge:Vector 2D Structure=(X=0,Y=-55)) -> (then)
   > SetHorizontalAlignment(self:Grid Slot Object Reference=@n3, InHorizontalAlignment=HAlign_Right)
   > SetVisibility(self=Get ErrorText(self=@n1), InVisibility=Collapsed)
  ]}
 SetFocusOnSearchBar {nodes [
   SetFocusOnSearchBar!() -> (then, Button:Common Button Base Object Reference) {is=title K2Node_FunctionEntry}
   > SetFocus(self=Get FilteredInputField())
  ]}
 SetLeaderboardCategories {nodes [
   n1 = SetLeaderboardCategories!() -> (then, LeaderboardCategories:Array of Fort Compete UILeaderboard Category Structures) {is=title K2Node_FunctionEntry}
   > Branch(Condition="integer > integer"(A=Length(TargetArray=@n1), B=1))
   > SetVisibility(self=Get Alternative_Button(), InVisibility=Literal enum ESlateVisibility(Enum=SelfHitTestInvisible))
   > GenerateContextMenu!(
    execute
    self
    LeaderboardCategories:Array of Fort Compete UILeaderboard Category Structures=Get LeaderboardCategories!() -> (LeaderboardCategories)
   ) -> (then) {is=title K2Node_CallFunction}
   > Update Alternative Button Text()
  ]}
 SetNavigationToListView {nodes [
   SetNavigationToListView!() -> (then, Navigation:EUINavigation Enum) {is=title K2Node_FunctionEntry}
   n1 = Get CommonListView_TeamsInfo()
   > SetFocus(self=@n1)
   > Return Node(ReturnValue=@n1)
  ]}
 Update Alternative Button Text {nodes [
   Update Alternative Button Text() {is=title K2Node_FunctionEntry}
   > UpdateText!(
    execute
    self:WBP UIKit Context Menu Anchor Button Base Object Reference=Get Alternative_Button()
    Text:Text=Select(
     Option 0:Text
     Option 1:Text=Get!() -> (CurrentLeaderboardCategory_LeaderboardName:Text, CurrentLeaderboardCategory_GroupingKey:String, CurrentLeaderboardCategory_InstanceId:String)
     Index=Get IsCurrentLeaderboardCategoryValid() -> (IsCurrentLeaderboardCategoryValid:bool)
    ) -> (ReturnValue:Text)
   ) -> (then)
  ]}
 UpdateDebugText {nodes [
   UpdateDebugText() {is=title K2Node_FunctionEntry}
   n1 = Get bShowDebugInfo!(self:Fort Compete UIManager Object Reference=Get FortCompeteUIManager()) -> (bShowDebugInfo:bool)
   > SetVisibility(
    self=Get DEBUG_TEXT() -> (DEBUG_TEXT:Common Text Object Reference)
    InVisibility=Select(Option 0:ESlateVisibility Enum=Collapsed, Option 1:ESlateVisibility Enum=SelfHitTestInvisible, Index=@n1) -> (ReturnValue:ESlateVisibility Enum)
   )
   > n2 = Branch(Condition=@n1)
   > n3 = Is Valid(InputObject=Get LeaderboardVM())
   "SetText (Text)"(
    execute=[@n4.else, @n2.else, @n3.Is Not Valid]
    self=Get DEBUG_TEXT()
    InText="LOCTEXT(\"Debug: Failed to get current leaderboard category.\")"
   )
   n4 = Branch(execute=@n3.Is Valid, Condition=Get IsCurrentLeaderboardCategoryValid())
   n5 = Get()
   n6 = IsEmpty(InString=@n5.CurrentLeaderboardCategory_GroupingKey)
   n7 = IsEmpty(InString=@n5.CurrentLeaderboardCategory_InstanceId)
   > "SetText (Text)"(
    self=Get DEBUG_TEXT()
    InText=Format Text!(
     Format="LOCTEXT(\"{groupingKey}:{instanceId}\")"
     groupingKey:String=Select(Option 0:String=@n5.CurrentLeaderboardCategory_GroupingKey, Option 1:String="<None>", Index=@n6) -> (ReturnValue:String)
     instanceId:String=Select(Option 0=@n5.CurrentLeaderboardCategory_InstanceId, Option 1="<None>", Index=@n7)
    ) -> (Result)
   )
  ]}
 UpdateListOffsetsToLoadLeaderboardPages {nodes [
   UpdateListOffsetsToLoadLeaderboardPages() {is=title K2Node_FunctionEntry}
   > n1 = Sequence!(execute) -> (then_0, then_1)
   > Reset(self:BP Leaderboard Page Loading Info Object Reference=Get NextPageLoadingInfo())
   > Set Page!(
    execute
    Page:int=GetNextLeaderboardPageToLoad!(self:Tournament Leaderboard VM Object Reference=Get LeaderboardVM()) -> (ReturnValue:int=0)
    self:BP Leaderboard Page Loading Info Object Reference=Get NextPageLoadingInfo()
   ) -> (then, Output_Get:int=0)
   n2 = Get CommonListView_TeamsInfo()
   > Set ListOffset!(
    execute
    ListOffset:int=int - int(
     A=GetNumItems!(self:List View Object Reference=@n2) -> (ReturnValue:int=0)
     B=Get MinElementsNumRemainingForNextPageRequest() -> (MinElementsNumRemainingForNextPageRequest=0)
    )
    self:BP Leaderboard Page Loading Info Object Reference=Get NextPageLoadingInfo()
   ) -> (then, Output_Get:int=0)
   Reset(execute=@n1.then_1, self=Get NextPageInGapLoadingInfo())
   > Branch(Condition=IsAGapInLoadedPagesOrder!(self:Tournament Leaderboard VM Object Reference=Get LeaderboardVM()) -> (ReturnValue:bool))
   > n3 = Set Page(
    Page=GetNextLeaderboardPageToLoadInGapOfMissingPages(self=Get LeaderboardVM(), bFromStartOfTheGap=true) -> (ReturnValue=0)
    self=Get NextPageInGapLoadingInfo()
   ) -> (Output_Get=0)
   n4 = Get FortCompeteUIManager()
   > n5 = Increment Int(Value=int * int(A=@n3, B=GetMaxLeaderboardEntriesPerPage!(self:Fort Compete UIManager Object Reference=@n4) -> (ReturnValue:int=0)))
   > Set ListOffset(ListOffset=@n5, self=Get NextPageInGapLoadingInfo()) -> (Output_Get=0)
  ]}
]
variables [
 BlockPagesLoading {type=boolean}
 ContextMenuActionContext {type=UIKitContextMenuActionContext, default=null}
 CurrentLeaderboardCategory {type=FortCompeteUILeaderboardCategory, default {groupingKey="", instanceId="", leaderboardName=""}}
 IsCurrentLeaderboardCategoryValid {type=boolean}
 IsLeaderboardListEmpty {type=boolean}
 LastScrollOffsetFractionalPart {type=number}
 LastScrollOffsetObject {type=Object, default=null}
 LastSearchedText {type=string, default=""}
 Leaderboard Entry Data {type=FortTournamentLeaderboardEntryData, default=null}
 Local Player Best Team Data {type=FortTournamentLeaderboardEntryData, default=null}
 MinElementsNumRemainingForNextPageRequest {type=integer, default=10}
 NextPageInGapLoadingInfo {type=BP_LeaderboardPageLoadingInfo_C, default=null}
 NextPageIsRequested {type=boolean}
 NextPageLoadingInfo {type=BP_LeaderboardPageLoadingInfo_C, default=null}
 VisibleListElementsNum {type=integer}
]
widgets {
 info {inheritedWidgetCount=0, parentClass=&CommonUserWidget, rootWidgetClass=&GridPanel, widgetCount=24}
 widgets [
  Root {widgetClassPath=&GridPanel, widget=Root, properties {columnFill [1], rowFill [0,0,1,0]}}
  CaptureForPostBufferUpdate {
   widgetClassPath=&WBP_CaptureForPostBufferUpdate
   bIsVariable=true
   parent=Root
   widget=CaptureForPostBufferUpdate
   slot {class=GridSlot, properties {rowSpan=3}}
  }
  Header {
   widgetClassPath=&GridPanel
   bIsVariable=true
   parent=Root
   widget=Header
   properties {columnFill [0,0,1], rowFill [1]}
   slot {class=GridSlot, properties {layer=1, padding=<ltrb 15 0 15 0>}}
  }
  FilteredInputField {
   widgetClassPath=/FortUILibrary/UIKit/InputField/WBP_UIKit_FilteredInputField
   bIsVariable=true
   parent=Header
   widget=FilteredInputField
   properties {
    bIsFocusable=true
    hideValidationInfo=true
    hintText=Player search
    isSearchField=true
    navigation=&self:WidgetTree.FilteredInputField.WidgetNavigation_0
    promptText=""
    showTextPrompt=false
    useHintText=true
    virtualKeyboardTrigger=OnAllFocusEvents
    height Override {enable=false, value=56}
    input Action Row {dataTable=&InputActionDataTable, rowName=GenericLeftTrigger}
   }
   slot {class=GridSlot, properties {column=2, padding=<ltrb 64 0 0 0>}}
  }
  FriendsSwitcher {
   widgetClassPath=/FortUILibrary/UIKit/Buttons/WBP_UIKit_SwitcherButton
   bIsVariable=true
   parent=Header
   widget=FriendsSwitcher
   properties {bToggleable=true, mobileMinHeight=80, switcherBockgroundmaterial=/Compete/MaterialInstances/MI_UI_CompeteToggleBackground, text=Friends}
   slot {class=GridSlot, properties {horizontalAlignment=HAlign_Left, verticalAlignment=VAlign_Center}}
  }
  Alternative_Button {
   widgetClassPath=/FortUILibrary/UIKit/ContextMenu/WBP_UIKit_ContextMenuAnchorButton
   bIsVariable=true
   parent=Header
   widget=Alternative_Button
   properties {
    bIsFocusable=true
    button Background Material=/FortUILibrary/UIKit/Buttons/MaterialInstances/CTAButtons/MI_UIKit_BgQuiet
    buttonIcon=null
    buttonText=Default
    closeMenuOnContextClicked=true
    menuPlacement=MenuPlacement_BelowRightAnchor
    min Width=303
    navigation=&self:WidgetTree.Alternative_Button.WidgetNavigation_0
    button Text Font {
     bForceMonospaced=false
     bMaterialIsStencil=false
     fontMaterial=/FortUILibrary/UIKit/Buttons/MaterialInstances/CTAButtons/MI_UIKit_TextColor_Loud
     fontObject=&HeadingNow
     letterSpacing=0
     monospacedWidth=1
     size=18
     skewAmount=0
     typefaceFontName=76 Bold
     outlineSettings {
      bApplyOutlineToDropShadows=false
      bMiteredCorners=false
      bSeparateFillAlpha=false
      outlineMaterial=null
      outlineSize=0
      outlineColor=<rgba 0 0 0 1>
     }
    }
   }
   slot {class=GridSlot, properties {column=1, padding=<ltrb 24 0 0 0>}}
  }
  StatsNames {
   widgetClassPath=&HorizontalBox
   bIsVariable=true
   parent=Root
   widget=StatsNames
   properties {bOverride_Cursor=true, flowDirectionPreference=Culture}
   slot {class=GridSlot, properties {layer=1, row=1, padding=<ltrb 15 16 15 0>}}
  }
  Rank {
   widgetClassPath=&CommonTextBlock
   parent=StatsNames
   widget=Rank
   properties {
    minDesiredWidth=103
    mobileFontSizeMultiplier=1.14
    text=Rank
    colorAndOpacity {colorUseRule=UseColor_Specified, specifiedColor=<rgba 0.5776 0.5776 0.5776 0.6>}
    font {
     bForceMonospaced=false
     bMaterialIsStencil=false
     fontMaterial=null
     fontObject=&HeadingNow
     letterSpacing=0
     monospacedWidth=1
     size=15.75
     skewAmount=0
     typefaceFontName=75 Medium
     outlineSettings {
      bApplyOutlineToDropShadows=false
      bMiteredCorners=false
      bSeparateFillAlpha=false
      outlineMaterial=null
      outlineSize=0
      outlineColor=<rgba 0 0 0 1>
     }
    }
    shadowOffset=<0 0>
   }
   slot {class=HorizontalBoxSlot, properties {padding=<ltrb 45 0 0 0>}}
  }
  Player {
   widgetClassPath=&CommonTextBlock
   parent=StatsNames
   widget=Player
   properties {
    minDesiredWidth=103
    mobileFontSizeMultiplier=1.14
    text=Player
    colorAndOpacity {colorUseRule=UseColor_Specified, specifiedColor=<rgba 0.5776 0.5776 0.5776 0.6>}
    font {
     bForceMonospaced=false
     bMaterialIsStencil=false
     fontMaterial=null
     fontObject=&HeadingNow
     letterSpacing=0
     monospacedWidth=1
     size=15.75
     skewAmount=0
     typefaceFontName=75 Medium
     outlineSettings {
      bApplyOutlineToDropShadows=false
      bMiteredCorners=false
      bSeparateFillAlpha=false
      outlineMaterial=null
      outlineSize=0
      outlineColor=<rgba 0 0 0 1>
     }
    }
    shadowOffset=<0 0>
   }
   slot {class=HorizontalBoxSlot, properties {size {sizeRule=Fill, value=1}}}
  }
  Points {
   widgetClassPath=&CommonTextBlock
   parent=StatsNames
   widget=Points
   properties {
    justification=Center
    minDesiredWidth=115
    mobileFontSizeMultiplier=1.14
    text=Points
    colorAndOpacity {colorUseRule=UseColor_Specified, specifiedColor=<rgba 0.5776 0.5776 0.5776 1>}
    font {
     bForceMonospaced=false
     bMaterialIsStencil=false
     fontMaterial=null
     fontObject=&HeadingNow
     letterSpacing=0
     monospacedWidth=1
     size=15.75
     skewAmount=0
     typefaceFontName=75 Medium
     outlineSettings {
      bApplyOutlineToDropShadows=false
      bMiteredCorners=false
      bSeparateFillAlpha=false
      outlineMaterial=null
      outlineSize=0
      outlineColor=<rgba 0 0 0 1>
     }
    }
    shadowOffset=<0 0>
   }
   slot {class=HorizontalBoxSlot, properties {horizontalAlignment=HAlign_Center}}
  }
  Matches {
   widgetClassPath=&CommonTextBlock
   bIsVariable=true
   parent=StatsNames
   widget=Matches
   properties {
    justification=Center
    minDesiredWidth=110
    mobileFontSizeMultiplier=1.14
    text=Matches
    colorAndOpacity {colorUseRule=UseColor_Specified, specifiedColor=<rgba 0.5776 0.5776 0.5776 0.6>}
    font {
     bForceMonospaced=false
     bMaterialIsStencil=false
     fontMaterial=null
     fontObject=&HeadingNow
     letterSpacing=0
     monospacedWidth=1
     size=15.75
     skewAmount=0
     typefaceFontName=75 Medium
     outlineSettings {
      bApplyOutlineToDropShadows=false
      bMiteredCorners=false
      bSeparateFillAlpha=false
      outlineMaterial=null
      outlineSize=0
      outlineColor=<rgba 0 0 0 1>
     }
    }
    shadowOffset=<0 0>
   }
   slot {class=HorizontalBoxSlot}
  }
  Wins {
   widgetClassPath=&CommonTextBlock
   bIsVariable=true
   parent=StatsNames
   widget=Wins
   properties {
    justification=Center
    minDesiredWidth=120
    mobileFontSizeMultiplier=1.14
    text=Wins
    colorAndOpacity {colorUseRule=UseColor_Specified, specifiedColor=<rgba 0.5776 0.5776 0.5776 0.6>}
    font {
     bForceMonospaced=false
     bMaterialIsStencil=false
     fontMaterial=null
     fontObject=&HeadingNow
     letterSpacing=0
     monospacedWidth=1
     size=15.75
     skewAmount=0
     typefaceFontName=75 Medium
     outlineSettings {
      bApplyOutlineToDropShadows=false
      bMiteredCorners=false
      bSeparateFillAlpha=false
      outlineMaterial=null
      outlineSize=0
      outlineColor=<rgba 0 0 0 1>
     }
    }
    shadowOffset=<0 0>
   }
   slot {class=HorizontalBoxSlot}
  }
  Elims {
   widgetClassPath=&CommonTextBlock
   bIsVariable=true
   parent=StatsNames
   widget=Elims
   properties {
    justification=Center
    minDesiredWidth=110
    mobileFontSizeMultiplier=1.14
    text=Elims
    colorAndOpacity {colorUseRule=UseColor_Specified, specifiedColor=<rgba 0.5776 0.5776 0.5776 0.6>}
    font {
     bForceMonospaced=false
     bMaterialIsStencil=false
     fontMaterial=null
     fontObject=&HeadingNow
     letterSpacing=0
     monospacedWidth=1
     size=15.75
     skewAmount=0
     typefaceFontName=75 Medium
     outlineSettings {
      bApplyOutlineToDropShadows=false
      bMiteredCorners=false
      bSeparateFillAlpha=false
      outlineMaterial=null
      outlineSize=0
      outlineColor=<rgba 0 0 0 1>
     }
    }
    shadowOffset=<0 0>
   }
   slot {class=HorizontalBoxSlot}
  }
  LeaderboardListSwitcher {
   widgetClassPath=&CommonVisibilitySwitcher
   bIsVariable=true
   parent=Root
   widget=LeaderboardListSwitcher
   slot {class=GridSlot, properties {layer=1, row=2, padding=<ltrb 0 10 0 0>}}
  }
  LeaderboardList {
   widgetClassPath=&GridPanel
   parent=LeaderboardListSwitcher
   widget=LeaderboardList
   properties {columnFill [1], rowFill [1]}
   slot {class=CommonVisibilitySwitcherSlot}
  }
  SizeBox_TeamsList {
   widgetClassPath=&SizeBox
   bIsVariable=true
   parent=LeaderboardList
   widget=SizeBox_TeamsList
   properties {bOverride_WidthOverride=true, clipping=ClipToBounds, heightOverride=480, widthOverride=1103}
   slot {class=GridSlot, properties {layer=1, padding=<ltrb 0 0 -16 0>}}
  }
  CommonListView_TeamsInfo {
   widgetClassPath=&CommonListView
   bIsVariable=true
   parent=SizeBox_TeamsList
   widget=CommonListView_TeamsInfo
   properties {
    bInEnableTouchAnimatedScrolling=true
    bReturnFocusToSelection=true
    clipping=Inherit
    entryWidgetClass=&/WBP_TournamentLeaderboardListItem.WBP_TournamentLeaderboardListItem_C
    navigation=&self:WidgetTree.CommonListView_TeamsInfo.WidgetNavigation_0
    numDesignerPreviewEntries=20
    verticalEntrySpacing=10
    scrollBarPadding=<ltrb 16 0 0 0>
    scrollBarStyle {
     thickness=12
     draggedThumbImage {
      bIsDynamicallyLoaded=false
      drawAs=Box
      imageType=FullColor
      mirroring=NoMirror
      resourceName=../../../Engine/Content/Slate/Common/Scrollbar_Thumb.png
      resourceObject=null
      tiling=NoTile
      imageSize=<0 0>
      margin=<ltrb 0.25 0.25 0.25 0.25>
      outlineSettings {
       bUseBrushTransparency=false
       roundingType=HalfHeightRadius
       width=0
       color {colorUseRule=UseColor_Specified, specifiedColor=<rgba 0 0 0 0>}
       cornerRadii=<0 0 0 0>
      }
      tintColor {colorUseRule=UseColor_Specified, specifiedColor=<rgba 1 1 1 1>}
      uVRegion {bIsValid=false, max=<0 0>, min=<0 0>}
     }
     horizontalBackgroundImage {
      bIsDynamicallyLoaded=false
      drawAs=NoDrawType
      imageType=NoImage
      mirroring=NoMirror
      resourceName=None
      resourceObject=null
      tiling=NoTile
      imageSize=<0 0>
      margin=<ltrb 0 0 0 0>
      outlineSettings {
       bUseBrushTransparency=false
       roundingType=HalfHeightRadius
       width=0
       color {colorUseRule=UseColor_Specified, specifiedColor=<rgba 0 0 0 0>}
       cornerRadii=<0 0 0 0>
      }
      tintColor {colorUseRule=UseColor_Specified, specifiedColor=<rgba 1 1 1 1>}
      uVRegion {bIsValid=false, max=<0 0>, min=<0 0>}
     }
     horizontalBottomSlotImage {
      bIsDynamicallyLoaded=false
      drawAs=Image
      imageType=FullColor
      mirroring=NoMirror
      resourceName=../../../Engine/Content/Slate/Common/Scrollbar_Background_Horizontal.png
      resourceObject=null
      tiling=NoTile
      imageSize=<8 8>
      margin=<ltrb 0 0 0 0>
      outlineSettings {
       bUseBrushTransparency=false
       roundingType=HalfHeightRadius
       width=0
       color {colorUseRule=UseColor_Specified, specifiedColor=<rgba 0 0 0 0>}
       cornerRadii=<0 0 0 0>
      }
      tintColor {colorUseRule=UseColor_Specified, specifiedColor=<rgba 1 1 1 1>}
      uVRegion {bIsValid=false, max=<0 0>, min=<0 0>}
     }
     horizontalTopSlotImage {
      bIsDynamicallyLoaded=false
      drawAs=Image
      imageType=FullColor
      mirroring=NoMirror
      resourceName=../../../Engine/Content/Slate/Common/Scrollbar_Background_Horizontal.png
      resourceObject=null
      tiling=NoTile
      imageSize=<8 8>
      margin=<ltrb 0 0 0 0>
      outlineSettings {
       bUseBrushTransparency=false
       roundingType=HalfHeightRadius
       width=0
       color {colorUseRule=UseColor_Specified, specifiedColor=<rgba 0 0 0 0>}
       cornerRadii=<0 0 0 0>
      }
      tintColor {colorUseRule=UseColor_Specified, specifiedColor=<rgba 1 1 1 1>}
      uVRegion {bIsValid=false, max=<0 0>, min=<0 0>}
     }
     hoveredThumbImage {
      bIsDynamicallyLoaded=false
      drawAs=Box
      imageType=FullColor
      mirroring=NoMirror
      resourceName=../../../Engine/Content/Slate/Common/Scrollbar_Thumb.png
      resourceObject=null
      tiling=NoTile
      imageSize=<0 0>
      margin=<ltrb 0.25 0.25 0.25 0.25>
      outlineSettings {
       bUseBrushTransparency=false
       roundingType=HalfHeightRadius
       width=0
       color {colorUseRule=UseColor_Specified, specifiedColor=<rgba 0 0 0 0>}
       cornerRadii=<0 0 0 0>
      }
      tintColor {colorUseRule=UseColor_Specified, specifiedColor=<rgba 1 1 1 1>}
      uVRegion {bIsValid=false, max=<0 0>, min=<0 0>}
     }
     normalThumbImage {
      bIsDynamicallyLoaded=false
      drawAs=Box
      imageType=FullColor
      mirroring=NoMirror
      resourceName=../../../Engine/Content/Slate/Common/Scrollbar_Thumb.png
      resourceObject=null
      tiling=NoTile
      imageSize=<0 0>
      margin=<ltrb 0.25 0.25 0.25 0.25>
      outlineSettings {
       bUseBrushTransparency=false
       roundingType=HalfHeightRadius
       width=0
       color {colorUseRule=UseColor_Specified, specifiedColor=<rgba 0 0 0 0>}
       cornerRadii=<0 0 0 0>
      }
      tintColor {colorUseRule=UseColor_Specified, specifiedColor=<rgba 1 1 1 1>}
      uVRegion {bIsValid=false, max=<0 0>, min=<0 0>}
     }
     verticalBackgroundImage {
      bIsDynamicallyLoaded=false
      drawAs=NoDrawType
      imageType=NoImage
      mirroring=NoMirror
      resourceName=None
      resourceObject=null
      tiling=NoTile
      imageSize=<0 0>
      margin=<ltrb 0 0 0 0>
      outlineSettings {
       bUseBrushTransparency=false
       roundingType=HalfHeightRadius
       width=0
       color {colorUseRule=UseColor_Specified, specifiedColor=<rgba 0 0 0 0>}
       cornerRadii=<0 0 0 0>
      }
      tintColor {colorUseRule=UseColor_Specified, specifiedColor=<rgba 1 1 1 1>}
      uVRegion {bIsValid=false, max=<0 0>, min=<0 0>}
     }
     verticalBottomSlotImage {
      bIsDynamicallyLoaded=false
      drawAs=Image
      imageType=FullColor
      mirroring=NoMirror
      resourceName=../../../Engine/Content/Slate/Common/Scrollbar_Background_Vertical.png
      resourceObject=null
      tiling=NoTile
      imageSize=<8 8>
      margin=<ltrb 0 0 0 0>
      outlineSettings {
       bUseBrushTransparency=false
       roundingType=HalfHeightRadius
       width=0
       color {colorUseRule=UseColor_Specified, specifiedColor=<rgba 0 0 0 0>}
       cornerRadii=<0 0 0 0>
      }
      tintColor {colorUseRule=UseColor_Specified, specifiedColor=<rgba 1 1 1 1>}
      uVRegion {bIsValid=false, max=<0 0>, min=<0 0>}
     }
     verticalTopSlotImage {
      bIsDynamicallyLoaded=false
      drawAs=Image
      imageType=FullColor
      mirroring=NoMirror
      resourceName=../../../Engine/Content/Slate/Common/Scrollbar_Background_Vertical.png
      resourceObject=null
      tiling=NoTile
      imageSize=<6 6>
      margin=<ltrb 0 0 0 0>
      outlineSettings {
       bUseBrushTransparency=false
       roundingType=HalfHeightRadius
       width=0
       color {colorUseRule=UseColor_Specified, specifiedColor=<rgba 0 0 0 0>}
       cornerRadii=<0 0 0 0>
      }
      tintColor {colorUseRule=UseColor_Specified, specifiedColor=<rgba 1 1 1 1>}
      uVRegion {bIsValid=false, max=<0 0>, min=<0 0>}
     }
    }
    widgetStyle {backgroundBrush {
      bIsDynamicallyLoaded=false
      drawAs=NoDrawType
      imageType=FullColor
      mirroring=NoMirror
      resourceName=None
      resourceObject=null
      tiling=NoTile
      imageSize=<0 0>
      margin=<ltrb 0 0 0 0>
      outlineSettings {
       bUseBrushTransparency=false
       roundingType=HalfHeightRadius
       width=0
       color {colorUseRule=UseColor_Specified, specifiedColor=<rgba 0 0 0 0>}
       cornerRadii=<0 0 0 0>
      }
      tintColor {colorUseRule=UseColor_Specified, specifiedColor=<rgba 1 1 1 1>}
      uVRegion {bIsValid=false, max=<0 0>, min=<0 0>}
     }}
   }
   slot {class=SizeBoxSlot, properties {padding=<ltrb 15 15 0 15>}}
  }
  LocalPlayerBestTeamItem {
   widgetClassPath=&/WBP_TournamentLeaderboardListItem
   bIsVariable=true
   parent=LeaderboardList
   widget=LocalPlayerBestTeamItem
   properties {
    bIsFocusable=false
    bSelectable=false
    bShouldSelectUponReceivingFocus=true
    bSimulateHoverOnTouchInput=true
    is Local Player Best Team Item=true
    triggeringInputAction {dataTable=&InputActionDataTable, rowName=GenericRightStickClick}
   }
   slot {class=GridSlot, properties {row=3, padding=<ltrb 15 16 15 0>}}
  }
  LocalPlayerSeparator {
   widgetClassPath=&Image
   bIsVariable=true
   parent=LeaderboardList
   widget=LocalPlayerSeparator
   properties {
    brush {
     bIsDynamicallyLoaded=false
     drawAs=RoundedBox
     imageType=NoImage
     mirroring=NoMirror
     resourceName=None
     resourceObject=null
     tiling=NoTile
     imageSize=<32 3>
     margin=<ltrb 0 0 0 0>
     outlineSettings {
      bUseBrushTransparency=false
      roundingType=HalfHeightRadius
      width=0
      color {colorUseRule=UseColor_Specified, specifiedColor=<rgba 0 0 0 0>}
      cornerRadii=<0 0 0 0>
     }
     tintColor {colorUseRule=UseColor_Specified, specifiedColor=<rgba 1 1 1 1>}
     uVRegion {bIsValid=false, max=<0 0>, min=<0 0>}
    }
    colorAndOpacity=<rgba 0.521 0.521 0.521 0.502>
   }
   slot {class=GridSlot, properties {row=2, padding=<ltrb 60 16 60 0>}}
  }
  LoadingEntriesMessage {
   widgetClassPath=&/WBP_LoadingEntriesMessage
   bIsVariable=true
   parent=LeaderboardList
   widget=LoadingEntriesMessage
   properties {loadingMessageText=Loading leaderboard entries}
   slot {class=GridSlot, properties {horizontalAlignment=HAlign_Center, layer=5, verticalAlignment=VAlign_Bottom, padding=<ltrb 0 0 0 25>}}
  }
  EmptyList {
   widgetClassPath=&GridPanel
   parent=LeaderboardListSwitcher
   widget=EmptyList
   properties {columnFill [1], rowFill [1]}
   slot {class=CommonVisibilitySwitcherSlot}
  }
  Image_EmptylIST {
   widgetClassPath=&Image
   bIsVariable=true
   parent=EmptyList
   widget=Image_EmptylIST
   properties {
    renderOpacity=0.5
    brush {
     bIsDynamicallyLoaded=false
     drawAs=Image
     imageType=FullColor
     mirroring=NoMirror
     resourceName=None
     resourceObject=/Compete/MaterialInstances/MI_UI_CompeteLeaderboardBackground
     tiling=NoTile
     imageSize=<1143.03 584>
     margin=<ltrb 0 0 0 0>
     outlineSettings {
      bUseBrushTransparency=false
      roundingType=FixedRadius
      width=0
      color {colorUseRule=UseColor_Specified, specifiedColor=<rgba 0 0 0 0>}
      cornerRadii=<0.5 0 0 0>
     }
     tintColor {colorUseRule=UseColor_Specified, specifiedColor=<rgba 1 1 1 1>}
     uVRegion {bIsValid=false, max=<0 0>, min=<0 0>}
    }
   }
   slot {class=GridSlot}
  }
  EmptyListText {
   widgetClassPath=&CommonTextBlock
   parent=EmptyList
   widget=EmptyListText
   properties {
    justification=Center
    text="No Leaderboard Entries Available\r\nCheck Back Later"
    font {
     bForceMonospaced=false
     bMaterialIsStencil=false
     fontMaterial=null
     fontObject=&HeadingNow
     letterSpacing=0
     monospacedWidth=1
     size=18
     skewAmount=0
     typefaceFontName=75 Medium
     outlineSettings {
      bApplyOutlineToDropShadows=false
      bMiteredCorners=false
      bSeparateFillAlpha=false
      outlineMaterial=null
      outlineSize=0
      outlineColor=<rgba 0 0 0 1>
     }
    }
    shadowOffset=<0 0>
   }
   slot {class=GridSlot, properties {horizontalAlignment=HAlign_Center, verticalAlignment=VAlign_Center}}
  }
  DEBUG_TEXT {
   widgetClassPath=&CommonTextBlock
   bIsVariable=true
   parent=Root
   widget=DEBUG_TEXT
   properties {
    justification=Center
    text=DEBUG_TEXTBLOCK
    visibility=Collapsed
    font {
     bForceMonospaced=false
     bMaterialIsStencil=false
     fontMaterial=null
     fontObject=&Roboto
     letterSpacing=0
     monospacedWidth=1
     size=21
     skewAmount=0
     typefaceFontName=Bold
     outlineSettings {
      bApplyOutlineToDropShadows=false
      bMiteredCorners=false
      bSeparateFillAlpha=false
      outlineMaterial=null
      outlineSize=4
      outlineColor=<rgba 0 0 0 1>
     }
    }
    shadowOffset=<0 0>
   }
   slot {class=GridSlot, properties {row=3, verticalAlignment=VAlign_Center}}
  }
 ]
}
