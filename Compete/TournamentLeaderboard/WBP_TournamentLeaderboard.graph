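# Activatable tournament leaderboard screen showing a header, teams list and stats for session or series-point leaderboards.
# A SetTournamentLeaderboardType router event supplies the type and data; activation creates the matching leaderboard view model and initializes child widgets.
# Bottom-bar decoy buttons open watch-live, profile, match-history and rules screens, with visibility following data, selection and leaderboard type.
# Percentile payouts enable scoring rules, estimating total entries from the second listed entry's percentile.
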WidgetBlueprint /Compete/TournamentLeaderboard/WBP_TournamentLeaderboard
graphs [
 BP_GetDesiredFocusTarget {nodes [
   Get Desired Focus Target()
   n1 = Get CommonInputSubsystem()
   n2 = Get TeamsList() -> (TeamsList:WBP Tournament Leaderboard Teams List Object Reference)
   n3 = Get CommonListView_TeamsInfo!(self:WBP Tournament Leaderboard Teams List Object Reference=@n2) -> (CommonListView_TeamsInfo:Common List View Object Reference)
   n4 = "integer > integer"(A=GetNumItems!(self:List View Object Reference=@n3) -> (ReturnValue:int=0))
   n5 = Get Alternative_Button!(self:WBP Tournament Leaderboard Teams List Object Reference=@n2) -> (Alternative_Button:WBP UIKit Context Menu Anchor Button Object Reference)
   > Return Node!(
    execute
    ReturnValue:Widget Object Reference=Select!(
     MouseAndKeyboard:Widget Object Reference
     Gamepad:Widget Object Reference=Select!(Option 0:Widget Object Reference=@n5, Option 1:Widget Object Reference=@n3, Index:bool=@n4) -> (ReturnValue)
     Touch:Widget Object Reference
     Count:Widget Object Reference
     Index:ECommonInputType Enum=GetCurrentInputType(self=@n1)
    ) -> (ReturnValue:Widget Object Reference)
   )
  ]}
 EventGraph {nodes [
   n1 = "On Clicked (DecoyButton_WatchLive)"!() -> (OutputDelegate, then, Button:Common Button Base Object Reference) {is=title K2Node_ComponentBoundEvent}
   > Fire Event Element Interacted(Element=@n1.Button)
   > n2 = Create Widget!(
    execute:Exec
    Class:User Widget Class Reference=/Compete/TournamentLeaderboard/ModalMessages/WatchLive/WBP_WatchLiveModal.WBP_WatchLiveModal_C
    OwningPlayer:Player Controller Object Reference=GetOwningPlayer()
    FortTournamentWatchLiveVM:Watch Live VM Object Reference
   ) -> (then:Exec, ReturnValue:WBP Watch Live Modal Object Reference)
   > TryPushWidgetAsModalPanel!(execute, PushWidget:User Widget Object Reference=@n2) -> (then)
   n3 = Get TeamsList()
   > Branch(Condition=Get IsCurrentLeaderboardCategoryValid!(self:WBP Tournament Leaderboard Teams List Object Reference=@n3) -> (IsCurrentLeaderboardCategoryValid:bool))
   > Initialize!(
    execute
    self:WBP Watch Live Modal Object Reference=@n2
    WatchLiveVM:Watch Live VM Object Reference=Get WatchLiveVM() -> (WatchLiveVM:Watch Live VM Object Reference)
    LeaderboardCategory:"Fort Compete UILeaderboard Category Structure (by ref)"=Get CurrentLeaderboardCategory!(self:WBP Tournament Leaderboard Teams List Object Reference=@n3) -> (CurrentLeaderboardCategory:Fort Compete UILeaderboard Category Structure)
    bShowFriendsOnly:bool=Get bShowFriendsOnly!(self:Tournament Leaderboard VM Object Reference=Get LeaderboardVM() -> (LeaderboardVM:Tournament Leaderboard VM Object Reference)) -> (bShowFriendsOnly:bool)
    AlternativeButtonVisibility:ESlateVisibility Enum=Get Visibility!(self:Widget Object Reference=Get Alternative_Button(self=@n3)) -> (Visibility:ESlateVisibility Enum=Visible)
   ) -> (then)
   > ActivateWidget(self=@n2)
   n4 = "On Clicked (DecoyButton_Profile)"!() -> (OutputDelegate, then, Button:Common Button Base Object Reference) {is=title K2Node_ComponentBoundEvent}
   n5 = Get Selected Item!(self:List View Object Reference=Get CommonListView_TeamsInfo(self=Get TeamsList())) -> (ReturnValue:Object Reference) {is=title K2Node_CallFunction}
   n6 = Get CommonListView_TeamsInfo(self=Get TeamsList())
   n7 = Branch(
    execute=[@n4.then, @n15.then]
    Condition=Is Item Visible!(self:List View Object Reference=@n6, Item:Object Reference=@n5) -> (ReturnValue:bool)
   )
   > n8 = Cast To FortTournamentLeaderboardEntryData!(execute, Object=@n5) -> (then, CastFailed, AsTournament Leaderboard Entry VM:Tournament Leaderboard Entry VM Object Reference) {is=title K2Node_DynamicCast}
   > n9 = Branch(Condition=Get SingleProfileSelected() -> (SingleProfileSelected:bool))
   > OpenSinglePlayerProfile!(execute, self, TeamsInfo:Fort Tournament Base Entry VM Object Reference=@n8) -> (then)
   n10 = Create Widget!(
    execute=@n9.else
    Class=/Compete/TournamentLeaderboard/ModalMessages/Profiles/WBP_ProfilesModal.WBP_ProfilesModal_C
    OwningPlayer=GetOwningPlayer()
    FortTournamentLeaderboardEntryData:Tournament Leaderboard Entry VM Object Reference
   ) -> (then, ReturnValue:WBP Profiles Modal Object Reference)
   > TryPushWidgetAsModalPanel(PushWidget=@n10)
   > ActivateWidget(self=@n10)
   > Set FortTournamentLeaderboardEntryData!(
    execute
    FortTournamentLeaderboardEntryData:Tournament Leaderboard Entry VM Object Reference=@n8
    self:WBP Profiles Modal Object Reference=@n10
   ) -> (then, Output_Get:Tournament Leaderboard Entry VM Object Reference)
   Scroll Item Into View!(execute=@n7.else, self:List View Object Reference=@n6, Item:Object Reference=@n5) -> (then)
   n11 = "On Clicked (DecoyButton_QualificationRules)"!() -> (OutputDelegate, then, Button:Common Button Base Object Reference) {is=title K2Node_ComponentBoundEvent}
   > Fire Event Element Interacted(Element=@n11.Button)
   > n12 = QueueModalWidgetSoftClassAsync!(
    execute:Exec
    SoftWidgetClass:Common Activatable Widget Soft Class Reference=/Compete/TournamentLeaderboard/ModalMessages/QualificationRules/WBP_QualificationRulesModal.WBP_QualificationRulesModal_C
   ) -> (then:Exec, OnModalWidgetQueued:Exec, ActivatableWidget:Common Activatable Widget Object Reference) {is=title K2Node_AsyncAction}
   n13 = Cast To WBP_QualificationRulesModal!(execute=@n12.OnModalWidgetQueued, Object=@n12) -> (then, CastFailed, AsWBP Qualification Rules Modal:WBP Qualification Rules Modal Object Reference) {is=title K2Node_DynamicCast}
   n14 = Get LeaderboardVM()
   > SetFortTournamentLeaderboardVM!(execute, self:WBP Qualification Rules Modal Object Reference=@n13, Viewmodel:Tournament Leaderboard VM Object Reference=@n14) -> (then)
   > UpdateCanShowScoringRules()
   > SetCurrentLeaderboardCategory!(
    execute
    self:WBP Qualification Rules Modal Object Reference=@n13
    InLeaderboardCategory:Fort Compete UILeaderboard Category Structure=Get CurrentLeaderboardCategory(self=Get TeamsList())
   ) -> (then)
   > SetPercentileThresholds!(
    execute
    self:WBP Qualification Rules Modal Object Reference=@n13
    InThresholds:Array of Tournament Payout Threshold Data Structures=Get PercentilePayoutThresholds() -> (PercentilePayoutThresholds:Array of Tournament Payout Threshold Data Structures)
    TotalTournamentEntries:int=Get TotalTournamentEntries() -> (TotalTournamentEntries:int=0)
    CanShowPercentiles:bool=Get bCanShowScoringRules() -> (bCanShowScoringRules:bool)
   ) -> (then)
   n15 = "On Clicked (DecoyButton_Profiles)"!() -> (OutputDelegate, then, Button:Common Button Base Object Reference) {is=title K2Node_ComponentBoundEvent}
   n16 = "On Clicked (DecoyButton_MatchHistory)"!() -> (OutputDelegate, then, Button:Common Button Base Object Reference) {is=title K2Node_ComponentBoundEvent}
   > Fire Event Element Interacted(Element=@n16.Button)
   > n17 = GetTournamentLeaderboardScreenSelectedEntry!(
    execute
    self:Fort Compete UIManager Object Reference=Get FortCompeteUIManager() -> (ReturnValue:Fort Compete UIManager Object Reference) {is=title K2Node_GetSubsystem}
   ) -> (then, ReturnValue:Tournament Leaderboard Entry VM Object Reference)
   > Is Valid(InputObject=@n17)
   > n18 = Create Widget!(
    execute
    Class=/Compete/TournamentLeaderboard/ModalMessages/MatchHistory/WBP_MatchHistory.WBP_MatchHistory_C
    OwningPlayer=GetOwningPlayer()
    LeaderboardEntryVM:Tournament Leaderboard Entry VM Object Reference
    UIKitDialogViewModel:UI Kit Dialog Viewmodel Object Reference
   ) -> (then, ReturnValue:WBP Match History Object Reference)
   > SetUIKitDialogViewModel!(
    execute
    self:WBP Match History Object Reference=@n18
    Viewmodel:UI Kit Dialog Viewmodel Object Reference=Get MatchHistoryDialogVM!(self:Tournament Leaderboard VM Object Reference=Get LeaderboardVM()) -> (MatchHistoryDialogVM:UI Kit Dialog Viewmodel Object Reference)
   ) -> (then)
   > TryPushWidgetAsModalPanel(PushWidget=@n18)
   > ActivateWidget(self=@n18)
   n19 = "On Clicked (DecoyButton_Back)"!() -> (OutputDelegate, then, Button:Common Button Base Object Reference) {is=title K2Node_ComponentBoundEvent}
   > Fire Event Element Interacted(Element=@n19.Button)
   > DeactivateWidget()
   n20 = Create Widget!(
    execute
    Class=&WBP_QualificationRulesModal_C
    OwningPlayer=GetOwningPlayer()
    FortTournamentLeaderboardVM:Tournament Leaderboard VM Object Reference=@n14
   ) -> (then, ReturnValue:WBP Qualification Rules Modal Object Reference)
   > TryPushWidgetAsModalPanel(PushWidget=@n20)
   > ActivateWidget(self=@n20)
   > UpdateCanShowScoringRules()
   > SetPercentileThresholds(
    self=@n20
    InThresholds=Get PercentilePayoutThresholds()
    TotalTournamentEntries=Get TotalTournamentEntries() -> (TotalTournamentEntries=0)
    CanShowPercentiles=Get bCanShowScoringRules()
   )
   BindEvents() {is=title K2Node_CustomEvent}
   > Assign On Input Method Changed!(execute:Exec, self:Common Input Object Reference=Get CommonInputSubsystem(), Delegate:"Delegate (by ref)"=Create Event()) -> (then:Exec) {is=title K2Node_AssignDelegate}
   n21 = Get Alternative_Button(self=Get TeamsList())
   > Assign On Hovered!(execute:Exec, self:WBP UIKit Context Menu Anchor Button Base Object Reference=@n21, Delegate:"Delegate (by ref)"=Create Event()) -> (then:Exec) {is=title K2Node_AssignDelegate}
   > Assign On Unhovered!(execute:Exec, self:WBP UIKit Context Menu Anchor Button Base Object Reference=@n21, Delegate:"Delegate (by ref)"=Create Event()) -> (then:Exec) {is=title K2Node_AssignDelegate}
   > Assign On Item Selection Changed!(
    execute:Exec
    self:List View Object Reference=Get CommonListView_TeamsInfo(self=Get TeamsList())
    Delegate:"Delegate (by ref)"=Create Event()
   ) -> (then:Exec) {is=title K2Node_AssignDelegate}
   OnAlternativeButtonHovered() {is=title K2Node_CustomEvent}
   n22 = Get CommonInputSubsystem()
   > Branch(Condition="Equal (Enum)"(A=GetCurrentInputType(self=@n22), B=Gamepad))
   > n23 = Get!(execute:Exec) -> (then:Exec, else:Exec, CurrentProfileButton:WBP Bottom Bar Decoy Button Object Reference)
   > SetBottomButtonVisibility!(execute, self:WBP Bottom Bar Decoy Button Object Reference=@n23, ShouldButtonBeVisible:bool=false) -> (then)
   OnAlternativeButtonUnHovered() {is=title K2Node_CustomEvent}
   > UpdateProfileButtonVisibility()
   Event PreConstruct()
   > SetMobileStyle() {is=title K2Node_CallFunction}
   Event On Deactivated()
   > DeinitializeVM(self:Tournament Leaderboard VM Object Reference=Get LeaderboardVM())
   n24 = Get Set Leaderboard Type Async Task Ref() -> (Set Leaderboard Type Async Task Ref:Async Action Start Listening To Event Object Reference)
   > Is Valid(InputObject=@n24)
   > StopListeningToEvent(self:Async Action Start Listening To Event Object Reference=@n24)
   Event Construct()
   > BindEvents() {is=title K2Node_CallFunction}
   > n25 = Cast To FortPlayerController(Object=GetOwningPlayer())
   > Set FortPlayerController(FortPlayerController:Fort Player Controller Object Reference=@n25) -> (Output_Get:Fort Player Controller Object Reference)
   > OnInputMethodChanged!(execute, self, bNewInputType:ECommonInputType Enum=GetCurrentInputType(self=Get CommonInputSubsystem())) -> (then)
   > SetBottomButtonVisibility(
    self=[
     Get DecoyButton_Profiles() -> (DecoyButton_Profiles:WBP Bottom Bar Decoy Button Object Reference)
     Get DecoyButton_Profile() -> (DecoyButton_Profile:WBP Bottom Bar Decoy Button Object Reference)
    ]
    ShouldButtonBeVisible=false
   )
   n26 = Get Set Leaderboard Type Async Task Ref()
   > n27 = Is Valid(InputObject=@n26)
   > n28 = StopListeningToEvent(self=@n26)
   n29 = StartListeningToEvent!(
    execute:Exec=[@n28.then, @n27.Is Not Valid]
    Target:Gameplay Event Router Component Object Reference=GetGlobalEventRouter()
    EventType:Script Struct Object Reference=/Script/CompeteUI.SetTournamentLeaderboardType
    Context:Object Reference
    EventBubblingRule:EEventBubblingRule Enum
   ) -> (
    then:Exec
    AsyncTaskProxy:Async Action Start Listening To Event Object Reference
    OnEventReceived:Exec
    Payload:Set Tournament Leaderboard Type Structure
   )
   > Set Set Leaderboard Type Async Task Ref(Set Leaderboard Type Async Task Ref:Async Action Start Listening To Event Object Reference=@n29) -> (Output_Get:Async Action Start Listening To Event Object Reference)
   n30 = Break Set Tournament Leaderboard Type!(SetTournamentLeaderboardType:"Set Tournament Leaderboard Type Structure (by ref)"=@n29.Payload) -> (
    bIsTournamentSeriesPointLeaderboard:bool=~
    TournamentWatchLiveVM:Watch Live VM Object Reference
    TournamentHeaderVM:Tournament Header VM Object Reference
    PayoutDisplayData:Array of Tournament Leaderboard Payout Table Data Structures
   ) {is=title K2Node_BreakStruct}
   SetIsSeriesPointLeaderboards!(execute=@n29.OnEventReceived, self, IsSeriesPointLeaderboards:bool=@n30) -> (then)
   > Set WatchLiveVM(WatchLiveVM:Watch Live VM Object Reference=@n30.TournamentWatchLiveVM) -> (Output_Get:Watch Live VM Object Reference)
   > Set TournamentHeaderVM(TournamentHeaderVM:Tournament Header VM Object Reference=@n30.TournamentHeaderVM) -> (Output_Get:Tournament Header VM Object Reference)
   > Set Payout Display Data(Payout Display Data:Array of Tournament Leaderboard Payout Table Data Structures=@n30.PayoutDisplayData) -> (Output_Get:Array of Tournament Leaderboard Payout Table Data Structures)
   Event On Activated()
   > n31 = Construct Object from Class(Class=Select(
     Option 0:Tournament Leaderboard VM Class Reference=/Script/CompeteUI.FortTournamentSessionLeaderboardVM
     Option 1:Tournament Leaderboard VM Class Reference=/Script/CompeteUI.FortTournamentSeriesPointLeaderboardVM
     Index=Get IsSeriesPointLeaderboard() -> (IsSeriesPointLeaderboard:bool)
    ) -> (ReturnValue:Tournament Leaderboard VM Class Reference)) -> (ReturnValue:Tournament Leaderboard VM Object Reference)
   > SetLeaderboardVM!(execute, self, Viewmodel:Tournament Leaderboard VM Object Reference=@n31) -> (then)
   > InitializeVM!(
    execute
    self:Tournament Leaderboard VM Object Reference=Get LeaderboardVM()
    InPayoutDisplayData:Array of Tournament Leaderboard Payout Table Data Structures=Get Payout Display Data() -> (Payout Display Data:Array of Tournament Leaderboard Payout Table Data Structures)
   ) -> (then)
   > Initialize!(
    execute
    self:WBP Tournament Leaderboard Teams List Object Reference=Get TeamsList()
    IsSeriesPointLeaderboards:bool=Get IsSeriesPointLeaderboard()
   ) -> (then)
   n32 = Get Header() -> (Header:WBP Tournament Details Header Object Reference)
   > SetFortTournamentHeaderVM!(
    execute
    self:WBP Tournament Details Header Object Reference=@n32
    Viewmodel:Tournament Header VM Object Reference=Get TournamentHeaderVM() -> (TournamentHeaderVM:Tournament Header VM Object Reference)
   ) -> (then)
   > Update Window View Model(self:WBP Tournament Details Header Object Reference=@n32)
   > n33 = Construct Object from Class(Class=/Script/CompeteUI.FortTournamentBackgroundVM) -> (ReturnValue:Tournament Background VM Object Reference)
   > SetupViewModel!(execute, self:Tournament Background VM Object Reference=@n33, InType:EFortTournamentBackgroundImageType Enum=SquarePosterImage) -> (then)
   > SetTournament Background VM!(
    execute
    self:WBP Tournament Background Object Reference=Get Background() -> (Background:WBP Tournament Background Object Reference)
    Viewmodel:Tournament Background VM Object Reference=@n33
   ) -> (then)
   > SetWatchLiveAvailable!(
    execute
    self
    Avaliable:bool=CanWatchLiveGames!(self:Fort Compete UIManager Object Reference=Get FortCompeteUIManager()) -> (ReturnValue:bool)
   ) -> (then)
   > SetTournamentName!(
    execute
    self:WBP Tournament Details Header Object Reference=Get Header()
    TournamentName:Text=Get TournamentName!(self:Tournament Leaderboard VM Object Reference=Get LeaderboardVM()) -> (TournamentName:Text)
   ) -> (then)
   > FilterPercentilePayoutData!(execute, self, InPayoutTableData:Array of Tournament Leaderboard Payout Table Data Structures=Get Payout Display Data()) -> (then)
   > OnParentActivated(self:WBP Tournament Leaderboard Teams List Object Reference=Get TeamsList())
  ]}
 FilterPercentilePayoutData {nodes [
   n1 = FilterPercentilePayoutData!() -> (then, InPayoutTableData:Array of Tournament Leaderboard Payout Table Data Structures) {is=title K2Node_FunctionEntry}
   > Clear(TargetArray:Array of Tournament Payout Threshold Data Structures=Get PercentilePayoutThresholds())
   > n2 = For Each Loop!(Exec, Array:Array of Tournament Leaderboard Payout Table Data Structures=@n1) -> (
    LoopBody
    Array Element_bIsMainWindowLeaderboard:bool=~
    Array Element_PayoutThresholdsData:Array of Tournament Payout Threshold Data Structures
    Array Index
    Completed
   )
   > n3 = For Each Loop with Break!(Exec:Exec, Array:Array of Tournament Payout Threshold Data Structures=@n2.Array Element_PayoutThresholdsData, Break:Exec) -> (LoopBody:Exec, Array Element:Tournament Payout Threshold Data Structure, Array Index:int, Completed:Exec)
   > Branch(Condition="Equal (Enum)"(
     A:EPayoutScoringType Enum=Break Tournament Payout Threshold Data!(TournamentPayoutThresholdData:"Tournament Payout Threshold Data Structure (by ref)"=@n3) -> (
      ScoringType:EPayoutScoringType Enum
      Threshold:double=~
      ThresholdDisplayText:Text
      ThresholdDisplayTextHeader:Text
      SourceLeaderboardDisplayName:Text
      PayoutItemsData:Array of Tournament Payout Item Data Structures
      CosmeticPayoutItemsVM:Array of Fort Item VM Object References
     ) {is=title K2Node_BreakStruct}
     B:EPayoutScoringType Enum=Percentile
    ))
   > Add Unique(
    TargetArray:Array of Tournament Payout Threshold Data Structures=Get PercentilePayoutThresholds()
    NewItem:"Tournament Payout Threshold Data Structure (by ref)"=@n3
   )
  ]}
 OnInputMethodChanged {nodes [
   n1 = OnInputMethodChanged!() -> (then, bNewInputType:ECommonInputType Enum) {is=title K2Node_FunctionEntry}
   > n2 = Branch(Condition=AreModalsActive!(LocalPlayerController:Fort Player Controller Object Reference=Get FortPlayerController() -> (FortPlayerController:Fort Player Controller Object Reference)) -> (ReturnValue:bool))
   n3 = Switch on ECommonInputType!(execute=@n2.else, Selection:ECommonInputType Enum=@n1) -> (MouseAndKeyboard:Exec, Gamepad:Exec, Touch:Exec, Count:Exec) {is=title K2Node_SwitchEnum}
   RequestRefreshFocus(execute=@n3.Gamepad, self:Common Activatable Widget Object Reference)
  ]}
 OnTeamListItemSelected {nodes [
   n1 = OnTeamListItemSelected!() -> (then, Item:Object Reference, bIsSelected:bool=~) {is=title K2Node_FunctionEntry}
   n2 = Get TeamsList()
   n3 = "Get (a copy)"(
    Array:Array of Tournament Leaderboard Entry VM Object References=GetLeaderboardEntries!(self:Tournament Leaderboard VM Object Reference=Get LeaderboardVM()) -> (ReturnValue:Array of Tournament Leaderboard Entry VM Object References)
    Dimension 1=GetIndexForItem!(self:List View Object Reference=Get CommonListView_TeamsInfo(self=@n2), Item:Object Reference=@n1) -> (ReturnValue:int=0)
   ) -> (Output:Tournament Leaderboard Entry VM Object Reference)
   > Branch(Condition=AND Boolean(A=@n1.bIsSelected, B=IsValid(Object=@n3)))
   n4 = Get PlayersProfiles!(self:Fort Tournament Base Entry VM Object Reference=@n3) -> (PlayersProfiles:Map of Strings to Tournament Player Profile VM Object References)
   > Set SingleProfileSelected(SingleProfileSelected:bool="Equal (Integer)"(
     A=Length!(TargetMap:Map of Strings to Tournament Player Profile VM Object References=@n4) -> (ReturnValue:int=0) {is=title K2Node_CallFunction}
     B=1
    )) -> (Output_Get:bool)
   > n5 = Set CurrentProfileButton(CurrentProfileButton:WBP Bottom Bar Decoy Button Object Reference=Select(
     Option 0:WBP Bottom Bar Decoy Button Object Reference=Get DecoyButton_Profiles()
     Option 1:WBP Bottom Bar Decoy Button Object Reference=Get DecoyButton_Profile()
     Index=Get SingleProfileSelected()
    ) -> (ReturnValue:WBP Bottom Bar Decoy Button Object Reference)) -> (Output_Get:WBP Bottom Bar Decoy Button Object Reference)
   n6 = Get DecoyButton_Profiles()
   > SetBottomButtonVisibility(self=Select(Option 0=@n6, Option 1=Get DecoyButton_Profile(), Index="Equal (Object)"(A=@n5, B=@n6)), ShouldButtonBeVisible=false)
   > UpdateProfileButtonVisibility()
  ]}
 OpenSinglePlayerProfile {nodes [
   n1 = OpenSinglePlayerProfile!() -> (then, TeamsInfo:Fort Tournament Base Entry VM Object Reference) {is=title K2Node_FunctionEntry}
   > n2 = Values!(execute, TargetMap:Map of Strings to Tournament Player Profile VM Object References=Get PlayersProfiles(self=@n1)) -> (then, Values:Array of Tournament Player Profile VM Object References)
   > ShowCareerStatsScreen!(
    execute
    self:Fort Compete UIManager Object Reference=Get FortCompeteUIManager()
    AccountId:String=Get AccountId!(self:Tournament Player Profile VM Object Reference="Get (a copy)"(Array:Array of Tournament Player Profile VM Object References=@n2) -> (Output:Tournament Player Profile VM Object Reference)) -> (AccountId:String)
    GameMode:String=br-combined
    bIsRanked:bool=true
    InSeasonTrack:String=GetLatestSeasonTrackGUID!(
     self:Fort Stats Page Manager Object Reference=Get FortStatsPageManager() -> (ReturnValue:Fort Stats Page Manager Object Reference) {is=title K2Node_GetSubsystem}
     GameMode:"Name (by ref)"=MakeLiteralName(Value=br-combined)
    ) -> (ReturnValue:String)
    InLimitedGameModeIndex:int=-1
   ) -> (then)
  ]}
 Set Leaderboard Entries {nodes [
   n1 = Set Leaderboard Entries!() -> (then, New Entries:Array of Tournament Leaderboard Entry VM Object References) {is=title K2Node_FunctionEntry}
   n2 = NOT Boolean(A=OR Boolean(
     A=Is Empty!(TargetArray:Array of Tournament Leaderboard Entry VM Object References=@n1) -> (ReturnValue:bool)
     B=Get IsSeriesPointLeaderboard()
    ))
   > n3 = Set ProfilesButtonVis(ProfilesButtonVis:bool=@n2) -> (Output_Get:bool)
   n4 = Get bIsTournamentMode!(self:Fort Runtime Options Object Reference=GetRuntimeOptions()) -> (bIsTournamentMode:bool)
   > SetBottomButtonVisibility(
    self=Get DecoyButton_MatchHistory() -> (DecoyButton_MatchHistory:WBP Bottom Bar Decoy Button Object Reference)
    ShouldButtonBeVisible=AND Boolean(A=@n3, B=NOT Boolean(A=@n4))
   )
   n5 = Get Stats() -> (Stats:WBP Tournament Leaderboard Stats Object Reference)
   > SetDataVisibility!(execute, self:WBP Tournament Leaderboard Stats Object Reference=@n5, Visible:bool=@n2) -> (then)
   > SetDataVisibility(self:WBP Tournament Leaderboard Teams List Object Reference=Get TeamsList(), Visible=@n2)
   > SetFortTournamentLeaderboardVM(self:WBP Tournament Leaderboard Stats Object Reference=@n5, Viewmodel=Get LeaderboardVM())
   > UpdateProfileButtonVisibility()
   > UpdateCanShowScoringRules()
   > n6 = Branch(Condition=Get IsSeriesPointLeaderboard())
   n7 = Get FortPlayerController()
   n8 = AreModalsActive(LocalPlayerController=@n7)
   n9 = HasUserFocusedDescendants!(self:Widget Object Reference, PlayerController:Player Controller Object Reference=@n7) -> (ReturnValue:bool)
   Branch(execute=[@n10.then, @n6.then], Condition=AND Boolean(A=NOT Boolean(A=@n8), B=NOT Boolean(A=@n9)))
   > RequestRefreshFocus()
   n10 = SetTriggeringInputAction!(
    execute=@n6.else
    self:Common Button Base Object Reference=Get DecoyButton_QualificationRules() -> (DecoyButton_QualificationRules:WBP Bottom Bar Decoy Button Object Reference)
    InputActionRow:"Data Table Row Handle Structure (by ref)"=Select(
     Option 0:"Data Table Row Handle Structure (by ref)"=Get QualificationRulesInput() -> (QualificationRulesInput:Data Table Row Handle Structure)
     Option 1:"Data Table Row Handle Structure (by ref)"=Get ScoringRulesInput() -> (ScoringRulesInput:Data Table Row Handle Structure)
     Index=Get bCanShowScoringRules()
    ) -> (ReturnValue:"Data Table Row Handle Structure (by ref)")
   ) -> (then)
  ]}
 SetIsSeriesPointLeaderboards {nodes [
   n1 = SetIsSeriesPointLeaderboards!() -> (then, IsSeriesPointLeaderboards:bool=~) {is=title K2Node_FunctionEntry}
   > n2 = Set IsSeriesPointLeaderboard(IsSeriesPointLeaderboard:bool=@n1) -> (Output_Get:bool)
   > SetBottomButtonVisibility(
    self=[Get DecoyButton_WatchLive() -> (DecoyButton_WatchLive:WBP Bottom Bar Decoy Button Object Reference), Get DecoyButton_QualificationRules()]
    ShouldButtonBeVisible=NOT Boolean(A=@n2)
   )
  ]}
 SetMobileStyle {nodes [
   SetMobileStyle() {is=title K2Node_FunctionEntry}
   > Branch(Condition=IsMobileGame())
   > SetGridSlotPadding!(
    execute
    Widget:Widget Object Reference=Get Main() -> (Main:Grid Panel Object Reference)
    Padding_Left:float=48
    Padding_Top:float=30
    Padding_Right:float=48
    Padding_Bottom:float=30
   ) -> (then)
   n1 = Get SizeBox_Content() -> (SizeBox_Content:Size Box Object Reference)
   > ClearWidthOverride(self:Size Box Object Reference=@n1)
   n2 = SlotAsGridSlot(Widget=@n1)
   > SetHorizontalAlignment(self:Grid Slot Object Reference=@n2, InHorizontalAlignment=HAlign_Fill)
   n3 = Get!(self:Grid Slot Object Reference=@n2) -> (Padding_Left:float=~, Padding_Top:float=~, Padding_Right:float=~, Padding_Bottom:float=~)
   > SetPadding!(
    execute
    self:Grid Slot Object Reference=@n2
    InPadding_Left:float=48
    InPadding_Top:float=@n3.Padding_Top
    InPadding_Right:float=48
    InPadding_Bottom:float=@n3.Padding_Bottom
   ) -> (then)
   n4 = Get SizeBox_TeamsList() -> (SizeBox_TeamsList:Size Box Object Reference)
   > ClearWidthOverride(self=@n4)
   > SetHorizontalAlignment(self=SlotAsGridSlot(Widget=@n4), InHorizontalAlignment=HAlign_Fill)
   > SetPadding(self=SlotAsGridSlot(Widget=Get Header()), InPadding_Left=0, InPadding_Top=0, InPadding_Right=0, InPadding_Bottom=0)
  ]}
 SetProfilesAvailable {nodes [
   n1 = SetProfilesAvailable!() -> (then, Avaliable:bool=~) {is=title K2Node_FunctionEntry}
   > n2 = Get!(execute:Exec) -> (then:Exec, else:Exec, CurrentProfileButton:WBP Bottom Bar Decoy Button Object Reference)
   n3 = Get IsSeriesPointLeaderboard()
   > SetBottomButtonVisibility(self=@n2, ShouldButtonBeVisible=AND Boolean(A=@n1, B=NOT Boolean(A=@n3)))
  ]}
 SetRegion {nodes [
   n1 = SetRegion!() -> (then, Region:Text) {is=title K2Node_FunctionEntry}
   > "SetText (Text)"(self=Get CommonText_Region() -> (CommonText_Region:Common Text Object Reference), InText=@n1)
  ]}
 SetWatchLiveAvailable {nodes [
   n1 = SetWatchLiveAvailable!() -> (then, Avaliable:bool=~) {is=title K2Node_FunctionEntry}
   n2 = Get IsSeriesPointLeaderboard()
   > SetBottomButtonVisibility(self=Get DecoyButton_WatchLive(), ShouldButtonBeVisible=AND Boolean(A=@n1, B=NOT Boolean(A=@n2)))
  ]}
 UpdateCanShowScoringRules {nodes [
   UpdateCanShowScoringRules() {is=title K2Node_FunctionEntry}
   > n1 = IsQualificationRulesPercentileFeatureEnabled!(execute) -> (then, ReturnValue:bool)
   n2 = Get PercentilePayoutThresholds()
   n3 = Get CommonListView_TeamsInfo(self=Get TeamsList())
   n4 = GetNumItems(self=@n3) -> (ReturnValue=0)
   > n5 = Branch(Condition=AND Boolean!(
     A=@n1
     B=Is Not Empty!(TargetArray:Array of Tournament Payout Threshold Data Structures=@n2) -> (ReturnValue:bool)
     C:bool="integer > integer"(A=@n4, B=1)
    ) -> (ReturnValue))
   > n6 = Cast To FortTournamentLeaderboardEntryData(Object=GetItemAt!(self:List View Object Reference=@n3, Index:int=1) -> (ReturnValue:Object Reference))
   n7 = Get Percentile!(self:Fort Tournament Base Entry VM Object Reference=@n6) -> (Percentile:double)
   > n8 = Set bCanShowScoringRules(bCanShowScoringRules:bool="float > float"(A=@n7)) -> (Output_Get:bool)
   > Branch(Condition=@n8)
   > Set TotalTournamentEntries(TotalTournamentEntries:int=Round(A=float / float(A=2, B=@n7)) {is=title K2Node_CallFunction}) -> (Output_Get:int=0)
   Set bCanShowScoringRules(execute=[@n6.CastFailed, @n5.else], bCanShowScoringRules=false)
  ]}
 UpdateProfileButtonVisibility {nodes [
   UpdateProfileButtonVisibility() {is=title K2Node_FunctionEntry}
   > n1 = Get()
   > SetBottomButtonVisibility(self=@n1, ShouldButtonBeVisible=Get ProfilesButtonVis() -> (ProfilesButtonVis:bool))
  ]}
]
variables [
 BackgroundMaterial {type=MaterialInstanceDynamic}
 bCanShowScoringRules {type=boolean}
 CurrentProfileButton {type=WBP_BottomBarDecoyButton_C, default=null}
 FortPlayerController {type=FortPlayerController, default=null}
 IsSeriesPointLeaderboard {type=boolean}
 Payout Display Data {type="array<TournamentLeaderboardPayoutTableData>", default []}
 PercentilePayoutThresholds {type="array<TournamentPayoutThresholdData>", default []}
 ProfilesButtonVis {type=boolean, default=true}
 QualificationRulesInput {type=DataTableRowHandle, default {dataTable=/Compete/DataTables/DT_CompeteUIInputActions, rowName=QualificationRules}}
 ScoringRulesInput {type=DataTableRowHandle, default {dataTable=&DT_CompeteUIInputActions, rowName=ScoringRules}}
 Set Leaderboard Type Async Task Ref {type=AsyncAction_StartListeningToEvent, default=null}
 SingleProfileSelected {type=boolean}
 TotalTournamentEntries {type=integer}
 TournamentHeaderVM {type=FortTournamentHeaderVM, default=null}
 WatchLiveVM {type=FortTournamentWatchLiveVM, default=null}
]
widgets {
 info {inheritedWidgetCount=0, parentClass=&CommonActivatableWidget, rootWidgetClass=&Overlay, widgetCount=19}
 widgets [
  Overlay_0 {widgetClassPath=&Overlay, widget=Overlay_0}
  Background {
   widgetClassPath=/Compete/TournamentDetails/WBP_TournamentBackground
   bIsVariable=true
   parent=Overlay_0
   widget=Background
   properties {backgroundMaterial=/Compete/MaterialInstances/MI_UI_Compete_LeaderboardBackground}
   slot {class=OverlaySlot, properties {horizontalAlignment=HAlign_Fill, verticalAlignment=VAlign_Fill}}
  }
  CaptureForPostBufferUpdate {
   widgetClassPath=&WBP_CaptureForPostBufferUpdate
   bIsVariable=true
   parent=Overlay_0
   widget=CaptureForPostBufferUpdate
   slot {class=OverlaySlot, properties {horizontalAlignment=HAlign_Fill, verticalAlignment=VAlign_Fill}}
  }
  SafeZone_Content {
   widgetClassPath=&SafeZone
   bIsVariable=true
   parent=Overlay_0
   widget=SafeZone_Content
   slot {class=OverlaySlot, properties {horizontalAlignment=HAlign_Fill, verticalAlignment=VAlign_Fill}}
  }
  Root {
   widgetClassPath=&GridPanel
   parent=SafeZone_Content
   widget=Root
   properties {columnFill [1], rowFill [1,0.1]}
   slot {class=SafeZoneSlot}
  }
  SizeBox_Content {
   widgetClassPath=&SizeBox
   bIsVariable=true
   parent=Root
   widget=SizeBox_Content
   properties {bOverride_WidthOverride=true, widthOverride=1773}
   slot {class=GridSlot, properties {horizontalAlignment=HAlign_Center, layer=7, rowSpan=1, padding=<ltrb 0 69 0 30>}}
  }
  Main {
   widgetClassPath=&GridPanel
   bIsVariable=true
   parent=SizeBox_Content
   widget=Main
   properties {columnFill [1], rowFill [0,1]}
   slot {class=SizeBoxSlot}
  }
  Header {
   widgetClassPath=/Compete/TournamentDetails/WBP_TournamentDetailsHeader
   bIsVariable=true
   parent=Main
   widget=Header
   slot {class=GridSlot, properties {horizontalAlignment=HAlign_Left, verticalAlignment=VAlign_Top, padding=<ltrb 15 0 0 0>}}
  }
  SizeBox_TeamsList {
   widgetClassPath=&SizeBox
   bIsVariable=true
   parent=Main
   widget=SizeBox_TeamsList
   properties {bOverride_WidthOverride=true, widthOverride=1108}
   slot {class=GridSlot, properties {horizontalAlignment=HAlign_Left, row=1, padding=<ltrb 0 40 97 0>}}
  }
  TeamsList {
   widgetClassPath=&/WBP_TournamentLeaderboardTeamsList
   bIsVariable=true
   parent=SizeBox_TeamsList
   widget=TeamsList
   properties {bDisplayInActionBar=true, priority=10}
   slot {class=SizeBoxSlot}
  }
  Stats {
   widgetClassPath=&/WBP_TournamentLeaderboardStats
   bIsVariable=true
   parent=Main
   widget=Stats
   slot {class=GridSlot, properties {column=1, horizontalAlignment=HAlign_Right, row=1, verticalAlignment=VAlign_Top, padding=<ltrb 0 14 0 0>}}
  }
  SafeZone_1 {
   widgetClassPath=&SafeZone
   bIsVariable=true
   parent=Overlay_0
   widget=SafeZone_1
   properties {padBottom=false, padLeft=false}
   slot {class=OverlaySlot, properties {horizontalAlignment=HAlign_Right, padding=<ltrb 0 90 80 0>}}
  }
  CommonText_Region {
   widgetClassPath=&CommonTextBlock
   bIsVariable=true
   parent=SafeZone_1
   widget=CommonText_Region
   properties {
    scrollStyle=&TextScrollStyle-Base_C
    text=NAE
    font {
     bForceMonospaced=false
     bMaterialIsStencil=false
     fontMaterial=null
     fontObject=&HeadingNow
     letterSpacing=0
     monospacedWidth=1
     size=18
     skewAmount=0
     typefaceFontName=75 Medium
     outlineSettings {
      bApplyOutlineToDropShadows=false
      bMiteredCorners=false
      bSeparateFillAlpha=false
      outlineMaterial=null
      outlineSize=2
      outlineColor=<rgba 0 0 0 0.2>
     }
    }
    shadowOffset=<0 0>
   }
   slot {class=SafeZoneSlot}
  }
  DecoyButton_Back {
   widgetClassPath=/Game/UI/Frontend/BottomBar/WBP_BottomBarDecoyButton
   bIsVariable=true
   parent=Overlay_0
   widget=DecoyButton_Back
   properties {
    holdData=&CommonUI_HoldData_Base_C
    triggeringInputAction {dataTable=&InputActionDataTable, rowName=Back}
   }
   slot {class=OverlaySlot}
  }
  DecoyButton_MatchHistory {
   widgetClassPath=&WBP_BottomBarDecoyButton
   bIsVariable=true
   parent=Overlay_0
   widget=DecoyButton_MatchHistory
   properties {
    holdData=&CommonUI_HoldData_Base_C
    triggeringInputAction {dataTable=&DT_CompeteUIInputActions, rowName=MatchHistory}
   }
   slot {class=OverlaySlot}
  }
  DecoyButton_Profiles {
   widgetClassPath=&WBP_BottomBarDecoyButton
   bIsVariable=true
   parent=Overlay_0
   widget=DecoyButton_Profiles
   properties {
    holdData=&CommonUI_HoldData_Base_C
    triggeringInputAction {dataTable=&DT_CompeteUIInputActions, rowName=Profiles}
   }
   slot {class=OverlaySlot}
  }
  DecoyButton_Profile {
   widgetClassPath=&WBP_BottomBarDecoyButton
   bIsVariable=true
   parent=Overlay_0
   widget=DecoyButton_Profile
   properties {
    holdData=&CommonUI_HoldData_Base_C
    triggeringInputAction {dataTable=&DT_CompeteUIInputActions, rowName=Profile}
   }
   slot {class=OverlaySlot}
  }
  DecoyButton_QualificationRules {
   widgetClassPath=&WBP_BottomBarDecoyButton
   bIsVariable=true
   parent=Overlay_0
   widget=DecoyButton_QualificationRules
   properties {
    holdData=&CommonUI_HoldData_Base_C
    triggeringInputAction {dataTable=&DT_CompeteUIInputActions, rowName=QualificationRules}
   }
   slot {class=OverlaySlot}
  }
  DecoyButton_WatchLive {
   widgetClassPath=&WBP_BottomBarDecoyButton
   bIsVariable=true
   parent=Overlay_0
   widget=DecoyButton_WatchLive
   properties {
    holdData=&CommonUI_HoldData_Base_C
    triggeringInputAction {dataTable=&DT_CompeteUIInputActions, rowName=WatchLive}
   }
   slot {class=OverlaySlot}
  }
 ]
}
